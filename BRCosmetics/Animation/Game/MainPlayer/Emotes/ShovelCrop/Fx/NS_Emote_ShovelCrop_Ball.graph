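# Niagara effect for the ShovelCrop emote: glowing balls with lightning, on a six-second timeline.
# NE_Ball bursts at timed beats on sampled skeletal mesh locations; three Lightning emitters spawn from its particles via attribute readers.
# Lightning_Chain stretches meshes between indexed position-array points using a scratch module.
# User.Switch selects white, black or gold looks, User.Star gates visibility, SpawnTime_Offset shifts bursts, and camera distance fades alpha.
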NiagaraSystem /BRCosmetics/Animation/Game/MainPlayer/Emotes/ShovelCrop/Fx/NS_Emote_ShovelCrop_Ball
dependencies {
 usedDataInterfaces [
  &NiagaraDataInterfaceCamera
  &NiagaraDataInterfaceArrayFloat
  &NiagaraDataInterfaceCurve
  &NiagaraDataInterfaceVector2DCurve
  &NiagaraDataInterfaceColorCurve
  &NiagaraDataInterfaceArrayInt32
  &NiagaraDataInterfaceParticleRead
  &NiagaraDataInterfaceSpriteRendererInfo
  &NiagaraDataInterfaceSkeletalMesh
  &NiagaraDataInterfaceArrayPosition
 ]
 usedDynamicInputs [&MergeHelper_VectorSwitch]
 usedModules [
  &self:SystemSpawnScript.SystemScriptSource.SystemScriptGraph.NiagaraNodeAssignment_0.SetVariables_#1
  &self:SystemSpawnScript.SystemScriptSource.SystemScriptGraph.NiagaraNodeAssignment_3.SetVariables_#2
  &SystemState
 ]
}
emitters {
 Lightning01 {
  emitterUpdateScript [
   &EmitterState(Inactive Response=::NewEnumerator1, Life Cycle Mode=::NewEnumerator1, Loop Duration=6, UseLoopDelay=false)
   &SpawnParticlesFromOtherEmitter(
    Attribute Reader=DataInterface() {propertyValues {EmitterBinding {bindingMode=Other, emitterName=NE_Ball}}}
    Calculate Spawn Rate Per Particle=false
    Spawn Rate=&FloatFromCurve(FloatCurve=DataInterface() {propertyValues {Curve {keys [<at 0 100>,<at 1 400>]}}}, CurveIndex=$Emitter.NormalizedLoopAge)
    Use Maximum Age=false
    Use Minimum Age=false
   )
  ]
  particleSpawnScript [
   &InitializeParticle(
    Color=<rgba 5 5 5 1>
    Lifetime Max=0.15
    Lifetime Min=0.1
    Lifetime Mode=::NewEnumerator1
    Mesh Scale=&RandomRangeVector(Minimum=<1.2 1.2 1.2>, Maximum=<1.5 1.5 1.5>)
    Mesh Scale Mode=::NewEnumerator4
    Position Offset=&RandomRangeVector(Minimum=<-2 -3 -18>, Maximum=<2 -3 -10>)
    UsePositionOffset=false
    Write Material Random=true
   )
   &CameraOffset(Camera Offset Amount=5)
   &ShapeLocation(
    MaskSpawnGroup=false
    Override Seed=false
    Rotation Axis=<0 1 0>
    Rotation Mode=::NewEnumerator1
    Shape Origin=&VectorFromFloat(Value=0)
    Sphere Radius=5
    UseOwnerScale=true
   )
   &AddVelocity(Module.Use Falloff Distance=false, Override Seed=false, Velocity Mode=::NewEnumerator1, Velocity Speed=1)
   /SetVariables(Particles.Random=&RandomRangeFloat(Minimum=0, Maximum=1)) {isSetParameters=true}
   &InitialMeshOrientation(Mesh Orientation Mode=::NewEnumerator7, Rotation=<0 0.25 0>, Use Rotation Vector=true) {module=InitialMeshOrientation002}
   &InitialMeshOrientation(
    Mesh Orientation Mode=::NewEnumerator2
    Rotation=&RandomRangeVector(Minimum=<0 0 0>, Maximum=<1 0 0>)
    Use Rotation Vector=true
   ) {module=InitialMeshOrientation003}
   &SampleParticlesFromOtherEmitter()
  ]
  particleUpdateScript [
   &ParticleState()
   /SetVariables(Particles.Cam_float:NiagaraFloat=&RemapRange(
     Input Value To Remap=/Niagara/DynamicInputs/Camera/DistanceToCamera(Use Owner Position:NiagaraBool=true, Camera Query:NiagaraDataInterfaceCamera=DataInterface())
     Input Min=80
     Input Max=200
     Output Min=0.1
     Output Max=1
    )) {isSetParameters=true}
   &ScaleMeshSize(Scale Factor=&Multiply_VectorByFloat(
     Vector=<1 1 1>
     Float=&FloatFromCurve(
      FloatCurve=DataInterface() {propertyValues {Curve {keys [<at 0 1>,<at 0.5 1>,<at 1 1.2>]}}}
      Scale Curve=&RandomRangeFloat(Minimum=0.3, Maximum=0.85)
     )
    ))
   &ScaleColor(
    Scale Alpha=&Multiply_Float(
     A=$Particles.Cam_float
     B=&MakeCustomFloatFromBool(
      Bool=&SetBoolByFloatComparison(A=$User.Switch, B=1)
      True Float=10
      False Float=&MakeCustomFloatFromBool(Bool=&SetBoolByFloatComparison(A=$User.Switch, B=0), True Float=10, False Float=10)
     )
    )
    Scale RGB=&MakeCustomVectorFromBool(
     Bool=&SetBoolByFloatComparison(A=$User.Switch, B=1)
     True Vector=&MakeVectorFromLinearColorRGB(Color=<rgba 1 1 1 1>)
     False Vector=&MakeCustomVectorFromBool(
      Bool=&SetBoolByFloatComparison(A=$User.Switch, B=0)
      True Vector=&MakeVectorFromLinearColorRGB(Color=<rgba 0 0 0 1>)
      False Vector=&MakeVectorFromLinearColorRGB(Color=&ColorFromCurve(ColorCurve=DataInterface() {propertyValues {AlphaCurve {keys []}, BlueCurve {keys [<at 0.5854 0.4688>]}, GreenCurve {keys [<at 0.5854 4.2129>]}, RedCurve {keys [<at 0.5854 5>]}}}))
     )
    )
    ScaleA=true
    ScaleRGB=true
   )
   &DynamicMaterialParameters(
    Index 0 Param 1=&FloatFromCurve(FloatCurve=DataInterface() {propertyValues {Curve {keys [{tangentMode=RCTM_User, time=0, value=0}, {tangentMode=RCTM_User, time=1, value=1}]}}})
    Index 0 Param 2=0.7
    Index 0 Param 3=&FloatFromCurve(FloatCurve=DataInterface() {propertyValues {Curve {keys [<at 0 0 RCIM_Cubic>, {arriveTangent=-0.3684, interpMode=RCIM_Cubic, leaveTangent=-0.3684, tangentMode=RCTM_User, time=1, value=3}]}}})
    Index 0 Param 4=1
    Index 1 Param 1=0.8
    Index 1 Param 3=&Lerp_Float(A=1, B=1.3, Alpha=$Particles.Random)
    Index 1 Param 4=1
    Module.Param0WriteEnabled=true
    Module.Param0WriteEnabled1=true
    Module.Param1WriteEnabled=true
    Module.Param1WriteEnabled1=true
    Module.Param2WriteEnabled=true
    Module.Param2WriteEnabled1=true
    Module.Param3WriteEnabled=true
    Module.Param3WriteEnabled1=true
    Write Parameter Index 1=true
   )
   /SetVariables(Particles.Glow_color:LinearColor=&MakeLinearColorFromVectorAndFloat(
     "Vector (RGB)"=&MakeCustomVectorFromBool(
      Bool=&SetBoolByFloatComparison(A=$User.Switch, B=1)
      True Vector=&MakeVectorFromLinearColorRGB(Color=<rgba 1 1 1 1>)
      False Vector=&MakeCustomVectorFromBool(
       Bool=&SetBoolByFloatComparison(A=$User.Switch, B=0)
       True Vector=&MakeVectorFromLinearColorRGB(Color=<rgba 0 0 0 1>)
       False Vector=&MakeVectorFromLinearColorRGB(Color=&ColorFromCurve(ColorCurve=DataInterface() {propertyValues {AlphaCurve {keys []}, BlueCurve {keys [<at 0.6385 0>]}, GreenCurve {keys [<at 0.6385 0.8049>]}, RedCurve {keys [<at 0.6385 2>]}}}))
      )
     )
     "Float (Alpha)"=&Multiply_Float(
      A=$Particles.Cam_float
      B=&MakeCustomFloatFromBool(
       Bool=&SetBoolByFloatComparison(A=$User.Switch, B=1)
       True Float=0
       False Float=&MakeCustomFloatFromBool(Bool=&SetBoolByFloatComparison(A=$User.Switch, B=0), True Float=0, False Float=3)
      )
     )
    )) {isSetParameters=true}
   &SolveForcesAndVelocity(Clamp Velocity=false, Limit Acceleration=false)
   /SetVariables(
    Particles.Glow Dy02:Vector4f=&MakeVector4(
     X=&FloatFromCurve(FloatCurve=DataInterface() {propertyValues {Curve {keys [{tangentMode=RCTM_User, time=0, value=0.2}, {tangentMode=RCTM_User, time=0.5, value=1}]}}})
     Y=0.7
     Z=0
     W=1
    )
    Particles.Glow Dy03:Vector4f=&MakeVector4(
     X=&FloatFromCurve(FloatCurve=DataInterface() {propertyValues {Curve {keys [<at 0.5 1>,<at 1 0>]}}}, Scale Curve=0.8)
     Y=0
     Z=&Lerp_Float(A=1, B=1.3, Alpha=$Particles.Random)
     W=1
    )
   ) {isSetParameters=true}
   &KillParticles(Kill Particles=&SetBoolByFloatComparison(A=$User.Star, B=1, Comparison Type=::NewEnumerator4)) {module=KillParticles001}
  ]
  rendererClasses [&NiagaraMeshRendererProperties]
  renderers [
   {rendererClass=&NiagaraMeshRendererProperties, rendererIndex=0}
   {rendererClass=&NiagaraMeshRendererProperties, rendererIndex=1}
  ]
 }
 Lightning02 {
  emitterUpdateScript [
   &EmitterState(Life Cycle Mode=::NewEnumerator1, Loop Behavior=::NewEnumerator1, Loop Delay=2.5, Loop Duration=1.5, UseLoopDelay=true)
   &SpawnParticlesFromOtherEmitter(
    Attribute Reader=DataInterface() {propertyValues {EmitterBinding {bindingMode=Other, emitterName=NE_Ball}}}
    Calculate Spawn Rate Per Particle=false
    Spawn Rate=100
    Use Maximum Age=false
    Use Minimum Age=false
   )
  ]
  particleSpawnScript [
   &InitializeParticle(
    Lifetime=&RandomRangeFloat(Minimum=0.1, Maximum=0.2)
    Mesh Scale Mode=::NewEnumerator3
    Mesh Uniform Scale=&RandomRangeFloat(Minimum=0.5, Maximum=1.5)
    Position Mode=::NewEnumerator1
    Sprite Rotation Mode=::NewEnumerator2
    Sprite Size Max=<120 40>
    Sprite Size Min=<100 20>
    Sprite Size Mode=::NewEnumerator2
    UsePositionOffset=false
    Write Material Random=true
   )
   &InitialMeshOrientation(Rotation=&UniformRangedVector(Minimum=<0 -1 0>, Maximum=<0 1 0>), Use Rotation Vector=false) {module=InitialMeshOrientation001}
   &ShapeLocation(MaskSpawnGroup=false, Override Seed=false, Sphere Radius=0, UseOwnerScale=true)
   &CameraOffset(Camera Offset Amount=5)
   &SampleParticlesFromOtherEmitter(Transform Data Option=1)
   &SubUVAnimation(
    Evaluation Type=::NewEnumerator0
    "Sprite Renderer "=DataInterface() {propertyValues {SpriteRenderer=&self:Lightning02_0.NiagaraSpriteRendererProperties_0}}
    SubUV Animation Mode:ENiagara_SubUVLookupModeV2=::NewEnumerator2
    UseEndFrame=false
    UseStartFrame=false
   )
   &SubUVAnimation(
    Evaluation Type=::NewEnumerator0
    "Sprite Renderer "=DataInterface() {propertyValues {SpriteRenderer=&self:Lightning02_0.NiagaraSpriteRendererProperties_1}}
    SubUV Animation Mode:ENiagara_SubUVLookupModeV2=::NewEnumerator2
    UseEndFrame=false
    UseStartFrame=false
   ) {module=SubUVAnimation001}
  ]
  particleUpdateScript [
   &ParticleState()
   /SetVariables(Particles.Cam_float:NiagaraFloat=&RemapRange(
     Input Value To Remap=&DistanceToCamera(Use Owner Position:NiagaraBool=true, Camera Query:NiagaraDataInterfaceCamera=DataInterface())
     Input Min=80
     Input Max=200
     Output Min=0.1
     Output Max=1
    )) {isSetParameters=true}
   &ScaleSpriteSize(Uniform Curve Sprite Scale=DataInterface() {propertyValues {Curve {keys [<at 0 1 RCIM_Constant>,<at 0.1261 0 RCIM_Constant>,<at 0.2252 1 RCIM_Constant>,<at 0.3506 0 RCIM_Constant>,<at 0.4646 1 RCIM_Constant>,<at 0.5903 0 RCIM_Constant>,<at 0.6578 1 RCIM_Constant>,<at 0.7752 0 RCIM_Constant>,<at 0.8661 1 RCIM_Constant>,<at 1 0 RCIM_Constant>]}}})
   &ScaleMeshSize(Scale Factor=&VectorFromFloat(Value=&Multiply_Float(
      A=1
      B=&FloatFromCurve(FloatCurve=DataInterface() {propertyValues {Curve {keys [<at 0 -0.05>,<at 0.1 0.9>,<at 0.7 0.45>,<at 1 0>]}}})
     ))) {module=ScaleMeshSize001}
   &DynamicMaterialParameters(
    Index 0 Param 1=&FloatFromCurve(FloatCurve=DataInterface() {propertyValues {Curve {
        defaultValue=0
        keys [
         {tangentMode=RCTM_User, time=0, value=0}
         <at 0.5159 0.6424>
         <at 0.7018 0.6518>
         <at 0.7318 0.6894>
         <at 0.9218 0.6988>
         {tangentMode=RCTM_User, time=1, value=0.8}
        ]
       }}})
    Index 0 Param 2=&RandomRangeFloat(Minimum=5, Maximum=20)
    Index 0 Param 3=&FloatFromCurve(
     FloatCurve=DataInterface() {propertyValues {Curve {
        defaultValue=0
        keys [
         {tangentMode=RCTM_User, time=0.484, value=1}
         <at 0.5159 0.6075>
         <at 0.7018 0.5981>
         <at 0.7318 0.5604>
         <at 0.9218 0.551>
         {tangentMode=RCTM_User, time=1, value=0.3}
        ]
       }}}
     Scale Curve=&RandomRangeFloat(Minimum=5, Maximum=20)
    )
    Index 0 Param 4=&FloatFromCurve(
     FloatCurve=DataInterface() {propertyValues {Curve {
        defaultValue=0
        keys [
         {arriveTangent=5.2125, interpMode=RCIM_Cubic, leaveTangent=5.2125, tangentMode=RCTM_User, time=0, value=0}
         {arriveTangent=0.9556, interpMode=RCIM_Cubic, leaveTangent=0.9556, tangentMode=RCTM_User, time=0.2598, value=0.6352}
         {interpMode=RCIM_Cubic, tangentMode=RCTM_User, time=0.8, value=1}
        ]
       }}}
     CurveIndex=&Multiply_Float(A=$Particles.NormalizedAge, B=&RandomRangeFloat(Minimum=0.6, Maximum=1))
    )
    Index 1 Param 1=&RandomRangeFloat(Minimum=0, Maximum=2)
    Index 1 Param 2=&RandomRangeFloat(Minimum=0.9, Maximum=1.1)
    Index 1 Param 4=&FloatFromCurve(FloatCurve=DataInterface() {propertyValues {Curve {defaultValue=1, keys [{tangentMode=RCTM_User, time=0, value=0}, {tangentMode=RCTM_User, time=0.3, value=1}]}}})
    Module.Index0Float0Write=true
    Module.Index0Float1Write=true
    Module.Index0Float2Write=true
    Module.Index0Float3Write=true
    Module.Index1Float0Write=true
    Module.Index1Float1Write=true
    Module.Index1Float2Write=true
    Module.Index1Float3Write=true
    Write Parameter Index 1=::NewEnumerator1
   )
   /SetVariables(Particles.Glow_color:LinearColor=&MakeLinearColorFromVectorAndFloat(
     "Vector (RGB)"=&MakeCustomVectorFromBool(
      Bool=&SetBoolByFloatComparison(A=$User.Switch, B=1)
      True Vector=&MakeVectorFromLinearColorRGB(Color=<rgba 1 1 1 1>)
      False Vector=&MakeCustomVectorFromBool(
       Bool=&SetBoolByFloatComparison(A=$User.Switch, B=0)
       True Vector=&MakeVectorFromLinearColorRGB(Color=<rgba 0 0 0 1>)
       False Vector=&MakeVectorFromLinearColorRGB(Color=&ColorFromCurve(ColorCurve=DataInterface() {propertyValues {AlphaCurve {keys []}, BlueCurve {keys [<at 0.7274 0>]}, GreenCurve {keys [<at 0.7274 0.6065>]}, RedCurve {keys [<at 0.7274 2>]}}}))
      )
     )
     "Float (Alpha)"=&Multiply_Float(
      A=$Particles.Cam_float
      B=&MakeCustomFloatFromBool(
       Bool=&SetBoolByFloatComparison(A=$User.Switch, B=1)
       True Float=0
       False Float=&MakeCustomFloatFromBool(Bool=&SetBoolByFloatComparison(A=$User.Switch, B=0), True Float=0, False Float=10)
      )
     )
    )) {isSetParameters=true}
   &SolveForcesAndVelocity(Clamp Velocity=false, Limit Acceleration=false)
   &ScaleColor(
    Scale Alpha=&Multiply_Float(
     A=$Particles.Cam_float
     B=&MakeCustomFloatFromBool(
      Bool=&SetBoolByFloatComparison(A=$User.Switch, B=1)
      True Float=10
      False Float=&MakeCustomFloatFromBool(Bool=&SetBoolByFloatComparison(A=$User.Switch, B=0), True Float=10, False Float=1)
     )
    )
    Scale RGB=&MakeCustomVectorFromBool(
     Bool=&SetBoolByFloatComparison(A=$User.Switch, B=1)
     True Vector=&MakeVectorFromLinearColorRGB(Color=<rgba 1 1 1 1>)
     False Vector=&MakeCustomVectorFromBool(
      Bool=&SetBoolByFloatComparison(A=$User.Switch, B=0)
      True Vector=&MakeVectorFromLinearColorRGB(Color=<rgba 0 0 0 1>)
      False Vector=&MakeVectorFromLinearColorRGB(Color=&ColorFromCurve(ColorCurve=DataInterface() {propertyValues {AlphaCurve {keys []}, BlueCurve {keys [<at 0.7274 1.25>]}, GreenCurve {keys [<at 0.7274 8.8729>]}, RedCurve {keys [<at 0.7274 10>]}}}))
     )
    )
    ScaleA=true
    ScaleRGB=true
   )
   &KillParticles(Kill Particles=&SetBoolByFloatComparison(A=$User.Star, B=1, Comparison Type=::NewEnumerator4)) {module=KillParticles001}
  ]
  rendererClasses [&NiagaraSpriteRendererProperties]
  renderers [
   {rendererClass=&NiagaraSpriteRendererProperties}
   {rendererClass=&NiagaraSpriteRendererProperties, rendererIndex=1}
  ]
 }
 Lightning03 {
  emitterUpdateScript [
   &EmitterState(
    Inactive Response=::NewEnumerator1
    Life Cycle Mode=::NewEnumerator1
    Loop Behavior=::NewEnumerator1
    Loop Delay=2.5
    Loop Duration=1.5
    UseLoopDelay=true
   )
   &SpawnParticlesFromOtherEmitter(
    Attribute Reader=DataInterface() {propertyValues {EmitterBinding {bindingMode=Other, emitterName=NE_Ball}}}
    Calculate Spawn Rate Per Particle=false
    Spawn Rate=50
    Use Maximum Age=false
    Use Minimum Age=false
   )
  ]
  particleSpawnScript [
   &InitializeParticle(
    Lifetime Max=0.2
    Lifetime Min=0.15
    Lifetime Mode=::NewEnumerator1
    Mesh Scale Mode=::NewEnumerator3
    Mesh Uniform Scale=&RandomRangeFloat(Minimum=0.5, Maximum=1.2)
    Position Offset=<0 10 0>
    UsePositionOffset=true
    Write Material Random=true
   )
   &CameraOffset(Camera Offset Amount=5)
   &InitialMeshOrientation(Rotation=&UniformRangedVector(Minimum=<0 -1 0>, Maximum=<0 1 0>), Use Rotation Vector=false) {module=InitialMeshOrientation001}
   &SampleParticlesFromOtherEmitter(Transform Data Option=1)
  ]
  particleUpdateScript [
   &ParticleState()
   /SetVariables(Particles.Cam_float:NiagaraFloat=&RemapRange(
     Input Value To Remap=&DistanceToCamera(Use Owner Position:NiagaraBool=true, Camera Query:NiagaraDataInterfaceCamera=DataInterface())
     Input Min=80
     Input Max=200
     Output Min=0.1
     Output Max=1
    )) {isSetParameters=true}
   &ScaleMeshSize(Scale Factor=&Multiply_VectorByFloat(Vector=<1 1 1>, Float=0.7))
   &ScaleColor(
    Scale Alpha=&Multiply_Float(A=$Particles.Cam_float, B=1)
    Scale RGB=&MakeCustomVectorFromBool(
     Bool=&SetBoolByFloatComparison(A=$User.Switch, B=1)
     True Vector=&MakeVectorFromLinearColorRGB(Color=<rgba 1 1 1 1>)
     False Vector=&MakeCustomVectorFromBool(
      Bool=&SetBoolByFloatComparison(A=$User.Switch, B=0)
      True Vector=&MakeVectorFromLinearColorRGB(Color=<rgba 0 0 0 1>)
      False Vector=&MakeVectorFromLinearColorRGB(Color=&ColorFromCurve(ColorCurve=DataInterface() {propertyValues {AlphaCurve {keys []}, BlueCurve {keys [<at 0.7274 1.061>]}, GreenCurve {keys [<at 0.7274 8.8485>]}, RedCurve {keys [<at 0.7274 10>]}}}))
     )
    )
    ScaleA=true
    ScaleRGB=true
   )
   &SolveForcesAndVelocity(Clamp Velocity=false, Limit Acceleration=false)
   &KillParticles(Kill Particles=&SetBoolByFloatComparison(A=$User.Star, B=1, Comparison Type=::NewEnumerator4)) {module=KillParticles001}
  ]
  rendererClasses [&NiagaraMeshRendererProperties]
  renderers [{rendererClass=&NiagaraMeshRendererProperties, rendererIndex=0}]
 }
 Lightning_Chain {
  emitterUpdateScript [
   &EmitterState(Life Cycle Mode=::NewEnumerator1, Loop Behavior=::NewEnumerator1, Loop Duration=6, UseLoopDelay=false)
   &SpawnBurst_Instantaneous(Spawn Time=&Add_Float(A=0.4, B=$User.SpawnTime_Offset), Use Spawn Probability=false, UseLoopCountLimit=false) {module=SpawnBurst_Instantaneous001}
   &SpawnBurst_Instantaneous(
    Spawn Group=1
    Spawn Time=&Add_Float(A=0.75, B=$User.SpawnTime_Offset)
    Use Spawn Probability=false
    UseLoopCountLimit=false
   ) {module=SpawnBurst_Instantaneous002}
   &SpawnBurst_Instantaneous(
    Spawn Group=2
    Spawn Time=&Add_Float(A=1.27, B=$User.SpawnTime_Offset)
    Use Spawn Probability=false
    UseLoopCountLimit=false
   ) {module=SpawnBurst_Instantaneous003}
   &SpawnBurst_Instantaneous(
    Spawn Group=3
    Spawn Time=&Add_Float(A=1.6, B=$User.SpawnTime_Offset)
    Use Spawn Probability=false
    UseLoopCountLimit=false
   ) {module=SpawnBurst_Instantaneous004}
   &SpawnBurst_Instantaneous(
    Spawn Group=4
    Spawn Time=&Add_Float(A=1.95, B=$User.SpawnTime_Offset)
    Use Spawn Probability=false
    UseLoopCountLimit=false
   ) {module=SpawnBurst_Instantaneous005}
   &SpawnBurst_Instantaneous(Spawn Group=5, Spawn Time=2.5, Use Spawn Probability=false, UseLoopCountLimit=false)
   &SpawnBurst_Instantaneous(Spawn Group=6, Spawn Time=3, Use Spawn Probability=false, UseLoopCountLimit=false) {module=SpawnBurst_Instantaneous006}
   &SpawnBurst_Instantaneous(Spawn Group=7, Spawn Time=3.5, Use Spawn Probability=false, UseLoopCountLimit=false) {module=SpawnBurst_Instantaneous007}
   &SpawnBurst_Instantaneous(Spawn Group=8, Spawn Time=3.7, Use Spawn Probability=false, UseLoopCountLimit=false) {module=SpawnBurst_Instantaneous008}
  ]
  particleSpawnScript [
   &InitializeParticle(
    Lifetime=&MergeHelper_FloatSwitch(0=4, 1=4, 2=4, 3=4, 4=4, 5=0.3, 6=0.25, 7=0.1, 8=0.2, Index=&SpawnGroup(), Num=::NewEnumerator7)
    Mesh Scale Mode=::NewEnumerator3
    UsePositionOffset=false
    Write Material Random=true
   )
   /SetVariables(Particles.InitialIndex:NiagaraInt32=$Emitter.SpawnGroup) {isSetParameters=true}
  ]
  particleUpdateScript [
   &ParticleState()
   /SetVariables(Particles.Cam_float:NiagaraFloat=&RemapRange(
     Input Value To Remap=&DistanceToCamera(Use Owner Position:NiagaraBool=true, Camera Query:NiagaraDataInterfaceCamera=DataInterface())
     Input Min=80
     Input Max=200
     Output Min=0.1
     Output Max=1
    )) {isSetParameters=true}
   &self:ScratchModule_03(
    Index:NiagaraInt32=&MakeCustomIntFromBool(
     Bool=&SetBoolByIntComparison(A=$Particles.InitialIndex, B=5, Comparison Type=::NewEnumerator6)
     True INT=$Particles.InitialIndex
     False INT=&SelectIntFromArray(Direct Array Index=$Particles.InitialIndex, Int Selection Array=DataInterface() {propertyValues {IntData [0,0,0,0,0,2,3,2,3]}})
    )
    IndexB:NiagaraInt32=&MakeCustomIntFromBool(
     Bool=&SetBoolByIntComparison(A=$Particles.InitialIndex, B=5, Comparison Type=::NewEnumerator6)
     True INT=&Add_Integer(A=$Particles.InitialIndex, B=1)
     False INT=&SelectIntFromArray(Direct Array Index=$Particles.InitialIndex, Int Selection Array=DataInterface() {propertyValues {IntData [0,0,0,0,0,6,0,6,0]}})
    )
    Position Array:NiagaraDataInterfaceArrayPosition=$System.Position Array
   ) {module=ScratchModule_03}
   &ScaleMeshSize(Scale Factor=&MakeVector(
     X=&Divide_Float(
      A=&Lerp_Float(
       A=0
       B=$Particles.LightningLength
       Alpha=&FloatFromCurve(
        FloatCurve=DataInterface() {propertyValues {Curve {defaultValue=1, keys [{tangentMode=RCTM_User, time=0, value=0}, {tangentMode=RCTM_User, time=0.0262, value=1}]}}}
        Scale Curve=/Niagara/DynamicInputs/Integer/V2/Round_Float(Float:NiagaraFloat=&Waveform(Global Phase=$Particles.Age, "Waveform [1]"=::NewEnumerator0, "[1] Frequency"=10, "[1] Amplitude Min/Max"=<0 1>))
       )
      )
      B=80
     )
     Y=1
     Z=1
    )) {module=ScaleMeshSize001}
   /SetVariables(Particles.Scale2:Vector3f=&Multiply_VectorByFloat(
     Vector=&Multiply_Vector(
      A=&Multiply_VectorByFloat(
       Vector=&MakeVector(
        X=&Divide_Float(
         A=&Lerp_Float(
          A=0
          B=$Particles.LightningLength
          Alpha=&FloatFromCurve(
           FloatCurve=DataInterface() {propertyValues {Curve {defaultValue=1, keys [{tangentMode=RCTM_User, time=0, value=0}, {tangentMode=RCTM_User, time=0.0262, value=1}]}}}
           Scale Curve=&Round_Float(Float:NiagaraFloat=&Waveform(Global Phase=$Particles.Age, "Waveform [1]"=::NewEnumerator0, "[1] Frequency"=8, "[1] Amplitude Min/Max"=<0 1>))
          )
         )
         B=80
        )
        Y=1
        Z=1
       )
       Float=1
      )
      B=$Particles.Initial.Scale
     )
     Float=1
    )) {isSetParameters=true}
   &ScaleColor(
    Scale Alpha=&Multiply_Float(
     A=$Particles.Cam_float
     B=&MakeCustomFloatFromBool(
      Bool=&SetBoolByFloatComparison(A=$User.Switch, B=1)
      True Float=&FloatFromCurve(
       FloatCurve=DataInterface() {propertyValues {Curve {keys [<at 0 1 RCIM_Constant>,<at 0.7 0 RCIM_Constant>]}}}
       CurveIndex=$Emitter.NormalizedLoopAge
      )
      False Float=&MakeCustomFloatFromBool(
       Bool=&SetBoolByFloatComparison(A=$User.Switch, B=0)
       True Float=&FloatFromCurve(
        FloatCurve=DataInterface() {propertyValues {Curve {keys [<at 0 1 RCIM_Constant>,<at 0.7 0 RCIM_Constant>]}}}
        CurveIndex=$Emitter.NormalizedLoopAge
        Scale Curve=10
       )
       False Float=&FloatFromCurve(
        FloatCurve=DataInterface() {propertyValues {Curve {keys [<at 0 1 RCIM_Constant>,<at 0.7 0 RCIM_Constant>]}}}
        CurveIndex=$Emitter.NormalizedLoopAge
       )
      )
     )
    )
    Scale RGB=&MakeVectorFromLinearColorRGB(Color=&MakeCustomLinearColorFromBool(
      Bool=&SetBoolByFloatComparison(A=$User.Switch, B=1)
      True Color=<rgba 1 1 1 1>
      False Color=&MakeCustomLinearColorFromBool(
       Bool=&SetBoolByFloatComparison(A=$User.Switch, B=0)
       True Color=<rgba 0 0 0 1>
       False Color=&ColorFromCurve(ColorCurve=DataInterface() {propertyValues {
          AlphaCurve {keys []}
          BlueCurve {keys [{arriveTangent=3.1759, interpMode=RCIM_Cubic, time=0.0561, value=0.3125}, <at 0.0784 0.1061>]}
          GreenCurve {keys [{arriveTangent=28.4286, interpMode=RCIM_Cubic, time=0.0561, value=2.134}, <at 0.0784 0.95>]}
          RedCurve {keys [{arriveTangent=89.7773, interpMode=RCIM_Cubic, time=0.0561, value=3}, <at 0.0784 3>]}
         }})
      )
     ))
    ScaleA=true
    ScaleRGB=true
   )
   /SetVariables(Particles.Glow_Color:LinearColor=&MakeLinearColorFromVectorAndFloat(
     "Vector (RGB)"=&MakeVectorFromLinearColorRGB(Color=&MakeCustomLinearColorFromBool(
       Bool=&SetBoolByFloatComparison(A=$User.Switch, B=1)
       True Color=<rgba 1 1 1 1>
       False Color=&MakeCustomLinearColorFromBool(
        Bool=&SetBoolByFloatComparison(A=$User.Switch, B=0)
        True Color=<rgba 0 0 0 1>
        False Color=&ColorFromCurve(ColorCurve=DataInterface() {propertyValues {AlphaCurve {keys []}, BlueCurve {keys [<at 0.0784 0>]}, GreenCurve {keys [<at 0.0784 0.5717>]}, RedCurve {keys [<at 0.0784 2>]}}})
       )
      ))
     "Float (Alpha)"=&Multiply_Float(
      A=$Particles.Cam_float
      B=&MakeCustomFloatFromBool(
       Bool=&SetBoolByFloatComparison(A=$User.Switch, B=0)
       True Float=&FloatFromCurve(
        FloatCurve=DataInterface() {propertyValues {Curve {keys [<at 0 1 RCIM_Constant>,<at 0.7 0 RCIM_Constant>]}}}
        CurveIndex=$Emitter.NormalizedLoopAge
        Scale Curve=0
       )
       False Float=&MakeCustomFloatFromBool(
        Bool=&SetBoolByFloatComparison(A=$User.Switch, B=0)
        True Float=&FloatFromCurve(
         FloatCurve=DataInterface() {propertyValues {Curve {keys [<at 0 1 RCIM_Constant>,<at 0.7 0 RCIM_Constant>]}}}
         CurveIndex=$Emitter.NormalizedLoopAge
         Scale Curve=10
        )
        False Float=&FloatFromCurve(
         FloatCurve=DataInterface() {propertyValues {Curve {keys [<at 0 1 RCIM_Constant>,<at 0.7 0 RCIM_Constant>]}}}
         CurveIndex=$Emitter.NormalizedLoopAge
         Scale Curve=0.3
        )
       )
      )
     )
    )) {isSetParameters=true}
   /SetVariables(Particles.Dark_Color:LinearColor=&MakeLinearColorFromVectorAndFloat(
     "Vector (RGB)"=&MakeVectorFromLinearColorRGB(Color=&MakeCustomLinearColorFromBool(
       Bool=&SetBoolByFloatComparison(A=$User.Switch, B=1)
       True Color=<rgba 0 0 0 1>
       False Color=&MakeCustomLinearColorFromBool(
        Bool=&SetBoolByFloatComparison(A=$User.Switch, B=0)
        True Color=<rgba 2 1.3881 0.6409 1>
        False Color=&ColorFromCurve(ColorCurve=DataInterface() {propertyValues {AlphaCurve {keys []}, BlueCurve {keys [<at 0.0784 0>]}, GreenCurve {keys [<at 0.0784 0.5717>]}, RedCurve {keys [<at 0.0784 2>]}}})
       )
      ))
     "Float (Alpha)"=&Multiply_Float(
      A=$Particles.Cam_float
      B=&MakeCustomFloatFromBool(
       Bool=&SetBoolByFloatComparison(A=$User.Switch, B=1)
       True Float=&FloatFromCurve(
        FloatCurve=DataInterface() {propertyValues {Curve {keys [<at 0 1 RCIM_Constant>,<at 0.7 0 RCIM_Constant>]}}}
        CurveIndex=$Emitter.NormalizedLoopAge
        Scale Curve=1000
       )
       False Float=&MakeCustomFloatFromBool(
        Bool=&SetBoolByFloatComparison(A=$User.Switch, B=0)
        True Float=&FloatFromCurve(
         FloatCurve=DataInterface() {propertyValues {Curve {keys [<at 0 1 RCIM_Constant>,<at 0.7 0 RCIM_Constant>]}}}
         CurveIndex=$Emitter.NormalizedLoopAge
         Scale Curve=0.5
        )
        False Float=&FloatFromCurve(
         FloatCurve=DataInterface() {propertyValues {Curve {keys [<at 0 1 RCIM_Constant>,<at 0.7 0 RCIM_Constant>]}}}
         CurveIndex=$Emitter.NormalizedLoopAge
         Scale Curve=0
        )
       )
      )
     )
    )) {isSetParameters=true}
   &DynamicMaterialParameters(
    Index 0 Param 1=&RandomRangeFloat(Minimum=20, Maximum=30)
    Index 0 Param 2=&FloatFromCurve(
     FloatCurve=DataInterface() {propertyValues {Curve {defaultValue=1, keys [{tangentMode=RCTM_User, time=0, value=1}, <at 0.3 1>, {tangentMode=RCTM_User, time=0.5, value=2}]}}}
     CurveIndex=$Emitter.NormalizedLoopAge
     Scale Curve=&RandomRangeFloat(Minimum=2.8, Maximum=3.8)
    )
    Index 0 Param 3=&RandomRangeFloat(Minimum=0, Maximum=0.1)
    Index 0 Param 4=&RandomRangeFloat(Minimum=0, Maximum=0.1)
    Index 1 Param 1=&Multiply_Float(
     A=1
     B=&Waveform(Global Phase=$Particles.Age, "Waveform [1]"=::NewEnumerator0, "[1] Frequency"=5, "[1] Amplitude Min/Max"=<0 1>)
    )
    Index 1 Param 2=&RandomRangeFloat(Minimum=0, Maximum=1)
    Module.Index0Float0Write=true
    Module.Index0Float1Write=true
    Module.Index0Float2Write=true
    Module.Index0Float3Write=true
    Module.Index1Float0Write=true
    Module.Index1Float1Write=true
    Module.Index1Float2Write=true
    Module.Index1Float3Write=true
    Write Parameter Index 1=::NewEnumerator1
   )
   &KillParticles(Kill Particles=&SetBoolByFloatComparison(A=$User.Star, B=1, Comparison Type=::NewEnumerator4)) {module=KillParticles001}
   /SetVariables(
    Particles.Small_Lightning_Dynami1:Vector4f=&MakeVector4(
     X=&Multiply_Float(
      A=1
      B=&Waveform(Global Phase=$Particles.Age, "Waveform [1]"=::NewEnumerator0, "[1] Frequency"=1, "[1] Amplitude Min/Max"=<0 1>, "[1] Scale"=5)
     )
     Y=&RandomRangeFloat(Minimum=0, Maximum=1)
     Z=0
     W=0
    )
    Particles.Small_Lightning_Dynamic0:Vector4f=&MakeVector4(
     X=&RandomRangeFloat(Minimum=5, Maximum=10)
     Y=&FloatFromCurve(
      FloatCurve=DataInterface() {propertyValues {Curve {defaultValue=1, keys [{tangentMode=RCTM_User, time=0, value=1}, <at 0.3 1>, {tangentMode=RCTM_User, time=0.5, value=2}]}}}
      CurveIndex=$Emitter.NormalizedLoopAge
      Scale Curve=&RandomRangeFloat(Minimum=8, Maximum=10)
     )
     Z=&RandomRangeFloat(Minimum=0, Maximum=0.1)
     W=&RandomRangeFloat(Minimum=0, Maximum=0.1)
    )
   ) {isSetParameters=true}
  ]
  rendererClasses [&NiagaraMeshRendererProperties]
  renderers [
   {rendererClass=&NiagaraMeshRendererProperties, rendererIndex=0}
   {rendererClass=&NiagaraMeshRendererProperties, rendererIndex=1}
   {rendererClass=&NiagaraMeshRendererProperties, rendererIndex=2}
   {rendererClass=&NiagaraMeshRendererProperties, rendererIndex=3}
   {rendererClass=&NiagaraMeshRendererProperties, rendererIndex=4}
   {rendererClass=&NiagaraMeshRendererProperties, rendererIndex=5}
  ]
 }
 NE_Ball {
  emitterUpdateScript [
   &EmitterState(Inactive Response=::NewEnumerator1, Life Cycle Mode=::NewEnumerator1, Loop Behavior=::NewEnumerator1, Loop Duration=6, UseLoopDelay=false)
   &SpawnBurst_Instantaneous(Use Spawn Probability=false, UseLoopCountLimit=false) {module=SpawnBurst_Instantaneous006}
   &SpawnBurst_Instantaneous(Spawn Time=0.4, Use Spawn Probability=false, UseLoopCountLimit=false) {module=SpawnBurst_Instantaneous007}
   &SpawnBurst_Instantaneous(Spawn Time=0.75, Use Spawn Probability=false, UseLoopCountLimit=false) {module=SpawnBurst_Instantaneous008}
   &SpawnBurst_Instantaneous(Spawn Time=1.27, Use Spawn Probability=false, UseLoopCountLimit=false) {module=SpawnBurst_Instantaneous009}
   &SpawnBurst_Instantaneous(Spawn Time=1.6, Use Spawn Probability=false, UseLoopCountLimit=false) {module=SpawnBurst_Instantaneous010}
   &SpawnBurst_Instantaneous(Spawn Time=1.95, Use Spawn Probability=false, UseLoopCountLimit=false) {module=SpawnBurst_Instantaneous011}
   &SpawnBurst_Instantaneous(
    Spawn Group=1
    Spawn Time=&Add_Float(A=0, B=$User.SpawnTime_Offset)
    Use Spawn Probability=false
    UseLoopCountLimit=false
   )
   &SpawnBurst_Instantaneous(
    Spawn Group=1
    Spawn Time=&Add_Float(A=0.4, B=$User.SpawnTime_Offset)
    Use Spawn Probability=false
    UseLoopCountLimit=false
   ) {module=SpawnBurst_Instantaneous001}
   &SpawnBurst_Instantaneous(
    Spawn Group=1
    Spawn Time=&Add_Float(A=0.75, B=$User.SpawnTime_Offset)
    Use Spawn Probability=false
    UseLoopCountLimit=false
   ) {module=SpawnBurst_Instantaneous002}
   &SpawnBurst_Instantaneous(
    Spawn Group=1
    Spawn Time=&Add_Float(A=1.27, B=$User.SpawnTime_Offset)
    Use Spawn Probability=false
    UseLoopCountLimit=false
   ) {module=SpawnBurst_Instantaneous003}
   &SpawnBurst_Instantaneous(
    Spawn Group=1
    Spawn Time=&Add_Float(A=1.6, B=$User.SpawnTime_Offset)
    Use Spawn Probability=false
    UseLoopCountLimit=false
   ) {module=SpawnBurst_Instantaneous004}
   &SpawnBurst_Instantaneous(
    Spawn Group=1
    Spawn Time=&Add_Float(A=1.95, B=$User.SpawnTime_Offset)
    Use Spawn Probability=false
    UseLoopCountLimit=false
   ) {module=SpawnBurst_Instantaneous005}
   &SpawnBurst_Instantaneous(
    Spawn Group=2
    Spawn Time=&Add_Float(A=0, B=$User.SpawnTime_Offset)
    Use Spawn Probability=false
    UseLoopCountLimit=false
   ) {module=SpawnBurst_Instantaneous012}
   &SpawnBurst_Instantaneous(
    Spawn Group=2
    Spawn Time=&Add_Float(A=0.4, B=$User.SpawnTime_Offset)
    Use Spawn Probability=false
    UseLoopCountLimit=false
   ) {module=SpawnBurst_Instantaneous013}
   &SpawnBurst_Instantaneous(
    Spawn Group=2
    Spawn Time=&Add_Float(A=0.75, B=$User.SpawnTime_Offset)
    Use Spawn Probability=false
    UseLoopCountLimit=false
   ) {module=SpawnBurst_Instantaneous014}
   &SpawnBurst_Instantaneous(
    Spawn Group=2
    Spawn Time=&Add_Float(A=1.27, B=$User.SpawnTime_Offset)
    Use Spawn Probability=false
    UseLoopCountLimit=false
   ) {module=SpawnBurst_Instantaneous015}
   &SpawnBurst_Instantaneous(
    Spawn Group=2
    Spawn Time=&Add_Float(A=1.6, B=$User.SpawnTime_Offset)
    Use Spawn Probability=false
    UseLoopCountLimit=false
   ) {module=SpawnBurst_Instantaneous016}
   &SpawnBurst_Instantaneous(
    Spawn Group=2
    Spawn Time=&Add_Float(A=1.95, B=$User.SpawnTime_Offset)
    Use Spawn Probability=false
    UseLoopCountLimit=false
   ) {module=SpawnBurst_Instantaneous017}
   &SpawnBurst_Instantaneous(Spawn Group=3, Use Spawn Probability=false, UseLoopCountLimit=false) {module=SpawnBurst_Instantaneous018}
   &SpawnBurst_Instantaneous(Spawn Group=3, Spawn Time=0.4, Use Spawn Probability=false, UseLoopCountLimit=false) {module=SpawnBurst_Instantaneous019}
   &SpawnBurst_Instantaneous(Spawn Group=3, Spawn Time=0.75, Use Spawn Probability=false, UseLoopCountLimit=false) {module=SpawnBurst_Instantaneous020}
   &SpawnBurst_Instantaneous(Spawn Group=3, Spawn Time=1.27, Use Spawn Probability=false, UseLoopCountLimit=false) {module=SpawnBurst_Instantaneous021}
   &SpawnBurst_Instantaneous(Spawn Group=3, Spawn Time=1.6, Use Spawn Probability=false, UseLoopCountLimit=false) {module=SpawnBurst_Instantaneous022}
   &SpawnBurst_Instantaneous(Spawn Group=3, Spawn Time=1.95, Use Spawn Probability=false, UseLoopCountLimit=false) {module=SpawnBurst_Instantaneous023}
   &SpawnBurst_Instantaneous(Spawn Group=4, Use Spawn Probability=false, UseLoopCountLimit=false) {module=SpawnBurst_Instantaneous024}
   &SpawnBurst_Instantaneous(Spawn Group=4, Spawn Time=0.4, Use Spawn Probability=false, UseLoopCountLimit=false) {module=SpawnBurst_Instantaneous025}
   &SpawnBurst_Instantaneous(Spawn Group=4, Spawn Time=0.75, Use Spawn Probability=false, UseLoopCountLimit=false) {module=SpawnBurst_Instantaneous026}
   &SpawnBurst_Instantaneous(Spawn Group=4, Spawn Time=1.27, Use Spawn Probability=false, UseLoopCountLimit=false) {module=SpawnBurst_Instantaneous027}
   &SpawnBurst_Instantaneous(Spawn Group=4, Spawn Time=1.6, Use Spawn Probability=false, UseLoopCountLimit=false) {module=SpawnBurst_Instantaneous028}
   &SpawnBurst_Instantaneous(Spawn Group=4, Spawn Time=1.95, Use Spawn Probability=false, UseLoopCountLimit=false) {module=SpawnBurst_Instantaneous029}
  ]
  particleSpawnScript [
   /SetVariables(Particles.VisibilityTag=&SpawnGroup()) {isSetParameters=true}
   &InitializeParticle(
    Lifetime=&MergeHelper_FloatSwitch(
     0=0.15
     1=5
     2=0.2
     3=5
     4=&RandomRangeFloat(Minimum=0.1, Maximum=0.15)
     Index=&SpawnGroup()
     Num=::NewEnumerator3
    )
    Mesh Scale Mode=::NewEnumerator3
    Mesh Uniform Scale=&MergeHelper_FloatSwitch(
     0=1
     1=1
     2=1
     3=1
     4=&RandomRangeFloat(Minimum=0.6, Maximum=1.2)
     Index=&SpawnGroup()
     Num=::NewEnumerator3
    )
    Sprite Rotation Angle=&MergeHelper_FloatSwitch(
     0=0
     1=0
     2=&RandomRangeFloat(Minimum=0, Maximum=360)
     3=&RandomRangeFloat(Minimum=0, Maximum=360)
     Index=&SpawnGroup()
     Num=::NewEnumerator2
    )
    Sprite Rotation Mode=::NewEnumerator0
    Sprite Size=&MergeHelper_Vector2Switch(0=<25 25>, 1=<25 25>, 2=<25 5>, 3=<25 25>, 4=?, Index=&SpawnGroup(), Num=::NewEnumerator2)
    Sprite Size Mode=::NewEnumerator4
    UsePositionOffset=false
    Write Material Random=true
   )
   /SetVariables(Particles.Wave_SpriteRotation:NiagaraFloat=&RandomRangeFloat(Minimum=0, Maximum=0)) {isSetParameters=true}
   /SetVariables(Particles.FlareIndexFloat:NiagaraFloat=$User.FlareIndex) {isSetParameters=true}
   &InitialMeshOrientation(Rotation=&UniformRangedVector(Minimum=<0 -1 0>, Maximum=<0 1 0>), Use Rotation Vector=false)
   /SetVariables() {isSetParameters=true}
   /SetVariables(Particles.MeshOrientation02:Quat4f=&MakeQuaternion(
     Quaternion From=1
     Axis=<1 0 0>
     Angle=0
     Yaw=0.25
     Pitch=&RandomRangeFloat(Minimum=0, Maximum=1)
     Roll=0.25
     From=?
     To=?
     Stabilization Mode:ENiagara_OrientVectorCount=::NewEnumerator0
     Side Direction:Vector3f=?
     Up Direction:Vector3f=?
     Angle Type=::NewEnumerator1
     Normalize:NiagaraBool=true
    )) {isSetParameters=true}
   &SkeletalMeshLocation(
    Bone Sampling Mode=::NewEnumerator4
    MaskSpawnGroup=false
    Orientation Sampling=::NewEnumerator1
    Sampled Position Offset=<-14 5 0>
    Skeletal Mesh=$System.Skeletal Mesh
   )
   /SetVariables(Particles.Switch1:NiagaraFloat=$User.Switch) {isSetParameters=true}
   /SetVariables(Particles.BG_SpriteRotation:NiagaraFloat=&RandomRangeFloat(Minimum=0, Maximum=360)) {isSetParameters=true}
   &CameraOffset(Camera Offset Amount=30)
  ]
  particleUpdateScript [
   &ParticleState()
   /SetVariables(Particles.Cam_float:NiagaraFloat=&RemapRange(
     Input Value To Remap=&DistanceToCamera(Use Owner Position:NiagaraBool=true, Camera Query:NiagaraDataInterfaceCamera=DataInterface())
     Input Min=80
     Input Max=200
     Output Min=0.1
     Output Max=1
    )) {isSetParameters=true}
   &SpriteRotationRate(Rotation Rate=&MergeHelper_FloatSwitch(
     0=0
     1=0
     2=0
     3=&SelectFloatFromArray(
      Array Sampling Mode=::NewEnumerator0
      Direct Array Index=?
      Float Selection Array=DataInterface() {propertyValues {FloatData [100,-100]}}
      Random Seed=0
     )
     4=0
     Index=&SpawnGroup()
     Num=::NewEnumerator3
    ))
   &ScaleSpriteSize(
    Uniform Curve Scale=&Multiply_Float(
     A=&MakeCustomFloatFromBool(
      Bool=&SetBoolByFloatComparison(A=$User.Star, B=0, Comparison Type=::NewEnumerator4)
      True Float=1
      False Float=0
     )
     B=&Waveform(
      "Waveform [1]"=::NewEnumerator0
      "[1] Phase"=$Particles.Age
      "[1] Frequency"=&Multiply_Float(
       A=7
       B=&FloatFromCurve(FloatCurve=DataInterface() {propertyValues {Curve {keys [<at 0 1>,<at 0.3 1>,<at 1 2>]}}}, CurveIndex=$Emitter.NormalizedLoopAge)
      )
      "[1] Amplitude Min/Max"=&MakeVector2D(
       X=&Multiply_Float(
        A=0.98
        B=&FloatFromCurve(FloatCurve=DataInterface() {propertyValues {Curve {keys [<at 0 1>,<at 0.2 1>,<at 0.7 0.85>]}}}, CurveIndex=$Emitter.NormalizedLoopAge)
       )
       Y=&Multiply_Float(
        A=1
        B=&FloatFromCurve(FloatCurve=DataInterface() {propertyValues {Curve {keys [<at 0 1>,<at 0.2 1>,<at 0.7 1.2>]}}}, CurveIndex=$Emitter.NormalizedLoopAge)
       )
      )
     )
    )
    Uniform Curve Sprite Scale=DataInterface() {propertyValues {Curve {defaultValue=0, keys [{tangentMode=RCTM_User, time=0, value=0}, <at 0.04 1>, {tangentMode=RCTM_User, time=1, value=1}]}}}
   ) {module=ScaleSpriteSize002}
   &ScaleSpriteSize(
    Scale Factor=&MakeVector2D(
     X=&FloatFromCurve(
      FloatCurve=DataInterface() {propertyValues {Curve {keys [<at 0 1>,<at 0.676 1>,<at 0.685 0.35>,<at 0.9981 0.2>]}}}
      CurveIndex=$Emitter.NormalizedLoopAge
     )
     Y=1
    )
    Scale Sprite Size Mode=::NewEnumerator2
   )
   &ScaleMeshSize(Scale Factor=&VectorFromFloat(Value=&FloatFromCurve(
      FloatCurve=DataInterface() {propertyValues {Curve {defaultValue=0, keys [{tangentMode=RCTM_User, time=0, value=0}, <at 0.04 1>, {tangentMode=RCTM_User, time=1, value=1}]}}}
      Scale Curve=&Multiply_Float(
       A=&MakeCustomFloatFromBool(
        Bool=&SetBoolByFloatComparison(A=$User.Star, B=0, Comparison Type=::NewEnumerator4)
        True Float=1
        False Float=0
       )
       B=&Waveform(
        "Waveform [1]"=::NewEnumerator0
        "[1] Phase"=$Particles.Age
        "[1] Frequency"=&Multiply_Float(
         A=6
         B=&FloatFromCurve(FloatCurve=DataInterface() {propertyValues {Curve {keys [<at 0 1>,<at 0.3 1>,<at 1 2>]}}}, CurveIndex=$Emitter.NormalizedLoopAge)
        )
        "[1] Amplitude Min/Max"=&MakeVector2D(
         X=&Multiply_Float(
          A=0.98
          B=&FloatFromCurve(FloatCurve=DataInterface() {propertyValues {Curve {keys [<at 0 1>,<at 0.2 1>,<at 0.7 0.85>]}}}, CurveIndex=$Emitter.NormalizedLoopAge)
         )
         Y=&Multiply_Float(
          A=1
          B=&FloatFromCurve(FloatCurve=DataInterface() {propertyValues {Curve {keys [<at 0 1>,<at 0.2 1>,<at 0.7 1.2>]}}}, CurveIndex=$Emitter.NormalizedLoopAge)
         )
        )
       )
      )
     )))
   &ScaleMeshSize(Scale Factor=&MakeVector(
     X=&FloatFromCurve(
      FloatCurve=DataInterface() {propertyValues {Curve {keys [<at 0 1>,<at 0.676 1>,<at 0.685 0.35>,<at 0.9981 0.2>]}}}
      CurveIndex=$Emitter.NormalizedLoopAge
     )
     Y=&FloatFromCurve(
      FloatCurve=DataInterface() {propertyValues {Curve {keys [<at 0 1>,<at 0.676 1>,<at 0.685 0.35>,<at 0.9981 0.2>]}}}
      CurveIndex=$Emitter.NormalizedLoopAge
     )
     Z=1
    )) {module=ScaleMeshSize001}
   /SetVariables(Particles.Energy_Scale:Vector3f=&Multiply_Vector(A=$Particles.Scale, B=&VectorFromFloat(Value=0.636))) {isSetParameters=true}
   /SetVariables(Particles.Energy_Color:LinearColor=&MakeLinearColorFromVectorAndFloat(
     "Vector (RGB)"=&MakeVectorFromLinearColorRGB(Color=&ColorFromCurve(ColorCurve=DataInterface() {propertyValues {
         AlphaCurve {keys []}
         BlueCurve {keys [<at 0.4727 0.0486 RCIM_Cubic>]}
         GreenCurve {keys [<at 0.4727 0.2976 RCIM_Cubic>]}
         RedCurve {keys [<at 0.4727 0.4661 RCIM_Cubic>]}
        }}))
     "Float (Alpha)"=&Multiply_Float(
      A=$Particles.Cam_float
      B=&MakeCustomFloatFromBool(
       Bool=&SetBoolByFloatComparison(A=$User.Switch, B=0)
       True Float=0
       False Float=&FloatFromCurve(FloatCurve=DataInterface() {propertyValues {Curve {defaultValue=1, keys [{tangentMode=RCTM_User, time=0, value=0}, <at 0.0522 0>, {tangentMode=RCTM_User, time=0.086, value=1}]}}})
      )
     )
    )) {isSetParameters=true}
   /SetVariables(Particles.Wave_Detail_Dyn:Vector4f=&MakeVector4(
     X=&FloatFromCurve(FloatCurve=DataInterface() {propertyValues {Curve {keys [<at 0.5 0>,<at 1 1>]}}}, Scale Curve=6)
     Y=0
     Z=0
     W=0
    )) {isSetParameters=true}
   /SetVariables(Particles.Wave_Detail_Color:LinearColor=&MakeLinearColorFromVectorAndFloat(
     "Vector (RGB)"=&MakeVectorFromLinearColorRGB(Color=<rgba 2 1.9618 0.276 1>)
     "Float (Alpha)"=&Multiply_Float(
      A=$Particles.Cam_float
      B=&MakeCustomFloatFromBool(
       Bool=&SetBoolByFloatComparison(A=$User.Switch, B=0)
       True Float=0
       False Float=&FloatFromCurve(FloatCurve=DataInterface() {propertyValues {Curve {defaultValue=1, keys [{tangentMode=RCTM_User, time=0, value=0}, <at 0.0522 0>, {tangentMode=RCTM_User, time=0.086, value=1}]}}})
      )
     )
    )) {isSetParameters=true}
   /SetVariables(Particles.Wave_Detail_Scale:Vector3f=&Multiply_Vector(A=$Particles.Scale, B=&VectorFromFloat(Value=0.2))) {isSetParameters=true}
   /SetVariables(Particles.Ring_Size:Vector2f=&Multiply_Vector2D(A=$Particles.SpriteSize, B=&Vector2DFromFloat(Value=1.1221))) {isSetParameters=true}
   /SetVariables(Particles.Ring_Color:LinearColor=&MakeLinearColorFromVectorAndFloat(
     "Vector (RGB)"=&MakeVectorFromLinearColorRGB(Color=&ColorFromCurve(ColorCurve=DataInterface() {propertyValues {
         AlphaCurve {keys []}
         BlueCurve {keys [<at 0.3756 0.4345 RCIM_Cubic>]}
         GreenCurve {keys [<at 0.3756 1.7467 RCIM_Cubic>]}
         RedCurve {keys [<at 0.3756 2.5 RCIM_Cubic>]}
        }}))
     "Float (Alpha)"=&Multiply_Float(
      A=$Particles.Cam_float
      B=&MakeCustomFloatFromBool(Bool=&SetBoolByFloatComparison(A=$User.Switch, B=0), True Float=0, False Float=1)
     )
    )) {isSetParameters=true}
   /SetVariables(Particles.Ring2_Size:Vector2f=&Multiply_Vector2D_ByFloat(
     Float=&FloatFromCurve(
      FloatCurve=DataInterface() {propertyValues {Curve {defaultValue=1, keys [{tangentMode=RCTM_User, time=0, value=0}, <at 0.2 0.875>, {tangentMode=RCTM_User, time=1, value=1.5}]}}}
      Scale Curve=60
     )
     Vector2D=<1 1>
    )) {isSetParameters=true}
   /SetVariables(Particles.Ring2_Color:LinearColor=&MakeLinearColorFromVectorAndFloat(
     "Vector (RGB)"=&MakeVectorFromLinearColorRGB(Color=&ColorFromCurve(ColorCurve=DataInterface() {propertyValues {
         AlphaCurve {keys []}
         BlueCurve {keys [<at 0.4674 0>,<at 0.5013 0.8542 RCIM_Cubic>]}
         GreenCurve {keys [<at 0.4674 0>,<at 0.5013 3.4673 RCIM_Cubic>]}
         RedCurve {keys [<at 0.4674 0>,<at 0.5013 4 RCIM_Cubic>]}
        }}))
     "Float (Alpha)"=&Multiply_Float(
      A=$Particles.Cam_float
      B=&FloatFromCurve(FloatCurve=DataInterface() {propertyValues {Curve {keys [<at 0 1>,<at 0.2456 0.95>,<at 1 0>]}}})
     )
    )) {isSetParameters=true}
   /SetVariables(Particles.Ring3_Color:LinearColor=&MakeLinearColorFromVectorAndFloat(
     "Vector (RGB)"=&MakeVectorFromLinearColorRGB(Color=&ColorFromCurve(ColorCurve=DataInterface() {propertyValues {
         AlphaCurve {keys []}
         BlueCurve {keys [<at 0.4674 0>,<at 0.5013 0 RCIM_Cubic>]}
         GreenCurve {keys [<at 0.4674 0>,<at 0.5013 3.7161 RCIM_Cubic>]}
         RedCurve {keys [<at 0.4674 0>,<at 0.5013 10 RCIM_Cubic>]}
        }}))
     "Float (Alpha)"=&Multiply_Float(
      A=$Particles.Cam_float
      B=&FloatFromCurve(FloatCurve=DataInterface() {propertyValues {Curve {keys [<at 0 1>,<at 0.2456 0.95>,<at 1 0>]}}}, Scale Curve=2)
     )
    )) {isSetParameters=true}
   /SetVariables(Particles.Wave_Detail_Center_Scale:Vector3f=&VectorFromFloat(Value=&Multiply_Float(A=0.24, B=1))) {isSetParameters=true}
   /SetVariables(Particles.Wave_Detail_Center_Color:LinearColor=&MakeLinearColorFromVectorAndFloat(
     "Vector (RGB)"=&MakeVectorFromLinearColorRGB(Color=<rgba 0.186 0.2 0.0563 1>)
     "Float (Alpha)"=&Multiply_Float(
      A=$Particles.Cam_float
      B=&MakeCustomFloatFromBool(Bool=&SetBoolByFloatComparison(A=$User.Switch, B=0), True Float=0, False Float=1)
     )
    )) {isSetParameters=true}
   /SetVariables(Particles.Dark_Size:Vector2f=&Multiply_Vector2D(A=$Particles.SpriteSize, B=&Vector2DFromFloat(Value=2.2053))) {isSetParameters=true}
   /SetVariables(Particles.Dark_Color:LinearColor=&MakeLinearColorFromVectorAndFloat(
     "Vector (RGB)"=&MakeVectorFromLinearColorRGB(Color=&ColorFromCurve(ColorCurve=DataInterface() {propertyValues {
         AlphaCurve {keys []}
         BlueCurve {keys [<at 0.4617 0.0163 RCIM_Cubic>]}
         GreenCurve {keys [<at 0.4617 0.2034 RCIM_Cubic>]}
         RedCurve {keys [<at 0.4617 0.5312 RCIM_Cubic>]}
        }}))
     "Float (Alpha)"=&Multiply_Float(
      A=$Particles.Cam_float
      B=&MakeCustomFloatFromBool(Bool=&SetBoolByFloatComparison(A=$User.Switch, B=0), True Float=0, False Float=1)
     )
    )) {isSetParameters=true}
   /SetVariables(Particles.Light_Size:Vector2f=&Multiply_Vector2D(A=$Particles.SpriteSize, B=&Vector2DFromFloat(Value=1.5915))) {isSetParameters=true}
   /SetVariables(Particles.Light_Color:LinearColor=&MakeLinearColorFromVectorAndFloat(
     "Vector (RGB)"=&MakeVectorFromLinearColorRGB(Color=&ColorFromCurve(ColorCurve=DataInterface() {propertyValues {
         AlphaCurve {keys []}
         BlueCurve {keys [<at 0 0.1771 RCIM_Cubic>]}
         GreenCurve {keys [<at 0 0.5999 RCIM_Cubic>]}
         RedCurve {keys [<at 0 1 RCIM_Cubic>]}
        }}))
     "Float (Alpha)"=&Multiply_Float(
      A=$Particles.Cam_float
      B=&MakeCustomFloatFromBool(Bool=&SetBoolByFloatComparison(A=$User.Switch, B=0), True Float=0, False Float=1)
     )
    )) {isSetParameters=true}
   /SetVariables(Particles.Light2_Size:Vector2f=&Multiply_Vector2D(A=$Particles.SpriteSize, B=&Vector2DFromFloat(Value=1.4764))) {isSetParameters=true}
   /SetVariables(Particles.Light2_Color:LinearColor=&MakeLinearColorFromVectorAndFloat(
     "Vector (RGB)"=&MakeVectorFromLinearColorRGB(Color=<rgba 0.0938 0.0226 0 1>)
     "Float (Alpha)"=&Multiply_Float(
      A=$Particles.Cam_float
      B=&MakeCustomFloatFromBool(Bool=&SetBoolByFloatComparison(A=$User.Switch, B=0), True Float=0, False Float=1)
     )
    )) {isSetParameters=true}
   /SetVariables(Particles.Light3_Size:Vector2f=&Multiply_Vector2D(A=$Particles.SpriteSize, B=<1.6317 1.6068>)) {isSetParameters=true}
   /SetVariables(Particles.Light3_Color:LinearColor=&MakeLinearColorFromVectorAndFloat(
     "Vector (RGB)"=&MakeVectorFromLinearColorRGB(Color=&ColorFromCurve(
       ColorCurve=DataInterface() {propertyValues {
         AlphaCurve {keys []}
         BlueCurve {keys [<at 0.3286 0.1095 RCIM_Cubic>,<at 0.7148 0.5475>]}
         GreenCurve {keys [<at 0.3286 0.6281 RCIM_Cubic>,<at 0.7148 3.1403>]}
         RedCurve {keys [<at 0.3286 1 RCIM_Cubic>,<at 0.7148 5>]}
        }}
       CurveIndex=$Emitter.NormalizedLoopAge
      ))
     "Float (Alpha)"=&Multiply_Float(
      A=$Particles.Cam_float
      B=&MakeCustomFloatFromBool(Bool=&SetBoolByFloatComparison(A=$User.Switch, B=0), True Float=0, False Float=1)
     )
    )) {isSetParameters=true}
   /SetVariables(Particles.Center_Size:Vector2f=&Multiply_Vector2D(A=$Particles.SpriteSize, B=&Vector2DFromFloat(Value=1.5))) {isSetParameters=true}
   /SetVariables(Particles.Center_Color:LinearColor=&MakeLinearColorFromVectorAndFloat(
     "Vector (RGB)"=&MakeVectorFromLinearColorRGB(Color=&ColorFromCurve(ColorCurve=DataInterface() {propertyValues {
         AlphaCurve {keys []}
         BlueCurve {keys [<at 0 0.6771 RCIM_Cubic>]}
         GreenCurve {keys [<at 0 7.2813 RCIM_Cubic>]}
         RedCurve {keys [<at 0 10 RCIM_Cubic>]}
        }}))
     "Float (Alpha)"=&Multiply_Float(
      A=$Particles.Cam_float
      B=&MakeCustomFloatFromBool(Bool=&SetBoolByFloatComparison(A=$User.Switch, B=0), True Float=0, False Float=1)
     )
    )) {isSetParameters=true}
   /SetVariables(Particles.Star_Size:Vector2f=&Multiply_Vector2D_ByFloat(
     Float=&FloatFromCurve(
      FloatCurve=DataInterface() {propertyValues {Curve {defaultValue=1, keys [{tangentMode=RCTM_User, time=0, value=0}, <at 0.0567 1.075>, {tangentMode=RCTM_User, time=1, value=0}]}}}
      Scale Curve=&RandomRangeFloat(Minimum=5, Maximum=7)
     )
     Vector2D=$Particles.Initial.SpriteSize
    )) {isSetParameters=true}
   /SetVariables(Particles.Star_Color:LinearColor=&MakeLinearColorFromVectorAndFloat(
     "Vector (RGB)"=&MakeVectorFromLinearColorRGB(Color=&ColorFromCurve(ColorCurve=DataInterface() {propertyValues {
         AlphaCurve {keys []}
         BlueCurve {keys [<at 0.3149 0 RCIM_Cubic>,<at 0.3427 1.0417>]}
         GreenCurve {keys [<at 0.3149 0 RCIM_Cubic>,<at 0.3427 4.7797>]}
         RedCurve {keys [<at 0.3149 0 RCIM_Cubic>,<at 0.3427 5>]}
        }}))
     "Float (Alpha)"=&Multiply_Float(A=$Particles.Cam_float, B=5)
    )) {isSetParameters=true}
   /SetVariables(Particles.Wave_Size:Vector2f=&Multiply_Vector2D(A=$Particles.SpriteSize, B=&Vector2DFromFloat(Value=1.7))) {isSetParameters=true}
   /SetVariables(Particles.Wave_Color:LinearColor=&MakeLinearColorFromVectorAndFloat(
     "Vector (RGB)"=&MakeVectorFromLinearColorRGB(Color=&ColorFromCurve(ColorCurve=DataInterface() {propertyValues {
         AlphaCurve {keys []}
         BlueCurve {keys [<at 0.4617 0.934 RCIM_Cubic>]}
         GreenCurve {keys [<at 0.4617 4.0173 RCIM_Cubic>]}
         RedCurve {keys [<at 0.4617 5 RCIM_Cubic>]}
        }}))
     "Float (Alpha)"=&Multiply_Float(
      A=$Particles.Cam_float
      B=&MakeCustomFloatFromBool(Bool=&SetBoolByFloatComparison(A=$User.Switch, B=0), True Float=0, False Float=0.5)
     )
    )) {isSetParameters=true}
   /SetVariables(Particles.Lightning_Color:LinearColor=&MakeLinearColorFromVectorAndFloat(
     "Vector (RGB)"=&MakeCustomVectorFromBool(
      Bool=&SetBoolByFloatComparison(A=$User.Switch, B=0)
      True Vector=&MakeVectorFromLinearColorRGB(Color=<rgba 0 0 0 1>)
      False Vector=&MakeVectorFromLinearColorRGB(Color=&ColorFromCurve(ColorCurve=DataInterface() {propertyValues {
          AlphaCurve {keys []}
          BlueCurve {keys [<at 0.5861 0 RCIM_Cubic>,<at 0.7274 1.061>]}
          GreenCurve {keys [<at 0.5861 0 RCIM_Cubic>,<at 0.7274 8.8485>]}
          RedCurve {keys [<at 0.5861 0 RCIM_Cubic>,<at 0.7274 10>]}
         }}))
     )
     "Float (Alpha)"=&Multiply_Float(
      A=$Particles.Cam_float
      B=&FloatFromCurve(
       FloatCurve=DataInterface() {propertyValues {Curve {keys [<at 0.15 1 RCIM_Cubic>, {arriveTangent=-1.1765, interpMode=RCIM_Cubic, leaveTangent=-1.1765, time=0.796, value=0.245}, <at 1 0>]}}}
       Scale Curve=10
      )
     )
    )) {isSetParameters=true}
   /SetVariables(Particles.Lightning_Scale:Vector3f=&Multiply_Vector(A=$Particles.Initial.Scale, B=&VectorFromFloat(Value=-84.9064))) {enabled=false, isSetParameters=true}
   /SetVariables(Particles.Black_Hole_color:LinearColor=&MakeLinearColorFromVectorAndFloat(
     "Vector (RGB)"=&MakeVectorFromLinearColorRGB(Color=&MakeCustomLinearColorFromBool(Bool=&SetBoolByFloatComparison(A=$User.Switch, B=1), True Color=<rgba 3 3 3 1>, False Color=<rgba 0 0 0 1>))
     "Float (Alpha)"=&Multiply_Float(
      A=$Particles.Cam_float
      B=&MakeCustomFloatFromBool(Bool=&SetBoolByFloatComparison(A=$User.Switch, B=0), True Float=1, False Float=0)
     )
    )) {isSetParameters=true}
   /SetVariables(Particles.Black_Hole_Size:Vector2f=&Multiply_Vector2D(
     A=$Particles.SpriteSize
     B=&Vector2DFromFloat(Value=&MakeCustomFloatFromBool(Bool=&SetBoolByFloatComparison(A=$User.Switch, B=1), True Float=1.4, False Float=0.58))
    )) {isSetParameters=true}
   /SetVariables(Particles.White_Dot_color:LinearColor=&MakeLinearColorFromVectorAndFloat(
     "Vector (RGB)"=&MakeVectorFromLinearColorRGB(Color=&ColorFromCurve(ColorCurve=DataInterface() {propertyValues {AlphaCurve {keys []}, BlueCurve {keys [<at 0.4541 1.0417>]}, GreenCurve {keys [<at 0.4541 4.7797>]}, RedCurve {keys [<at 0.4541 5>]}}}))
     "Float (Alpha)"=&Multiply_Float(
      A=$Particles.Cam_float
      B=&FloatFromCurve(FloatCurve=DataInterface() {propertyValues {Curve {defaultValue=1, keys [{tangentMode=RCTM_User, time=0, value=1}, <at 0.3 1>, {tangentMode=RCTM_User, time=0.5, value=0}]}}})
     )
    )) {isSetParameters=true}
   /SetVariables(Particles.White_Dot_Size:Vector2f=&Multiply_Vector2D(A=$Particles.Initial.SpriteSize, B=&Vector2DFromFloat(Value=0.2))) {isSetParameters=true}
   /SetVariables(Particles.Black_Ring_color:LinearColor=&MakeLinearColorFromVectorAndFloat(
     "Vector (RGB)"=&MakeVectorFromLinearColorRGB(Color=&ColorFromCurve(ColorCurve=DataInterface() {propertyValues {AlphaCurve {keys []}, BlueCurve {keys [<at 0 0 RCIM_Cubic>]}, GreenCurve {keys [<at 0 0 RCIM_Cubic>]}, RedCurve {keys [<at 0 0 RCIM_Cubic>]}}}))
     "Float (Alpha)"=&Multiply_Float(
      A=$Particles.Cam_float
      B=&MakeCustomFloatFromBool(
       Bool=&SetBoolByFloatComparison(A=$User.Switch, B=1)
       True Float=0
       False Float=&MakeCustomFloatFromBool(Bool=&SetBoolByFloatComparison(A=$User.Switch, B=0), True Float=6, False Float=0)
      )
     )
    )) {isSetParameters=true}
   /SetVariables(Particles.Black_Ring_Size:Vector2f=&Multiply_Vector2D(A=$Particles.SpriteSize, B=&Vector2DFromFloat(Value=1.6017))) {isSetParameters=true}
   /SetVariables(Particles.BG_color:LinearColor=&MakeLinearColorFromVectorAndFloat(
     "Vector (RGB)"=&MakeVectorFromLinearColorRGB(Color=&MakeCustomLinearColorFromBool(
       Bool=&SetBoolByFloatComparison(A=$User.Switch, B=1)
       True Color=<rgba 0 0 0 1>
       False Color=&MakeCustomLinearColorFromBool(
        Bool=&SetBoolByFloatComparison(A=$User.Switch, B=0)
        True Color=<rgba 3 2.2199 1.9375 1>
        False Color=<rgba 1.2624 1.846 2 1>
       )
      ))
     "Float (Alpha)"=&Multiply_Float(
      A=$Particles.Cam_float
      B=&MakeCustomFloatFromBool(
       Bool=&SetBoolByFloatComparison(A=$User.Switch, B=0)
       True Float=1
       False Float=&MakeCustomFloatFromBool(Bool=&SetBoolByFloatComparison(A=$User.Switch, B=1), True Float=1, False Float=0)
      )
     )
    )) {isSetParameters=true}
   /SetVariables(Particles.BG_Size:Vector2f=&Multiply_Vector2D(
     A=$Particles.SpriteSize
     B=&Vector2DFromFloat(Value=&MakeCustomFloatFromBool(
       Bool=&SetBoolByFloatComparison(A=$User.Switch, B=1)
       True Float=&RandomRangeFloat(Minimum=4, Maximum=3.8)
       False Float=&MakeCustomFloatFromBool(
        Bool=&SetBoolByFloatComparison(A=$User.Switch, B=0)
        True Float=&RandomRangeFloat(Minimum=3.6, Maximum=3.8)
        False Float=3.6
       )
      ))
    )) {isSetParameters=true}
   /SetVariables(Particles.BG_SIze_Mesh:Vector3f=&Multiply_Vector(
     A=$Particles.Scale
     B=&VectorFromFloat(Value=&MakeCustomFloatFromBool(
       Bool=&SetBoolByFloatComparison(A=$User.Switch, B=1)
       True Float=&RandomRangeFloat(Minimum=0.4, Maximum=0.6)
       False Float=&MakeCustomFloatFromBool(
        Bool=&SetBoolByFloatComparison(A=$User.Switch, B=0)
        True Float=&RandomRangeFloat(Minimum=0.3, Maximum=0.5)
        False Float=1
       )
      ))
    )) {isSetParameters=true}
   /SetVariables(Particles.Black_Star_Size:Vector2f=&Multiply_Vector2D_ByFloat(
     Float=&FloatFromCurve(
      FloatCurve=DataInterface() {propertyValues {Curve {defaultValue=1, keys [<at 0 1>,<at 0.66 1>,<at 0.68 0.2>]}}}
      CurveIndex=$Emitter.NormalizedLoopAge
      Scale Curve=&MakeCustomFloatFromBool(
       Bool=&SetBoolByFloatComparison(A=$User.Star, B=0)
       True Float=&Multiply_Float(A=0.6, B=&RandomRangeFloat(Minimum=6, Maximum=7))
       False Float=0
      )
     )
     Vector2D=$Particles.Initial.SpriteSize
    )) {isSetParameters=true}
   /SetVariables(Particles.Black_Ring_Size2:Vector2f=&Multiply_Vector2D_ByFloat(
     Float=&FloatFromCurve(
      FloatCurve=DataInterface() {propertyValues {Curve {defaultValue=1, keys [<at 0 1>,<at 0.66 1>]}}}
      CurveIndex=$Emitter.NormalizedLoopAge
      Scale Curve=&MakeCustomFloatFromBool(
       Bool=&SetBoolByFloatComparison(A=$User.Star, B=0)
       True Float=&Multiply_Float(A=0.4, B=&RandomRangeFloat(Minimum=5, Maximum=8))
       False Float=0
      )
     )
     Vector2D=$Particles.Initial.SpriteSize
    )) {isSetParameters=true}
   /SetVariables(Particles.Black_Star_Color:LinearColor=&MakeLinearColorFromVectorAndFloat(
     "Vector (RGB)"=&MakeVectorFromLinearColorRGB(Color=&MakeCustomLinearColorFromBool(
       Bool=&SetBoolByFloatComparison(A=$User.Switch, B=1)
       True Color=<rgba 0 0 0 1>
       False Color=&MakeCustomLinearColorFromBool(Bool=&SetBoolByFloatComparison(A=$User.Switch, B=0), True Color=<rgba 1 1 1 1>, False Color=<rgba 1 1 1 1>)
      ))
     "Float (Alpha)"=&Multiply_Float(A=$Particles.Cam_float, B=1)
    )) {isSetParameters=true}
   /SetVariables(Particles.Black_Ring_Color2:LinearColor=&MakeLinearColorFromVectorAndFloat(
     "Vector (RGB)"=&MakeVectorFromLinearColorRGB(Color=&MakeCustomLinearColorFromBool(
       Bool=&SetBoolByFloatComparison(A=$User.Switch, B=1)
       True Color=<rgba 0 0 0 1>
       False Color=&MakeCustomLinearColorFromBool(Bool=&SetBoolByFloatComparison(A=$User.Switch, B=0), True Color=<rgba 1 1 1 1>, False Color=<rgba 1 1 1 1>)
      ))
     "Float (Alpha)"=&Multiply_Float(A=$Particles.Cam_float, B=1)
    )) {isSetParameters=true}
   &PointForce(
    Force Origin=&AddVectorToPosition(Position=<0 0 0>, Vector=<0 0 100>)
    Force Strength=&FloatFromCurve(
     FloatCurve=DataInterface() {propertyValues {Curve {
        defaultValue=0
        keys [<at 0.2857 0 RCIM_Constant>,<at 0.4 1 RCIM_Constant>,<at 0.403 0 RCIM_Constant>,<at 0.45 1 RCIM_Constant>,<at 0.4528 0 RCIM_Constant>]
       }}}
     CurveIndex=$Emitter.NormalizedLoopAge
     Scale Curve=550
    )
    Module.Use Falloff=false
    Module.UseLinearForce=false
   )
   &JitterPosition(
    Jitter Amount:NiagaraFloat=&MergeHelper_FloatSwitch(
     0=0
     1=&FloatFromCurve(FloatCurve=DataInterface() {propertyValues {Curve {defaultValue=1, keys [{tangentMode=RCTM_User, time=0, value=0.5}, {tangentMode=RCTM_User, time=0.3, value=0.5}, <at 1 4>]}}})
     2=0
     3=0
     Index=&SpawnGroup()
     Num=::NewEnumerator2
    )
    Jitter Offset:Vector3f=&RandomRangeVector(Minimum=<-1 -0.5 -0.5>, Maximum=<1 0.5 0.5>, Evaluation Type=::NewEnumerator1)
   )
   &OffsetPosition(Position Offset=&MakeVector(
     X=&FloatFromCurve(
      FloatCurve=DataInterface() {propertyValues {Curve {defaultValue=1, keys [{tangentMode=RCTM_User, time=0, value=0}, <at 0.2 0>, {tangentMode=RCTM_User, time=0.5, value=-50}]}}}
      CurveIndex=$Emitter.NormalizedLoopAge
     )
     Y=0
     Z=0
    )) {enabled=false, module=OffsetPosition001}
   &OffsetPosition(Position Offset=&MakeVector(
     X=0
     Y=0
     Z=&FloatFromCurve(
      FloatCurve=DataInterface() {propertyValues {Curve {keys [<at 0.1053 0>,<at 0.66 0>,<at 0.68 -1>]}}}
      CurveIndex=$Emitter.NormalizedLoopAge
      Scale Curve=500
     )
    )) {enabled=false}
   &FN_Light_Attributes(
    Color=<rgba 10 9.1214 0.5214 1>
    Create Color Attribute=true
    Create Exponent Attribute=true
    Create Position Attribute=true
    Create Radius Attribute=true
    Enable Particle Light Enable=true
    Exponent=2
    Position:NiagaraPosition=$Particles.Position
    Radius=&FloatFromCurve(FloatCurve=DataInterface() {propertyValues {Curve {keys [<at 0 0.2>,<at 0.0133 1>,<at 1 1>]}}}, Scale Curve=100)
    Scale Color=&FloatFromCurve(
     FloatCurve=DataInterface() {propertyValues {Curve {keys [<at 0.0286 1>,<at 0.9 1>,<at 1 0>]}}}
     Scale Curve=&Multiply_Float(
      A=1
      B=&Waveform("Waveform [1]"=::NewEnumerator1, "[1] Phase"=$Particles.Age, "[1] Frequency"=4, "[1] Amplitude Min/Max"=<0 1>)
     )
    )
   )
   /SetVariables(Particles.Position_Center:NiagaraPosition=&AddVectorToPosition(Position=$Particles.Position, Vector=<1.9558 0 0.5531>)) {isSetParameters=true}
   /SetVariables(Particles.Position_Light3:NiagaraPosition=&AddVectorToPosition(Position=$Particles.Position, Vector=<1.0818 0 0>)) {isSetParameters=true}
   &self:ScratchModule_01(FlareIndex:NiagaraFloat=$Particles.Initial.FlareIndexFloat, Position Array:NiagaraDataInterfaceArrayPosition=$System.Position Array) {module=ScratchModule_01}
   &KillParticles(Kill Particles=&SetBoolByFloatComparison(A=$User.Kill, B=1, Comparison Type=::NewEnumerator4)) {module=KillParticles001}
   &SolveForcesAndVelocity(Clamp Velocity=false, Limit Acceleration=false)
   /SetVariables(Particles.Ring_Dynamic:Vector4f=&MakeVector4(
     X=1
     Y=1
     Z=1
     W=&FloatFromCurve(FloatCurve=DataInterface() {propertyValues {Curve {
         defaultValue=0.9
         keys [
          {interpMode=RCIM_Cubic, tangentMode=RCTM_User, time=0.6394, value=1}
          {arriveTangent=-0.9886, interpMode=RCIM_Cubic, leaveTangent=-0.9886, tangentMode=RCTM_User, time=1, value=0.3}
         ]
        }}})
    )) {isSetParameters=true}
   /SetVariables(Particles.Center_Dynamic:Vector4f=&MakeVector4(X=1, Y=1, Z=0.5, W=0.45)) {isSetParameters=true}
   /SetVariables(Particles.Lightning_Dynamic:Vector4f=&MakeVector4(
     X=&FloatFromCurve(FloatCurve=DataInterface() {propertyValues {Curve {keys [<at 0 1>,<at 0.7 0.35>,<at 1 0>]}}})
     Y=&RandomRangeFloat(Minimum=0.05, Maximum=0.08)
     Z=&FloatFromCurve(FloatCurve=DataInterface() {propertyValues {Curve {keys [<at 0.6 1>,<at 1 0>]}}}, Scale Curve=0.5)
     W=1
    )) {isSetParameters=true}
   /SetVariables(Particles.Black_Ring_Dynamic:Vector4f=&MakeVector4(X=1, Y=1, Z=1, W=1.1085)) {isSetParameters=true}
   /SetVariables(Particles.Black_Ring_Dynamic2:Vector4f=&MakeVector4(X=1.0046, Y=0.048, Z=0, W=0)) {isSetParameters=true}
  ]
  rendererClasses [
   &NiagaraSpriteRendererProperties
   &NiagaraLightRendererProperties
   &NiagaraMeshRendererProperties
  ]
  renderers [
   {rendererClass=&NiagaraSpriteRendererProperties}
   {rendererClass=&NiagaraSpriteRendererProperties, rendererIndex=1}
   {rendererClass=&NiagaraSpriteRendererProperties, rendererIndex=2}
   {rendererClass=&NiagaraLightRendererProperties, rendererIndex=3}
   {rendererClass=&NiagaraSpriteRendererProperties, rendererIndex=4}
   {rendererClass=&NiagaraSpriteRendererProperties, rendererIndex=5}
   {rendererClass=&NiagaraSpriteRendererProperties, rendererIndex=6}
   {rendererClass=&NiagaraSpriteRendererProperties, rendererIndex=7}
   {rendererClass=&NiagaraMeshRendererProperties, rendererIndex=8}
   {rendererClass=&NiagaraMeshRendererProperties, rendererIndex=9}
   {rendererClass=&NiagaraSpriteRendererProperties, rendererIndex=10}
   {rendererClass=&NiagaraSpriteRendererProperties, rendererIndex=11}
   {rendererClass=&NiagaraSpriteRendererProperties, rendererIndex=12}
   {rendererClass=&NiagaraMeshRendererProperties, rendererIndex=13}
   {rendererClass=&NiagaraSpriteRendererProperties, rendererIndex=14}
   {rendererClass=&NiagaraSpriteRendererProperties, rendererIndex=15}
   {rendererClass=&NiagaraSpriteRendererProperties, rendererIndex=16}
   {rendererClass=&NiagaraMeshRendererProperties, rendererIndex=17}
   {rendererClass=&NiagaraSpriteRendererProperties, rendererIndex=18}
   {rendererClass=&NiagaraSpriteRendererProperties, rendererIndex=19}
  ]
 }
}
summary {
 systemName=NS_Emote_ShovelCrop_Ball
 emitters [
  {
   emitterName=NE_Ball
   rendererClasses [
    &NiagaraSpriteRendererProperties
    &NiagaraLightRendererProperties
    &NiagaraMeshRendererProperties
   ]
  }
  {emitterName=Lightning_Chain, rendererClasses [&NiagaraMeshRendererProperties]}
  {emitterName=Lightning03, rendererClasses [&NiagaraMeshRendererProperties]}
  {emitterName=Lightning02, rendererClasses [&NiagaraSpriteRendererProperties]}
  {emitterName=Lightning01, rendererClasses [&NiagaraMeshRendererProperties]}
 ]
 userVariables [
  User.BurstIndex {type=NiagaraFloat, defaultValue=0, description=""}
  User.FlareIndex {type=NiagaraFloat, defaultValue=0, description=""}
  User.Intensity {type=NiagaraFloat, defaultValue=4, description=""}
  User.Kill {type=NiagaraFloat, defaultValue=0, description=""}
  User.SpawnTime_Offset {type=NiagaraFloat, defaultValue=0.01, description=""}
  User.Star {type=NiagaraFloat, defaultValue=0, description=""}
  User.Switch {type=NiagaraFloat, defaultValue=0, description=""}
  User.U_Scale {type=NiagaraFloat, defaultValue=2, description=""}
 ]
}
