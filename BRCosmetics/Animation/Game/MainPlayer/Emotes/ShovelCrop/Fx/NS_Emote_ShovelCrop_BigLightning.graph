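# Niagara system for the big lightning strike effect of the ShovelCrop player emote.
# A delayed central bolt mesh, a burst-plus-rate spray of short-lived rising lightning meshes, and a layered glow of sprites, flare meshes and a time-of-day-scaled light form the strike.
# User.Switch selects inverted black-and-white palettes (0 or 1) or gold for the spray and glow; the bolt stays gold.
# Alphas fade as the camera nears.
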NiagaraSystem /BRCosmetics/Animation/Game/MainPlayer/Emotes/ShovelCrop/Fx/NS_Emote_ShovelCrop_BigLightning
dependencies {
 usedDataInterfaces [
  &NiagaraDataInterfaceCamera
  &NiagaraDataInterfaceColorCurve
  &NiagaraDataInterfaceCurve
  &NiagaraDataInterfaceSkeletalMesh
 ]
 usedDynamicInputs []
 usedModules [
  &self:SystemSpawnScript.SystemScriptSource.SystemScriptGraph.NiagaraNodeAssignment_0.SetVariables_#1
  &SystemState
 ]
}
emitters {
 LightningCenter001 {
  emitterUpdateScript [
   &EmitterState(Life Cycle Mode=::NewEnumerator1, Loop Behavior=::NewEnumerator1, Loop Delay=0.3, Loop Duration=1, UseLoopDelay=true)
   &SpawnBurst_Instantaneous(Use Spawn Probability=false, UseLoopCountLimit=false)
  ]
  particleSpawnScript [
   &PartitionParticles()
   &InitializeParticle(
    Color=<rgba 1 0.7922 0.0876 1>
    Lifetime Max=0.1
    Lifetime Min=0.1
    Lifetime Mode=::NewEnumerator1
    Mesh Scale=<1 1 3>
    Mesh Scale Mode=::NewEnumerator4
    UsePositionOffset=false
    Write Material Random=true
   )
   &CameraOffset(Camera Offset Amount=50)
   &InitialMeshOrientation(Mesh Orientation Mode=::NewEnumerator2, Use Rotation Vector=true) {module=InitialMeshOrientation002}
   &AddVelocity(Module.Use Falloff Distance=false, Origin Offset=<0 0 100>, Override Seed=false, Velocity Mode=::NewEnumerator1, Velocity Speed=1)
   &InitialMeshOrientation(
    Mesh Orientation Mode=::NewEnumerator2
    Rotation=&RandomRangeVector(Minimum=<0 0 0>, Maximum=<0 0 0.5>)
    Use Rotation Vector=true
   )
  ]
  particleUpdateScript [
   &ParticleState()
   /SetVariables(Particles.Cam_float:NiagaraFloat=&RemapRange(
     Input Value To Remap=/Niagara/DynamicInputs/Camera/DistanceToCamera(Use Owner Position:NiagaraBool=true, Camera Query:NiagaraDataInterfaceCamera=DataInterface())
     Input Min=80
     Input Max=200
     Output Min=0.1
     Output Max=1
    )) {isSetParameters=true}
   &ScaleMeshSize(Scale Factor=&Multiply_VectorByFloat(Vector=<1 1 1>, Float=1))
   &ScaleColor(
    Scale Alpha=&Multiply_Float(A=$Particles.Cam_float, B=5)
    Scale RGB=&MakeVectorFromLinearColorRGB(Color=&ColorFromCurve(ColorCurve=DataInterface() {propertyValues {
        AlphaCurve {keys []}
        BlueCurve {keys [<at 0.0805 0>,<at 0.1231 6.9792>,<at 0.4469 5.8453>,<at 0.5413 0.3385>]}
        GreenCurve {keys [<at 0.0805 0>,<at 0.1231 9.9061>,<at 0.4469 9.1119>,<at 0.5413 2.6974>]}
        RedCurve {keys [<at 0.0805 0>,<at 0.1231 10>,<at 0.4469 10>,<at 0.5413 5>]}
       }}))
    ScaleA=true
    ScaleRGB=true
   )
   &SolveForcesAndVelocity(Clamp Velocity=false, Limit Acceleration=false)
   /SetVariables(Particles.Lightning glow color:LinearColor=&MakeLinearColorFromVectorAndFloat(
     "Vector (RGB)"=&MakeVectorFromLinearColorRGB(Color=&ColorFromCurve(ColorCurve=DataInterface() {propertyValues {
         AlphaCurve {keys []}
         BlueCurve {keys [<at 0.0753 0>,<at 0.1249 1.0417>,<at 0.4469 0.4167>]}
         GreenCurve {keys [<at 0.0753 0>,<at 0.1249 4.8193>,<at 0.4469 4.3349>]}
         RedCurve {keys [<at 0.0753 0>,<at 0.1249 10>,<at 0.4469 10>]}
        }}))
     "Float (Alpha)"=&Multiply_Float(
      A=$Particles.Cam_float
      B=&FloatFromCurve(FloatCurve=DataInterface() {propertyValues {Curve {keys [<at 0 1>,<at 0.5 1>,<at 1 0>]}}}, Scale Curve=0.3)
     )
    )) {isSetParameters=true}
   /SetVariables(Particles.Glow Dy02:Vector4f=&MakeVector4(X=1, Y=0, Z=0, W=5)) {isSetParameters=true}
   /SetVariables(Particles.Glow Dy03:Vector4f=&MakeVector4(X=15, Y=0, Z=1, W=1)) {isSetParameters=true}
   &DynamicMaterialParameters(
    Index 0 Param 1=1
    Index 0 Param 3=&FloatFromCurve(FloatCurve=DataInterface() {propertyValues {Curve {keys [
         {arriveTangent=4.0072, interpMode=RCIM_Cubic, leaveTangent=4.0072, tangentMode=RCTM_User, time=0, value=0}
         {arriveTangent=4.2299, interpMode=RCIM_Cubic, leaveTangent=4.2299, time=0.2543, value=0.6225}
         {arriveTangent=3.1884, interpMode=RCIM_Cubic, leaveTangent=3.1884, time=0.4395, value=1.8588}
         {arriveTangent=1.6993, interpMode=RCIM_Cubic, leaveTangent=1.6993, tangentMode=RCTM_User, time=1, value=3}
        ]}}})
    Index 0 Param 4=5
    Index 1 Param 1=15
    Index 1 Param 3=1
    Index 1 Param 4=1
    Module.Index0Float0Write=true
    Module.Index0Float1Write=true
    Module.Index0Float2Write=true
    Module.Index0Float3Write=true
    Module.Index1Float0Write=true
    Module.Index1Float1Write=true
    Module.Index1Float2Write=true
    Module.Index1Float3Write=true
    Write Parameter Index 1=::NewEnumerator1
   ) {module=DynamicMaterialParameters001}
  ]
  rendererClasses [&NiagaraMeshRendererProperties]
  renderers [
   {rendererClass=&NiagaraMeshRendererProperties, rendererIndex=0}
   {rendererClass=&NiagaraMeshRendererProperties, rendererIndex=1}
  ]
 }
 Lightning_Explosion002 {
  emitterUpdateScript [
   &EmitterState(Life Cycle Mode=::NewEnumerator1, Loop Behavior=::NewEnumerator1, Loop Duration=0.6, UseLoopDelay=true)
   &SpawnBurst_Instantaneous(Spawn Count=4, Use Spawn Probability=false, UseLoopCountLimit=false)
   &SpawnRate(SpawnRate=15, Use Spawn Probability=false)
  ]
  particleSpawnScript [
   &PartitionParticles()
   &InitializeParticle(
    Lifetime Max=0.1
    Lifetime Min=0.05
    Lifetime Mode=::NewEnumerator1
    Mesh Scale Mode=::NewEnumerator1
    Mesh Uniform Scale Max=4
    Mesh Uniform Scale Min=2
    Position Offset=<0 30 0>
    UsePositionOffset=true
    Write Material Random=true
   )
   &ShapeLocation(
    Cylinder Height=50
    Cylinder Height Midpoint=0
    Cylinder Radius=10
    MaskSpawnGroup=false
    Offset=&MakeVector(
     X=0
     Y=0
     Z=&FloatFromCurve(
      FloatCurve=DataInterface() {propertyValues {Curve {defaultValue=1, keys [{tangentMode=RCTM_User, time=0, value=80}, {tangentMode=RCTM_User, time=1, value=110}]}}}
      CurveIndex=$Emitter.NormalizedLoopAge
     )
    )
    Offset Mode=::NewEnumerator0
    Override Seed=false
    Shape Primitive=::NewEnumerator1
    Surface Only=false
    UseOwnerScale=true
   ) {module=ShapeLocation001}
   &CameraOffset(Camera Offset Amount=100)
   &ShapeLocation(MaskSpawnGroup=false, Override Seed=false, Sphere Radius=0, UseOwnerScale=true)
   &InitialMeshOrientation(
    Mesh Axis To Orient=<0 0 0>
    Mesh Orientation Mode=::NewEnumerator2
    Orientation Vector=<0 0 1>
    Rotation=&RandomRangeVector(Minimum=<0.4 0.066 0.208>, Maximum=<-0.208 0.322 -0.152>)
    Use Rotation Vector=true
   ) {module=InitialMeshOrientation002}
   /SetVariables(Particles.Random=&RandomRangeFloat(Minimum=0, Maximum=1)) {isSetParameters=true}
   &AddVelocity(Velocity=<0 0 100>) {module=AddVelocity001}
  ]
  particleUpdateScript [
   &ParticleState()
   /SetVariables(Particles.Cam_float:NiagaraFloat=&RemapRange(
     Input Value To Remap=&DistanceToCamera(Use Owner Position:NiagaraBool=true, Camera Query:NiagaraDataInterfaceCamera=DataInterface())
     Input Min=80
     Input Max=200
     Output Min=0.1
     Output Max=1
    )) {isSetParameters=true}
   &ScaleMeshSize(Scale Factor=&Multiply_VectorByFloat(
     Vector=<1 1 1>
     Float=&FloatFromCurve(
      FloatCurve=DataInterface() {propertyValues {Curve {defaultValue=1, keys [{tangentMode=RCTM_User, time=0, value=0}, <at 0.0446 1>, <at 0.8 1>, {tangentMode=RCTM_User, time=1, value=0.5}]}}}
      Scale Curve=&FloatFromCurve(FloatCurve=DataInterface() {propertyValues {Curve {keys [<at 0 1>,<at 0.485 0.8>,<at 1 0.5>]}}}, CurveIndex=$Emitter.NormalizedLoopAge)
     )
    ))
   &Drag(Use Linear Drag=true, Use Rotational Drag=false)
   &ScaleColor(
    Scale Alpha=&Multiply_Float(
     A=$Particles.Cam_float
     B=&MakeCustomFloatFromBool(
      Bool=&SetBoolByFloatComparison(A=$User.Switch, B=1)
      True Float=10
      False Float=&MakeCustomFloatFromBool(Bool=&SetBoolByFloatComparison(A=$User.Switch, B=0), True Float=10, False Float=5)
     )
    )
    Scale RGB=&MakeCustomVectorFromBool(
     Bool=&SetBoolByFloatComparison(A=$User.Switch, B=1)
     True Vector=&MakeVectorFromLinearColorRGB(Color=<rgba 1 1 1 1>)
     False Vector=&MakeCustomVectorFromBool(
      Bool=&SetBoolByFloatComparison(A=$User.Switch, B=0)
      True Vector=&MakeVectorFromLinearColorRGB(Color=<rgba 0 0 0 1>)
      False Vector=&MakeVectorFromLinearColorRGB(Color=&ColorFromCurve(ColorCurve=DataInterface() {propertyValues {AlphaCurve {keys []}, BlueCurve {keys [<at 0.7274 0.9375>]}, GreenCurve {keys [<at 0.7274 8.4258>]}, RedCurve {keys [<at 0.7274 10>]}}}))
     )
    )
    ScaleA=true
    ScaleRGB=true
   ) {module=ScaleColor001}
   /SetVariables(Particles.Glow_color:LinearColor=&MakeLinearColorFromVectorAndFloat(
     "Vector (RGB)"=&MakeCustomVectorFromBool(
      Bool=&SetBoolByFloatComparison(A=$User.Switch, B=1)
      True Vector=&MakeVectorFromLinearColorRGB(Color=<rgba 1 1 1 1>)
      False Vector=&MakeCustomVectorFromBool(
       Bool=&SetBoolByFloatComparison(A=$User.Switch, B=0)
       True Vector=&MakeVectorFromLinearColorRGB(Color=<rgba 0 0 0 1>)
       False Vector=&MakeVectorFromLinearColorRGB(Color=&ColorFromCurve(ColorCurve=DataInterface() {propertyValues {AlphaCurve {keys []}, BlueCurve {keys [<at 0.7274 0>]}, GreenCurve {keys [<at 0.7274 0.8048>]}, RedCurve {keys [<at 0.7274 2>]}}}))
      )
     )
     "Float (Alpha)"=&Multiply_Float(
      A=$Particles.Cam_float
      B=&MakeCustomFloatFromBool(
       Bool=&SetBoolByFloatComparison(A=$User.Switch, B=1)
       True Float=0
       False Float=&MakeCustomFloatFromBool(Bool=&SetBoolByFloatComparison(A=$User.Switch, B=0), True Float=0, False Float=3)
      )
     )
    )) {isSetParameters=true}
   &DynamicMaterialParameters(
    Index 0 Param 1=1
    Index 0 Param 2=0.7
    Index 0 Param 3=&FloatFromCurve(FloatCurve=DataInterface() {propertyValues {Curve {keys [{arriveTangent=4.0072, time=0, value=0}, {arriveTangent=1.6993, leaveTangent=1.6993, time=1.0028, value=3.5}]}}})
    Index 0 Param 4=5
    Index 1 Param 1=&RandomRangeFloat(Minimum=3, Maximum=6)
    Index 1 Param 2=&RandomRangeFloat(Minimum=0, Maximum=1)
    Index 1 Param 3=&RandomRangeFloat(Minimum=0.5, Maximum=1)
    Index 1 Param 4=1
    Index 2 Param 1=&FloatFromCurve(FloatCurve=DataInterface() {propertyValues {Curve {keys [<at 0 1>,<at 1 0.3>]}}})
    Module.Index0Float0Write=true
    Module.Index0Float1Write=true
    Module.Index0Float2Write=true
    Module.Index0Float3Write=true
    Module.Index1Float0Write=true
    Module.Index1Float1Write=true
    Module.Index1Float2Write=true
    Module.Index1Float3Write=true
    Module.Index2Float0Write=true
    Module.Index2Float1Write=true
    Module.Index2Float2Write=true
    Module.Index2Float3Write=true
    Write Parameter Index 1=::NewEnumerator1
    Write Parameter Index 2=::NewEnumerator1
   ) {module=DynamicMaterialParameters001}
   &SolveForcesAndVelocity(Clamp Velocity=false, Limit Acceleration=false)
   /SetVariables(Particles.Glow Dy02:Vector4f=<1 0 0 3>) {isSetParameters=true}
  ]
  rendererClasses [&NiagaraMeshRendererProperties]
  renderers [
   {rendererClass=&NiagaraMeshRendererProperties, rendererIndex=0}
   {rendererClass=&NiagaraMeshRendererProperties, rendererIndex=1}
  ]
 }
 glow {
  emitterUpdateScript [
   &EmitterState(Inactive Response=::NewEnumerator1, Life Cycle Mode=::NewEnumerator1, Loop Behavior=::NewEnumerator1, Loop Duration=10, UseLoopDelay=false)
   &SpawnBurst_Instantaneous(Use Spawn Probability=false, UseLoopCountLimit=false)
   &SpawnBurst_Instantaneous(Spawn Group=1, Use Spawn Probability=false, UseLoopCountLimit=false) {module=SpawnBurst_Instantaneous001}
   &SpawnBurst_Instantaneous(Spawn Group=2, Use Spawn Probability=false, UseLoopCountLimit=false) {module=SpawnBurst_Instantaneous002}
   &SpawnBurst_Instantaneous(Spawn Group=3, Use Spawn Probability=false, UseLoopCountLimit=false) {module=SpawnBurst_Instantaneous003}
   &SpawnBurst_Instantaneous(Spawn Group=4, Spawn Time=0.01, Use Spawn Probability=false, UseLoopCountLimit=false) {module=SpawnBurst_Instantaneous004}
   &SpawnBurst_Instantaneous(Spawn Group=5, Spawn Time=0.01, Use Spawn Probability=false, UseLoopCountLimit=false) {module=SpawnBurst_Instantaneous005}
  ]
  particleSpawnScript [
   /SetVariables(Particles.VisibilityTag=&SpawnGroup()) {isSetParameters=true}
   &InitializeParticle(
    Color Mode=::NewEnumerator0
    Lifetime=&MergeHelper_FloatSwitch(0=0.3, 1=0.32, 2=0.25, 3=0.25, 4=0.2, 5=0.2, Index=&SpawnGroup(), Num=::NewEnumerator4)
    Mesh Scale=<0.8 2 2>
    Mesh Scale Mode=::NewEnumerator4
    Position Offset=&MergeHelper_VectorSwitch(0=<0 5 0>, 1=<0 5 0>, 2=<0 5 0>, 3=<0 5 0>, 4=<0 5 0>, 5=<0 5 0>, Index=&SpawnGroup(), Num=::NewEnumerator4)
    Sprite Rotation Mode=::NewEnumerator1
    Sprite Size=<450 600>
    Sprite Size Mode=::NewEnumerator4
    UsePositionOffset=true
    Write Material Random=true
   )
   &CameraOffset(Camera Offset Amount=80)
   &SpriteFacingAndAlignment(Sprite Alignment=<0 0 0>, Sprite Facing=<0 0 1>, WriteAlignment=false, WriteFacing=true) {module=SpriteFacingAndAlignment001}
   &InitialMeshOrientation(Rotation=<0 -0.25 0>, Use Orientation Vector=true, Use Rotation Vector=true)
   &InitialMeshOrientation(
    Rotation=&RandomRangeVector(Minimum=<0 0 -1>, Maximum=<0 0 1>)
    Use Orientation Vector=true
    Use Rotation Vector=true
   ) {module=InitialMeshOrientation001}
  ]
  particleUpdateScript [
   &ParticleState()
   /SetVariables(Particles.Cam_float:NiagaraFloat=&RemapRange(
     Input Value To Remap=&DistanceToCamera(Use Owner Position:NiagaraBool=true, Camera Query:NiagaraDataInterfaceCamera=DataInterface())
     Input Min=80
     Input Max=200
     Output Min=0.1
     Output Max=1
    )) {isSetParameters=true}
   &DynamicMaterialParameters(
    Index 0 Param 1=2
    Index 0 Param 2=5
    Module.Param0WriteEnabled=true
    Module.Param1WriteEnabled=true
    Module.Param2WriteEnabled=true
    Module.Param3WriteEnabled=true
   )
   /SetVariables(Particles.Burst_Light_Color:LinearColor=&MakeLinearColorFromVectorAndFloat(
     "Vector (RGB)"=&MakeVectorFromLinearColorRGB(Color=&MakeCustomLinearColorFromBool(
       Bool=&SetBoolByFloatComparison(A=$User.Switch, B=1)
       True Color=<rgba 0 0 0 1>
       False Color=&MakeCustomLinearColorFromBool(
        Bool=&SetBoolByFloatComparison(A=$User.Switch, B=0)
        True Color=<rgba 1 1 1 1>
        False Color=<rgba 2 1.5759 0.2604 1>
       )
      ))
     "Float (Alpha)"=&Multiply_Float(
      A=$Particles.Cam_float
      B=&FloatFromCurve(
       FloatCurve=DataInterface() {propertyValues {Curve {keys [<at 0 1 RCIM_Cubic>, {arriveTangent=-1, interpMode=RCIM_Cubic, leaveTangent=-1, time=0.5911, value=0.5951}, <at 1 0>]}}}
       Scale Curve=&MakeCustomFloatFromBool(
        Bool=&SetBoolByFloatComparison(A=$User.Switch, B=1)
        True Float=0
        False Float=&MakeCustomFloatFromBool(Bool=&SetBoolByFloatComparison(A=$User.Switch, B=0), True Float=0, False Float=1)
       )
      )
     )
    )) {isSetParameters=true}
   /SetVariables(Particles.Burst_Light_Size:Vector2f=&Multiply_Vector2D_ByFloat(
     Float=&FloatFromCurve(FloatCurve=DataInterface() {propertyValues {Curve {keys [{arriveTangent=3.635, time=0, value=0}, {arriveTangent=0.4002, time=0.0513, value=0.8}, {tangentMode=RCTM_User, time=1, value=1}]}}})
     Vector2D=&Multiply_Vector2D(
      A=&Multiply_Vector2D_ByFloat(Float=0.6, Vector2D=$Particles.SpriteSize)
      B=&MakeVector2D(
       X=&FloatFromCurve(FloatCurve=DataInterface() {propertyValues {Curve {keys [<at 0 1>,<at 1 0.9>]}}}, Scale Curve=0.9)
       Y=&FloatFromCurve(FloatCurve=DataInterface() {propertyValues {Curve {keys [<at 0 1>,<at 1 1.3>]}}}, Scale Curve=0.8)
      )
     )
    )) {isSetParameters=true}
   /SetVariables(Particles.Burst_Dark_Size:Vector2f=&Multiply_Vector2D_ByFloat(Float=1.0996, Vector2D=$Particles.Burst_Light_Size)) {isSetParameters=true}
   /SetVariables(Particles.Burst_Dark_Color:LinearColor=&MakeLinearColorFromVectorAndFloat(
     "Vector (RGB)"=&MakeVectorFromLinearColorRGB(Color=&MakeCustomLinearColorFromBool(
       Bool=&SetBoolByFloatComparison(A=$User.Switch, B=1)
       True Color=<rgba 0 0 0 1>
       False Color=&MakeCustomLinearColorFromBool(
        Bool=&SetBoolByFloatComparison(A=$User.Switch, B=0)
        True Color=<rgba 1 1 1 1>
        False Color=<rgba 2 0.6239 0 1>
       )
      ))
     "Float (Alpha)"=&Multiply_Float(
      A=$Particles.Cam_float
      B=&FloatFromCurve(
       FloatCurve=DataInterface() {propertyValues {Curve {keys [<at 0 1 RCIM_Cubic>, {arriveTangent=-1, interpMode=RCIM_Cubic, leaveTangent=-1, time=0.5911, value=0.5951}, <at 1 0>]}}}
       Scale Curve=&MakeCustomFloatFromBool(
        Bool=&SetBoolByFloatComparison(A=$User.Switch, B=1)
        True Float=0
        False Float=&MakeCustomFloatFromBool(Bool=&SetBoolByFloatComparison(A=$User.Switch, B=0), True Float=0, False Float=0.2)
       )
      )
     )
    )) {isSetParameters=true}
   /SetVariables(Particles.Burst_Center_Size:Vector2f=&Multiply_Vector2D_ByFloat(
     Float=&MakeCustomFloatFromBool(
      Bool=&SetBoolByFloatComparison(A=$User.Switch, B=1)
      True Float=0.85
      False Float=&MakeCustomFloatFromBool(Bool=&SetBoolByFloatComparison(A=$User.Switch, B=0), True Float=0.91, False Float=0.91)
     )
     Vector2D=$Particles.Burst_Light_Size
    )) {isSetParameters=true}
   /SetVariables(Particles.Burst_Center_Color:LinearColor=&MakeLinearColorFromVectorAndFloat(
     "Vector (RGB)"=&MakeVectorFromLinearColorRGB(Color=&MakeCustomLinearColorFromBool(
       Bool=&SetBoolByFloatComparison(A=$User.Switch, B=1)
       True Color=<rgba 1 1 1 1>
       False Color=&MakeCustomLinearColorFromBool(Bool=&SetBoolByFloatComparison(A=$User.Switch, B=0), True Color=<rgba 0 0 0 1>, False Color=<rgba 2 2 2 1>)
      ))
     "Float (Alpha)"=&Multiply_Float(
      A=$Particles.Cam_float
      B=&FloatFromCurve(
       FloatCurve=DataInterface() {propertyValues {Curve {keys [<at 0 1 RCIM_Cubic>, {arriveTangent=-1, interpMode=RCIM_Cubic, leaveTangent=-1, time=0.5911, value=0.5951}, <at 1 0>]}}}
       Scale Curve=&MakeCustomFloatFromBool(
        Bool=&SetBoolByFloatComparison(A=$User.Switch, B=1)
        True Float=0.8
        False Float=&MakeCustomFloatFromBool(Bool=&SetBoolByFloatComparison(A=$User.Switch, B=0), True Float=0, False Float=0.8)
       )
      )
     )
    )) {isSetParameters=true}
   /SetVariables(Particles.Burst_Center_Dark_Color:LinearColor=&MakeLinearColorFromVectorAndFloat(
     "Vector (RGB)"=&MakeVectorFromLinearColorRGB(Color=&MakeCustomLinearColorFromBool(
       Bool=&SetBoolByFloatComparison(A=$User.Switch, B=1)
       True Color=<rgba 1 1 1 1>
       False Color=&MakeCustomLinearColorFromBool(Bool=&SetBoolByFloatComparison(A=$User.Switch, B=0), True Color=<rgba 0 0 0 1>, False Color=<rgba 2 2 2 1>)
      ))
     "Float (Alpha)"=&Multiply_Float(
      A=$Particles.Cam_float
      B=&FloatFromCurve(
       FloatCurve=DataInterface() {propertyValues {Curve {keys [<at 0 1 RCIM_Cubic>, {arriveTangent=-1, interpMode=RCIM_Cubic, leaveTangent=-1, time=0.5911, value=0.5951}, <at 1 0>]}}}
       Scale Curve=&MakeCustomFloatFromBool(
        Bool=&SetBoolByFloatComparison(A=$User.Switch, B=1)
        True Float=0
        False Float=&MakeCustomFloatFromBool(Bool=&SetBoolByFloatComparison(A=$User.Switch, B=0), True Float=0.8, False Float=0)
       )
      )
     )
    )) {isSetParameters=true}
   /SetVariables(Particles.Burst_Light2_Color:LinearColor=&MakeLinearColorFromVectorAndFloat(
     "Vector (RGB)"=&MakeVectorFromLinearColorRGB(Color=&MakeCustomLinearColorFromBool(
       Bool=&SetBoolByFloatComparison(A=$User.Switch, B=1)
       True Color=<rgba 0 0 0 1>
       False Color=&MakeCustomLinearColorFromBool(
        Bool=&SetBoolByFloatComparison(A=$User.Switch, B=0)
        True Color=<rgba 1 1 1 1>
        False Color=<rgba 2 1.5759 0.2604 1>
       )
      ))
     "Float (Alpha)"=&Multiply_Float(
      A=$Particles.Cam_float
      B=&FloatFromCurve(
       FloatCurve=DataInterface() {propertyValues {Curve {keys [<at 0 1 RCIM_Cubic>, {arriveTangent=-1, interpMode=RCIM_Cubic, leaveTangent=-1, time=0.5911, value=0.5951}, <at 1 0>]}}}
       Scale Curve=&MakeCustomFloatFromBool(
        Bool=&SetBoolByFloatComparison(A=$User.Switch, B=1)
        True Float=1
        False Float=&MakeCustomFloatFromBool(Bool=&SetBoolByFloatComparison(A=$User.Switch, B=0), True Float=0.2, False Float=0)
       )
      )
     )
    )) {isSetParameters=true}
   /SetVariables(Particles.Burst_Light2_Size:Vector2f=&Multiply_Vector2D_ByFloat(Float=1.1, Vector2D=$Particles.Burst_Light_Size)) {isSetParameters=true}
   /SetVariables(Particles.Flash_Color:LinearColor=&MakeLinearColorFromVectorAndFloat(
     "Vector (RGB)"=&MakeVectorFromLinearColorRGB(Color=<rgba 5 3.5125 0.3646 1>)
     "Float (Alpha)"=&Multiply_Float(
      A=$Particles.Cam_float
      B=&MakeCustomFloatFromBool(
       Bool=&SetBoolByFloatComparison(A=$User.Switch, B=0)
       True Float=0
       False Float=&FloatFromCurve(
        FloatCurve=DataInterface() {propertyValues {Curve {keys [<at 0 1 RCIM_Cubic>, {arriveTangent=-1, interpMode=RCIM_Cubic, leaveTangent=-1, time=0.5911, value=0.5951}, <at 1 0>]}}}
        Scale Curve=0.2
       )
      )
     )
    )) {isSetParameters=true}
   /SetVariables(Particles.Flash_size:Vector2f=&Multiply_Vector2D_ByFloat(
     Float=450
     Vector2D=&Multiply_Vector2D_ByFloat(
      Float=&FloatFromCurve(
       FloatCurve=DataInterface() {propertyValues {Curve {keys [{arriveTangent=3.635, time=0, value=0}, {arriveTangent=0.4002, time=0.0513, value=0.8}, {tangentMode=RCTM_User, time=1, value=1}]}}}
       Scale Curve=1.5843
      )
      Vector2D=<0.8 1>
     )
    )) {isSetParameters=true}
   /SetVariables(Particles.Ground_Flash_size:Vector2f=&Multiply_Vector2D_ByFloat(Float=1.2, Vector2D=$Particles.Flash_size)) {isSetParameters=true}
   &FN_Light_Attributes(
    Color=<rgba 30 24.417 0.7762 1>
    Create Color Attribute=true
    Create Exponent Attribute=true
    Create Radius Attribute=true
    Exponent=2
    Radius=120
    Scale Color=&Multiply_Float(
     A=&FloatFromCurve(FloatCurve=DataInterface() {propertyValues {Curve {keys [
          {interpMode=RCIM_Cubic, tangentMode=RCTM_User, time=0, value=1}
          <at 0.2 1>
          {interpMode=RCIM_Cubic, tangentMode=RCTM_User, time=0.6, value=1}
          {tangentMode=RCTM_User, time=0.8, value=0}
         ]}}})
     B=&FNTimeOfDayScalar(Night Scale=0.7, Morning Scale=0.3, Evening Scale=0.4)
    )
   )
   &DynamicMaterialParameters(
    Index 0 Param 1=&FloatFromCurve(FloatCurve=DataInterface() {propertyValues {Curve {keys [
         <at 0 0.8>
         {arriveTangent=-1.5696, interpMode=RCIM_Cubic, leaveTangent=-1.5696, tangentMode=RCTM_User, time=0.6, value=0.8}
         {arriveTangent=-1.326, interpMode=RCIM_Cubic, leaveTangent=-1.326, tangentMode=RCTM_User, time=1, value=0.4}
        ]}}})
    Module.Param0WriteEnabled=true
    Module.Param1WriteEnabled=true
    Module.Param2WriteEnabled=true
    Module.Param3WriteEnabled=true
   ) {module=DynamicMaterialParameters001}
   /SetVariables(Particles.Flare_Parameter:Vector4f=&MakeVector4(
     X=0
     Y=2
     Z=0
     W=&FloatFromCurve(FloatCurve=DataInterface() {propertyValues {Curve {keys [<at 0 1>,<at 1 3>]}}})
    )) {isSetParameters=true}
   /SetVariables(Particles.Flare1_Scale:Vector3f=&Multiply_VectorByFloat(
     Vector=$Particles.Scale
     Float=&FloatFromCurve(
      FloatCurve=DataInterface() {propertyValues {Curve {keys [{arriveTangent=3.635, time=0, value=0}, {arriveTangent=0.4002, time=0.0513, value=0.8}, {tangentMode=RCTM_User, time=1, value=1}]}}}
      Scale Curve=2.5
     )
    )) {isSetParameters=true}
   /SetVariables(Particles.Flare2_Scale:Vector3f=&Multiply_VectorByFloat(Vector=$Particles.Flare1_Scale, Float=1.1)) {isSetParameters=true}
   /SetVariables(Particles.Flare_Light_Color:LinearColor=&MakeLinearColorFromVectorAndFloat(
     "Vector (RGB)"=&MakeVectorFromLinearColorRGB(Color=&MakeCustomLinearColorFromBool(
       Bool=&SetBoolByFloatComparison(A=$User.Switch, B=1)
       True Color=<rgba 1 1 1 1>
       False Color=&MakeCustomLinearColorFromBool(
        Bool=&SetBoolByFloatComparison(A=$User.Switch, B=0)
        True Color=<rgba 0 0 0 1>
        False Color=<rgba 2 1.6747 0.6989 1>
       )
      ))
     "Float (Alpha)"=&Multiply_Float(
      A=$Particles.Cam_float
      B=&FloatFromCurve(
       FloatCurve=DataInterface() {propertyValues {Curve {keys [<at 0 1>,<at 1 1>]}}}
       Scale Curve=&MakeCustomFloatFromBool(
        Bool=&SetBoolByFloatComparison(A=$User.Switch, B=1)
        True Float=1
        False Float=&MakeCustomFloatFromBool(Bool=&SetBoolByFloatComparison(A=$User.Switch, B=0), True Float=2, False Float=1)
       )
      )
     )
    )) {isSetParameters=true}
   /SetVariables(Particles.Flare_Dark_Color:LinearColor=&MakeLinearColorFromVectorAndFloat(
     "Vector (RGB)"=&MakeVectorFromLinearColorRGB(Color=&MakeCustomLinearColorFromBool(
       Bool=&SetBoolByFloatComparison(A=$User.Switch, B=1)
       True Color=<rgba 0 0 0 1>
       False Color=&MakeCustomLinearColorFromBool(
        Bool=&SetBoolByFloatComparison(A=$User.Switch, B=0)
        True Color=<rgba 1 1 1 1>
        False Color=<rgba 0.5938 0.1852 0 1>
       )
      ))
     "Float (Alpha)"=&Multiply_Float(
      A=$Particles.Cam_float
      B=&FloatFromCurve(
       FloatCurve=DataInterface() {propertyValues {Curve {keys [<at 0 1>,<at 1 1>]}}}
       Scale Curve=&MakeCustomFloatFromBool(
        Bool=&SetBoolByFloatComparison(A=$User.Switch, B=1)
        True Float=2
        False Float=&MakeCustomFloatFromBool(Bool=&SetBoolByFloatComparison(A=$User.Switch, B=0), True Float=5, False Float=0.5)
       )
      )
     )
    )) {isSetParameters=true}
  ]
  rendererClasses [
   &NiagaraLightRendererProperties
   &NiagaraSpriteRendererProperties
   &NiagaraMeshRendererProperties
  ]
  renderers [
   {rendererClass=&NiagaraLightRendererProperties, rendererIndex=0}
   {rendererClass=&NiagaraSpriteRendererProperties, rendererIndex=1}
   {rendererClass=&NiagaraSpriteRendererProperties, rendererIndex=2}
   {rendererClass=&NiagaraSpriteRendererProperties, rendererIndex=3}
   {rendererClass=&NiagaraSpriteRendererProperties, rendererIndex=4}
   {rendererClass=&NiagaraSpriteRendererProperties, rendererIndex=5}
   {rendererClass=&NiagaraSpriteRendererProperties, rendererIndex=6}
   {rendererClass=&NiagaraMeshRendererProperties, rendererIndex=7}
   {rendererClass=&NiagaraMeshRendererProperties, rendererIndex=8}
   {rendererClass=&NiagaraSpriteRendererProperties, rendererIndex=9}
  ]
 }
}
summary {
 systemName=NS_Emote_ShovelCrop_BigLightning
 emitters [
  {emitterName=LightningCenter001, rendererClasses [&NiagaraMeshRendererProperties]}
  {
   emitterName=glow
   rendererClasses [
    &NiagaraLightRendererProperties
    &NiagaraSpriteRendererProperties
    &NiagaraMeshRendererProperties
   ]
  }
  {emitterName=Lightning_Explosion002, rendererClasses [&NiagaraMeshRendererProperties]}
 ]
 userVariables [User.Switch {type=NiagaraFloat, defaultValue=0, description=""}]
}
