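# Niagara effect for the ShovelCrop player emote, with one Aura emitter drawing short-lived warm-gold sprites, apparently at the ears.
# Two burst spawns use separate spawn groups to pick two skeletal-mesh bones or sockets with mirrored offsets.
# Particles follow those points each update, aligning to the sampled rotation, growing in quickly and fading over their lifetime.
# A camera-distance remap is present but disabled.
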NiagaraSystem /BRCosmetics/Animation/Game/MainPlayer/Emotes/ShovelCrop/Fx/NS_Emote_ShovelCrop_Ear
dependencies {
 usedDataInterfaces [&NiagaraDataInterfaceCamera, &NiagaraDataInterfaceCurve, &NiagaraDataInterfaceSkeletalMesh]
 usedDynamicInputs []
 usedModules [
  &self:SystemSpawnScript.SystemScriptSource.SystemScriptGraph.NiagaraNodeAssignment_0.SetVariables_#1
  &SystemState
 ]
}
emitters {Aura {
  emitterUpdateScript [
   &EmitterState(Life Cycle Mode=::NewEnumerator1, Loop Behavior=::NewEnumerator1, Loop Duration=3, UseLoopDelay=false)
   &SpawnBurst_Instantaneous(Use Spawn Probability=false, UseLoopCountLimit=false) {module=SpawnBurst_Instantaneous001}
   &SpawnBurst_Instantaneous(Spawn Group=1, Use Spawn Probability=false, UseLoopCountLimit=false)
  ]
  particleSpawnScript [
   &InitializeParticle(
    Color=<rgba 2 1.476 0.3594 1>
    Lifetime=2.5
    Position Offset=<0 10 110>
    Sprite Rotation Mode=::NewEnumerator1
    Sprite Size=<12 23>
    Sprite Size Mode=::NewEnumerator4
    Sprite UV Mode=::NewEnumerator1
    UsePositionOffset=true
    Write Material Random=true
   )
   &CameraOffset(Camera Offset Amount=-10)
  ]
  particleUpdateScript [
   &ParticleState()
   &SkeletalMeshLocation(
    Bone / Socket Index=&MergeHelper_IntSwitch(0=0, 1=1, 2=?, 3=?, 4=?, 5=?, Index=&SpawnGroup(), Num=::NewEnumerator0)
    Bone Sampling Mode=::NewEnumerator4
    MaskSpawnGroup=false
    Orientation Sampling=::NewEnumerator1
    Sampled Position Offset=&MergeHelper_VectorSwitch(0=<4 0 0>, 1=<-4 0 0>, 2=?, 3=?, Index=&SpawnGroup(), Num=::NewEnumerator0)
    Skeletal Mesh=$System.Skeletal Mesh
   )
   /SetVariables(Particles.Cam_float:NiagaraFloat=&RemapRange(
     Input Value To Remap=/Niagara/DynamicInputs/Camera/DistanceToCamera(Use Owner Position:NiagaraBool=true, Camera Query:NiagaraDataInterfaceCamera=DataInterface())
     Input Min=80
     Input Max=200
     Output Min=0.1
     Output Max=1
    )) {enabled=false, isSetParameters=true}
   &DynamicMaterialParameters(
    Index 0 Param 4=&RandomRangeFloat(Minimum=0, Maximum=1)
    Module.Param0WriteEnabled=true
    Module.Param1WriteEnabled=true
    Module.Param2WriteEnabled=true
    Module.Param3WriteEnabled=true
   )
   &ScaleSpriteSize(
    Scale Sprite Size Mode=::NewEnumerator0
    Uniform Scale Factor=&FloatFromCurve(FloatCurve=DataInterface() {propertyValues {Curve {defaultValue=0, keys [{tangentMode=RCTM_User, time=0, value=0}, <at 0.0727 1>, {tangentMode=RCTM_User, time=1, value=1}]}}})
   ) {module=ScaleSpriteSize002}
   &ScaleColor(
    Scale Alpha=&FloatFromCurve(FloatCurve=DataInterface() {propertyValues {Curve {
        defaultValue=0
        keys [
         {arriveTangent=0.3137, interpMode=RCIM_Cubic, leaveTangent=0.3137, tangentMode=RCTM_User, time=0, value=1}
         {arriveTangent=-2.0958, interpMode=RCIM_Cubic, leaveTangent=-2.0958, tangentMode=RCTM_User, time=1, value=0}
        ]
       }}})
    ScaleA=true
    ScaleRGB=true
   )
   /SetVariables(Particles.SpriteAlignment=&RotateVectorByQuaternion(
     Vector To Rotate=&MaskVector(Mask Source=0, Mask Parameter=?, Mask=0, Masked Vector=<1 0 0>, Passthrough Vector=<-1 0 0>)
     Quaternion=$Particles.SkeletalMeshLocation.SampledRotation
    )) {isSetParameters=true}
  ]
  rendererClasses [&NiagaraSpriteRendererProperties]
  renderers [{rendererClass=&NiagaraSpriteRendererProperties}]
 }}
summary {
 systemName=NS_Emote_ShovelCrop_Ear
 emitters [{emitterName=Aura, rendererClasses [&NiagaraSpriteRendererProperties]}]
 userVariables [
  User.bIsFrontEnd {type=NiagaraBool}
  User.bIsFrontEndPreview {type=NiagaraBool}
  User.FlareIndex {type=NiagaraFloat, defaultValue=0, description=""}
  User.SpawnTime_Offset {type=NiagaraFloat, defaultValue=0.01, description=""}
  User.Switch {type=NiagaraFloat, defaultValue=0, description=""}
  User.SystemOfs {type=Vector3f, description="", defaultValue {struct=Vector3f, value=<0 0 100>}}
  User.U_Scale {type=NiagaraFloat, defaultValue=2, description=""}
  User.Vel {type=NiagaraFloat, defaultValue=350, description=""}
 ]
}
