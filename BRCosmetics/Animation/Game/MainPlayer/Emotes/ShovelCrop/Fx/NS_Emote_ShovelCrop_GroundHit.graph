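# Niagara impact effect for the ShovelCrop emote's ground hit, bursting debris, dust and light where it lands.
# Emitters include mesh chunks, layered ground decals, glows and waves, brief lightning meshes, and spikes, streaks, dirt and smoke.
# Spawn groups vary motion and size within shared emitters, with per-layer particle attributes and visibility tags set for multiple renderers.
# Camera distance dims alpha up close, and front-end preview switches adjust brightness and scale.
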NiagaraSystem /BRCosmetics/Animation/Game/MainPlayer/Emotes/ShovelCrop/Fx/NS_Emote_ShovelCrop_GroundHit
dependencies {
 usedDataInterfaces [
  &NiagaraDataInterfaceCamera
  &NiagaraDataInterfaceCurve
  &NiagaraDataInterfaceColorCurve
  &NiagaraDataInterfaceVector2DCurve
 ]
 usedDynamicInputs []
 usedModules [&SystemState]
}
emitters {
 Chunks001 {
  emitterUpdateScript [
   &EmitterState(Life Cycle Mode=::NewEnumerator1, Loop Behavior=::NewEnumerator1, Loop Delay=0.05, Loop Duration=1.3, UseLoopDelay=false)
   &SpawnBurst_Instantaneous(Spawn Count=3, Use Spawn Probability=false, UseLoopCountLimit=false) {module=SpawnBurst_Instantaneous001}
   &SpawnBurst_Instantaneous(Spawn Count=20, Spawn Group=1, Spawn Time=0.05, Use Spawn Probability=false, UseLoopCountLimit=false) {module=SpawnBurst_Instantaneous002}
  ]
  particleSpawnScript [
   &InitializeParticle(
    Color Maximum=<rgba 0.5833 0.2744 0.1853 1>
    Color Minimum=<rgba 0.1198 0.0959 0.0867 1>
    Color Mode=::NewEnumerator2
    Lifetime Max=0.5
    Lifetime Min=0.15
    Lifetime Mode=::NewEnumerator1
    Mass Mode=::NewEnumerator1
    Mesh Scale Mode=::NewEnumerator3
    Mesh Uniform Scale=&MergeHelper_FloatSwitch(
     0=&RandomRangeFloat(Minimum=3, Maximum=4)
     1=&RandomRangeFloat(Minimum=0.5, Maximum=2)
     2=?
     3=?
     Index=&SpawnGroup()
     Num=::NewEnumerator0
    )
    Sprite Rotation Mode=::NewEnumerator2
    Sprite UV Mode=::NewEnumerator3
    UsePositionOffset=false
    Write Material Random=true
   )
   &ShapeLocation(
    MaskSpawnGroup=false
    Non Uniform Scale=<1 1 0.5>
    Offset=<0 0 15>
    Offset Mode=::NewEnumerator0
    Override Seed=false
    Sphere Radius=40
    UseOwnerScale=true
   )
   &AddVelocity(
    Module.Use Falloff Distance=false
    Origin Offset=<0 0 -30>
    Override Seed=false
    Velocity Mode=::NewEnumerator1
    Velocity Speed=&Multiply_Float(
     A=&RandomRangeFloat(Minimum=300, Maximum=420)
     B=&MergeHelper_FloatSwitch(0=2, 1=4, 2=?, 3=?, Index=&SpawnGroup(), Num=::NewEnumerator0)
    )
   )
   &InitialMeshOrientation(Use Rotation Vector=false)
  ]
  particleUpdateScript [
   &ParticleState()
   /SetVariables(Particles.Cam_float:NiagaraFloat=&RemapRange(
     Input Value To Remap=/Niagara/DynamicInputs/Camera/DistanceToCamera(Use Owner Position:NiagaraBool=true, Camera Query:NiagaraDataInterfaceCamera=DataInterface())
     Input Min=80
     Input Max=200
     Output Min=0.1
     Output Max=1
    )) {isSetParameters=true}
   &UpdateMeshOrientation(
    Rotation Rate=&FloatFromCurve(FloatCurve=DataInterface() {propertyValues {Curve {keys [<at 0 1.5>,<at 1 1>]}}})
    Rotation Vector=&RandomRangeVector(Minimum=<-1 -1 -1>, Maximum=<1 1 1>)
   )
   &ScaleMeshSize(Scale Factor=&Multiply_VectorByFloat(
     Vector=<1 1 1>
     Float=&FloatFromCurve(
      FloatCurve=DataInterface() {propertyValues {Curve {keys [
          {tangentMode=RCTM_User, time=0, value=1}
          {arriveTangent=-0.0405, interpMode=RCIM_Cubic, leaveTangent=-0.0405, tangentMode=RCTM_User, time=0.7594, value=0.9647}
          {arriveTangent=-3.3468, interpMode=RCIM_Cubic, leaveTangent=-3.3468, tangentMode=RCTM_User, time=1, value=0}
         ]}}}
      Scale Curve=0.2
     )
    ))
   &Drag(Drag=3, Use Linear Drag=true, Use Rotational Drag=false)
   &InheritSourceMovement(ApplyForce=false, ApplyPosition=true, ApplyPositionFalloff=false)
   &GravityForce(Gravity=&Multiply_VectorByFloat(Vector=&RandomRangeVector(Minimum=<0 0 -980>, Maximum=<0 0 -1200>), Float=1.5))
   &SolveForcesAndVelocity(Clamp Velocity=false, Limit Acceleration=false)
   &ScaleColor(
    Scale Alpha=&Multiply_Float(A=$Particles.Cam_float, B=1)
    Scale RGB=&Multiply_VectorByFloat(
     Vector=<1 1 1>
     Float=&MakeCustomFloatFromBool(
      Bool=$User.bIsFrontEndPreview
      True Float=&RandomRangeFloat(Minimum=4.5, Maximum=5)
      False Float=&RandomRangeFloat(Minimum=0.5, Maximum=1.5)
     )
    )
    ScaleA=true
    ScaleRGB=true
   )
  ]
  rendererClasses [&NiagaraMeshRendererProperties]
  renderers [{rendererClass=&NiagaraMeshRendererProperties, rendererIndex=0}]
 }
 Ground {
  emitterUpdateScript [
   &EmitterState(Inactive Response=::NewEnumerator1, Life Cycle Mode=::NewEnumerator1, Loop Behavior=::NewEnumerator1, Loop Duration=2, UseLoopDelay=false)
   &SpawnBurst_Instantaneous(Use Spawn Probability=false, UseLoopCountLimit=false)
   &SpawnBurst_Instantaneous(Spawn Group=1, Use Spawn Probability=false, UseLoopCountLimit=false) {module=SpawnBurst_Instantaneous001}
   &SpawnBurst_Instantaneous(Spawn Group=2, Spawn Time=0.05, Use Spawn Probability=false, UseLoopCountLimit=false) {module=SpawnBurst_Instantaneous002}
  ]
  particleSpawnScript [
   /SetVariables(Particles.VisibilityTag=&SpawnGroup()) {isSetParameters=true}
   &InitializeParticle(
    Color Mode=::NewEnumerator0
    Lifetime=&MergeHelper_FloatSwitch(0=0.5, 1=1.5, 2=0.3, 3=?, Index=&SpawnGroup(), Num=::NewEnumerator1)
    Mesh Scale Mode=::NewEnumerator3
    Mesh Uniform Scale=3.5
    Sprite Rotation Angle=100
    Sprite Rotation Mode=::NewEnumerator1
    UsePositionOffset=true
    Write Material Random=true
   )
   &CameraOffset(Camera Offset Amount=60)
   &SpriteFacingAndAlignment(Sprite Alignment=<0 0 0>, Sprite Facing=<0 0 1>, WriteAlignment=false, WriteFacing=true) {module=SpriteFacingAndAlignment001}
   /SetVariables(Particles.Dark_Position:NiagaraPosition=&AddVectorToPosition(Position=$Particles.Position, Vector=<1.0994 2.394 0>)) {isSetParameters=true}
   &InitialMeshOrientation(
    Mesh Axis To Orient=<0 0 1>
    Mesh Orientation Mode=::NewEnumerator2
    Orientation Vector=<0 0 1>
    Rotation=<0 0.25 0>
    Use Rotation Vector=true
   )
   /SetVariables() {isSetParameters=true}
   /SetVariables(Particles.Wave_Position:NiagaraPosition=&AddVectorToPosition(Position=$Particles.Position, Vector=<0 0 63.4418>)) {isSetParameters=true}
  ]
  particleUpdateScript [
   &ParticleState()
   /SetVariables(Particles.Cam_float:NiagaraFloat=&RemapRange(
     Input Value To Remap=&DistanceToCamera(Use Owner Position:NiagaraBool=true, Camera Query:NiagaraDataInterfaceCamera=DataInterface())
     Input Min=80
     Input Max=200
     Output Min=0.1
     Output Max=1
    )) {isSetParameters=true}
   &SpriteFacingAndAlignment(Sprite Alignment=<0 0 1>, Sprite Facing=<0 0 1>, WriteAlignment=true, WriteFacing=true) {enabled=false}
   /SetVariables(Particles.Decal_Light_Color:LinearColor=&MakeLinearColorFromVectorAndFloat(
     "Vector (RGB)"=&MakeVectorFromLinearColorRGB(Color=<rgba 5 2.8596 0 1>)
     "Float (Alpha)"=&Multiply_Float(
      A=$Particles.Cam_float
      B=&FloatFromCurve(FloatCurve=DataInterface() {propertyValues {Curve {keys [<at 0 1 RCIM_Cubic>, {arriveTangent=-1, interpMode=RCIM_Cubic, leaveTangent=-1, time=0.5, value=0.85}, <at 1 0 RCIM_Cubic>]}}})
     )
    )) {isSetParameters=true}
   /SetVariables(Particles.Decal_Light_Size:Vector2f=&Multiply_Vector2D_ByFloat(Float=250, Vector2D=<1 1>)) {isSetParameters=true}
   /SetVariables(Particles.Decal_Dark_size:Vector2f=&Multiply_Vector2D_ByFloat(Float=1.0393, Vector2D=$Particles.Decal_Light_Size)) {isSetParameters=true}
   /SetVariables(Particles.Decal_Dark_Color:LinearColor=&MakeLinearColorFromVectorAndFloat(
     "Vector (RGB)"=&MakeVectorFromLinearColorRGB(Color=<rgba 0.026 0.026 0.026 1>)
     "Float (Alpha)"=&Multiply_Float(
      A=$Particles.Cam_float
      B=&FloatFromCurve(
       FloatCurve=DataInterface() {propertyValues {Curve {keys [<at 0 1 RCIM_Cubic>, {arriveTangent=-1, interpMode=RCIM_Cubic, leaveTangent=-1, time=0.5, value=0.85}, <at 1 0 RCIM_Cubic>]}}}
       Scale Curve=4
      )
     )
    )) {isSetParameters=true}
   /SetVariables(Particles.Glow_Size:Vector2f=&Multiply_Vector2D_ByFloat(Float=184, Vector2D=<1 1>)) {isSetParameters=true}
   /SetVariables(Particles.Glow_Size2:Vector2f=&Multiply_Vector2D_ByFloat(Float=1.5, Vector2D=$Particles.Glow_Size)) {isSetParameters=true}
   /SetVariables(Particles.Glow_Color:LinearColor=&MakeLinearColorFromVectorAndFloat(
     "Vector (RGB)"=&MakeVectorFromLinearColorRGB(Color=<rgba 1 0.4565 0 1>)
     "Float (Alpha)"=&Multiply_Float(
      A=$Particles.Cam_float
      B=&FloatFromCurve(
       FloatCurve=DataInterface() {propertyValues {Curve {keys [<at 0 1 RCIM_Cubic>, {arriveTangent=-1, interpMode=RCIM_Cubic, leaveTangent=-1, time=0.5911, value=0.5951}, <at 1 0>]}}}
       Scale Curve=0.2
      )
     )
    )) {isSetParameters=true}
   /SetVariables(Particles.Dark_Color:LinearColor=&MakeLinearColorFromVectorAndFloat(
     "Vector (RGB)"=&MakeVectorFromLinearColorRGB(Color=<rgba 0.125 0.0627 0 1>)
     "Float (Alpha)"=&Multiply_Float(
      A=$Particles.Cam_float
      B=&FloatFromCurve(
       FloatCurve=DataInterface() {propertyValues {Curve {keys [<at 0 1 RCIM_Cubic>, {arriveTangent=-1, interpMode=RCIM_Cubic, leaveTangent=-1, time=0.5911, value=0.5951}, <at 1 0>]}}}
       Scale Curve=0.5
      )
     )
    )) {isSetParameters=true}
   /SetVariables(Particles.Dark_Size:Vector2f=&Multiply_Vector2D_ByFloat(Float=1.2868, Vector2D=$Particles.Glow_Size)) {isSetParameters=true}
   /SetVariables(Particles.Dark_Size2:Vector2f=&Multiply_Vector2D_ByFloat(Float=1.5, Vector2D=$Particles.Dark_Size)) {isSetParameters=true}
   &DynamicMaterialParameters(
    Index 0 Param 2=1
    Index 0 Param 4=&FloatFromCurve(FloatCurve=DataInterface() {propertyValues {Curve {keys [<at 0 1>,<at 0.6 1>,<at 1 0>]}}})
    Module.Index0Float0Write=true
    Module.Index0Float1Write=true
    Module.Index0Float2Write=true
    Module.Index0Float3Write=true
   )
   /SetVariables(
    Particles.WaveColor:LinearColor=&MultiplyLinearColor(
     Linear Color=<rgba 1 0.8074 0.1619 1>
     Scale Factor=/Niagara/DynamicInputs/Multiply/Multiply_Vector4_ByFloat(
      Vector4:Vector4f=&Vector4FromCurve(Vector4Curve=DataInterface() {propertyValues {
         AlphaCurve {keys [{interpMode=RCIM_Cubic, tangentMode=RCTM_User, time=0, value=0.3}, <at 1 0>]}
         BlueCurve {keys [<at 0 1>,<at 1 1>]}
         GreenCurve {keys [<at 0 1>,<at 1 1>]}
         RedCurve {keys [<at 0 1>,<at 1 1>]}
        }})
      Float:NiagaraFloat=&FN_FrontEndFloatModifier(Gameplay=1, Front End=0.5)
     )
    )
    Particles.WaveDynamicMaterialParameter:Vector4f=&MakeVector4(
     X=&FloatFromCurve(FloatCurve=DataInterface() {propertyValues {Curve {keys [{arriveTangent=-2.1015, interpMode=RCIM_Cubic, leaveTangent=-2.1015, tangentMode=RCTM_User, time=0, value=0.5}, <at 1 -0.15>]}}})
     Y=&FloatFromCurve(FloatCurve=DataInterface() {propertyValues {Curve {keys [{arriveTangent=-0.3016, interpMode=RCIM_Cubic, leaveTangent=-0.3016, tangentMode=RCTM_User, time=0, value=1}, <at 1 0.9>]}}})
     Z=&FloatFromCurve(FloatCurve=DataInterface() {propertyValues {Curve {keys [
          {arriveTangent=-0.0878, interpMode=RCIM_Cubic, leaveTangent=-0.0878, tangentMode=RCTM_User, time=0, value=0}
          {arriveTangent=1.2291, interpMode=RCIM_Cubic, leaveTangent=1.2291, tangentMode=RCTM_User, time=0.1859, value=0.3101}
          <at 1 0.8>
         ]}}})
     W=0
    )
    Particles.WaveScale:Vector3f=&Multiply_VectorByFloat(
     Vector=$Particles.Scale
     Float=&FloatFromCurve(FloatCurve=DataInterface() {propertyValues {Curve {defaultValue=0.5, keys [{tangentMode=RCTM_User, time=0, value=0}, <at 0.0357 0.85>, {tangentMode=RCTM_User, time=1, value=1}]}}})
    )
   ) {isSetParameters=true}
   /SetVariables(Particles.Wave_Scale2:Vector3f=&Multiply_VectorByFloat(
     Vector=$Particles.Scale
     Float=&FloatFromCurve(FloatCurve=DataInterface() {propertyValues {Curve {defaultValue=0.5, keys [{tangentMode=RCTM_User, time=0, value=0}, <at 0.0357 0.85>, {tangentMode=RCTM_User, time=1, value=1}]}}})
    )) {isSetParameters=true}
  ]
  rendererClasses [&NiagaraSpriteRendererProperties, &NiagaraMeshRendererProperties]
  renderers [
   {rendererClass=&NiagaraSpriteRendererProperties}
   {rendererClass=&NiagaraSpriteRendererProperties, rendererIndex=1}
   {rendererClass=&NiagaraSpriteRendererProperties, rendererIndex=2}
   {rendererClass=&NiagaraSpriteRendererProperties, rendererIndex=3}
   {rendererClass=&NiagaraMeshRendererProperties, rendererIndex=4}
   {rendererClass=&NiagaraMeshRendererProperties, rendererIndex=5}
   {rendererClass=&NiagaraSpriteRendererProperties, rendererIndex=6}
   {rendererClass=&NiagaraSpriteRendererProperties, rendererIndex=7}
  ]
 }
 Lightning {
  emitterUpdateScript [
   &EmitterState(Life Cycle Mode=::NewEnumerator1, Loop Behavior=::NewEnumerator1, Loop Delay=0.1, Loop Duration=0.35, UseLoopDelay=true)
   &SpawnRate(SpawnRate=10, Use Spawn Probability=false)
  ]
  particleSpawnScript [
   &PartitionParticles()
   &InitializeParticle(
    Lifetime Max=0.15
    Lifetime Min=0.1
    Lifetime Mode=::NewEnumerator1
    Mesh Scale=&RandomRangeVector(Minimum=<1 1 2>, Maximum=<1 1 3>)
    Mesh Scale Mode=::NewEnumerator4
    UsePositionOffset=false
    Write Material Random=true
   )
   &CameraOffset(Camera Offset Amount=60)
   &ShapeLocation(
    Handle Radius=0
    Large Radius=20
    MaskSpawnGroup=false
    Offset Mode=::NewEnumerator0
    Override Seed=false
    Shape Primitive=::NewEnumerator3
    UseOwnerScale=true
   )
   &InitialMeshOrientation(
    Mesh Orientation Mode=::NewEnumerator2
    Rotation=&RandomRangeVector(Minimum=<0 -0.08 0>, Maximum=<0 -0.04 0>)
    Use Rotation Vector=true
   ) {module=InitialMeshOrientation002}
   &AddVelocity(Module.Use Falloff Distance=false, Origin Offset=<0 0 100>, Override Seed=false, Velocity Mode=::NewEnumerator1, Velocity Speed=1)
   &InitialMeshOrientation(
    Mesh Orientation Mode=::NewEnumerator2
    Rotation=&RandomRangeVector(Minimum=<0 0 0>, Maximum=<0 0 1>)
    Use Rotation Vector=true
   )
   /SetVariables(Particles.Random=&RandomRangeFloat(Minimum=0, Maximum=1)) {isSetParameters=true}
  ]
  particleUpdateScript [
   &ParticleState()
   /SetVariables(Particles.Cam_float:NiagaraFloat=&RemapRange(
     Input Value To Remap=&DistanceToCamera(Use Owner Position:NiagaraBool=true, Camera Query:NiagaraDataInterfaceCamera=DataInterface())
     Input Min=80
     Input Max=200
     Output Min=0.1
     Output Max=1
    )) {isSetParameters=true}
   &ScaleColor(
    Scale Alpha=&Multiply_Float(A=$Particles.Cam_float, B=5)
    Scale RGB=&MakeVectorFromLinearColorRGB(Color=&ColorFromCurve(ColorCurve=DataInterface() {propertyValues {
        AlphaCurve {keys []}
        BlueCurve {keys [<at 0.0805 0>,<at 0.1231 6.9792>,<at 0.4469 5.8453>,<at 0.5413 0.3385>]}
        GreenCurve {keys [<at 0.0805 0>,<at 0.1231 9.9061>,<at 0.4469 9.1119>,<at 0.5413 2.6974>]}
        RedCurve {keys [<at 0.0805 0>,<at 0.1231 10>,<at 0.4469 10>,<at 0.5413 5>]}
       }}))
    ScaleA=true
    ScaleRGB=true
   )
   &ScaleMeshSize(Scale Factor=&Multiply_VectorByFloat(Vector=<1 1 1>, Float=0.7))
   &DynamicMaterialParameters(
    Index 0 Param 1=1
    Index 0 Param 3=&FloatFromCurve(FloatCurve=DataInterface() {propertyValues {Curve {keys [
         {arriveTangent=4.0072, interpMode=RCIM_Cubic, leaveTangent=4.0072, tangentMode=RCTM_User, time=0, value=0}
         {arriveTangent=4.2299, interpMode=RCIM_Cubic, leaveTangent=4.2299, time=0.2543, value=0.6225}
         {arriveTangent=3.1884, interpMode=RCIM_Cubic, leaveTangent=3.1884, time=0.4395, value=1.8588}
         {arriveTangent=1.6993, interpMode=RCIM_Cubic, leaveTangent=1.6993, tangentMode=RCTM_User, time=1, value=3}
        ]}}})
    Index 0 Param 4=5
    Index 1 Param 1=&RandomRangeFloat(Minimum=2, Maximum=4)
    Index 1 Param 2=&RandomRangeFloat(Minimum=0, Maximum=1)
    Index 1 Param 3=1
    Index 1 Param 4=1
    Module.Param0WriteEnabled=true
    Module.Param0WriteEnabled1=true
    Module.Param1WriteEnabled=true
    Module.Param1WriteEnabled1=true
    Module.Param2WriteEnabled=true
    Module.Param2WriteEnabled1=true
    Module.Param3WriteEnabled=true
    Module.Param3WriteEnabled1=true
    Write Parameter Index 1=true
   )
   /SetVariables(Particles.Lightning glow color:LinearColor=&MakeLinearColorFromVectorAndFloat(
     "Vector (RGB)"=&MakeVectorFromLinearColorRGB(Color=&ColorFromCurve(ColorCurve=DataInterface() {propertyValues {
         AlphaCurve {keys []}
         BlueCurve {keys [<at 0.0753 0>,<at 0.1249 1.0417>,<at 0.4469 0.4167>]}
         GreenCurve {keys [<at 0.0753 0>,<at 0.1249 4.8193>,<at 0.4469 4.3349>]}
         RedCurve {keys [<at 0.0753 0>,<at 0.1249 10>,<at 0.4469 10>]}
        }}))
     "Float (Alpha)"=&FloatFromCurve(FloatCurve=DataInterface() {propertyValues {Curve {keys [<at 0 1>,<at 0.5 1>,<at 1 0>]}}}, Scale Curve=0.3)
    )) {isSetParameters=true}
   &SolveForcesAndVelocity(Clamp Velocity=false, Limit Acceleration=false)
   /SetVariables(Particles.Glow Dy02:Vector4f=&MakeVector4(X=1, Y=0, Z=0, W=5)) {isSetParameters=true}
   /SetVariables(Particles.Glow Dy03:Vector4f=&MakeVector4(X=&RandomRangeFloat(Minimum=2, Maximum=4), Y=0, Z=1, W=1)) {isSetParameters=true}
  ]
  rendererClasses [&NiagaraMeshRendererProperties]
  renderers [
   {rendererClass=&NiagaraMeshRendererProperties, rendererIndex=0}
   {rendererClass=&NiagaraMeshRendererProperties, rendererIndex=1}
  ]
 }
 spikes {
  emitterUpdateScript [
   &EmitterState(Life Cycle Mode=::NewEnumerator1, Loop Behavior=::NewEnumerator1, Loop Duration=1.5, UseLoopDelay=false)
   &SpawnBurst_Instantaneous(Spawn Count=15, Use Spawn Probability=false, UseLoopCountLimit=false)
   &SpawnBurst_Instantaneous(Spawn Count=15, Spawn Time=0.1, Use Spawn Probability=false, UseLoopCountLimit=false) {module=SpawnBurst_Instantaneous001}
   &SpawnBurst_Instantaneous(Spawn Count=15, Spawn Group=1, Use Spawn Probability=false, UseLoopCountLimit=false) {module=SpawnBurst_Instantaneous002}
   &SpawnBurst_Instantaneous(Spawn Count=15, Spawn Group=1, Spawn Time=0.1, Use Spawn Probability=false, UseLoopCountLimit=false) {module=SpawnBurst_Instantaneous003}
   &SpawnBurst_Instantaneous(Spawn Count=20, Spawn Group=2, Use Spawn Probability=false, UseLoopCountLimit=false) {module=SpawnBurst_Instantaneous004}
   &SpawnBurst_Instantaneous(Spawn Count=15, Spawn Group=2, Spawn Time=0.1, Use Spawn Probability=false, UseLoopCountLimit=false) {module=SpawnBurst_Instantaneous005}
   &SpawnBurst_Instantaneous(Spawn Count=10, Spawn Group=3, Use Spawn Probability=false, UseLoopCountLimit=false) {module=SpawnBurst_Instantaneous006}
   &SpawnBurst_Instantaneous(Spawn Count=10, Spawn Group=3, Spawn Time=0.1, Use Spawn Probability=false, UseLoopCountLimit=false) {module=SpawnBurst_Instantaneous007}
   &SpawnBurst_Instantaneous(Spawn Count=5, Spawn Group=4, Spawn Time=0.05, Use Spawn Probability=false, UseLoopCountLimit=false) {module=SpawnBurst_Instantaneous008}
   &SpawnBurst_Instantaneous(Spawn Count=5, Spawn Group=4, Spawn Time=0.1, Use Spawn Probability=false, UseLoopCountLimit=false) {module=SpawnBurst_Instantaneous009}
   &SpawnRate(
    SpawnGroup=5
    SpawnRate=&FloatFromCurve(
     FloatCurve=DataInterface() {propertyValues {Curve {keys [<at 0 1>,<at 0.2879 1>,<at 0.35 0>]}}}
     CurveIndex=$Emitter.NormalizedLoopAge
     Scale Curve=50
    )
    Use Spawn Probability=false
   )
  ]
  particleSpawnScript [
   /SetVariables(Particles.VisibilityTag=&SpawnGroup()) {isSetParameters=true}
   &InitializeParticle(
    Color=<rgba 5 3.7892 0.625 1>
    Lifetime=&MergeHelper_FloatSwitch(
     0=&RandomRangeFloat(Minimum=0.1, Maximum=0.2)
     1=&RandomRangeFloat(Minimum=0.1, Maximum=0.2)
     2=&RandomRangeFloat(Minimum=0.2, Maximum=0.8)
     3=&RandomRangeFloat(Minimum=1, Maximum=1.5)
     4=&RandomRangeFloat(Minimum=1.1, Maximum=1.7)
     5=&RandomRangeFloat(Minimum=0.08, Maximum=0.12)
     Index=&SpawnGroup()
     Num=::NewEnumerator4
    )
    Sprite Rotation Angle=&MergeHelper_FloatSwitch(
     0=0
     1=0
     2=0
     3=0
     4=&RandomRangeFloat(Minimum=0, Maximum=1)
     5=0
     Index=&SpawnGroup()
     Num=::NewEnumerator4
    )
    Sprite Rotation Mode=::NewEnumerator1
    Sprite Size Mode=::NewEnumerator3
    Uniform Sprite Size=&MergeHelper_FloatSwitch(
     0=0
     1=0
     2=&Lerp_Float(
      A=6
      B=10
      Alpha=&FloatFromCurve(
       FloatCurve=DataInterface() {propertyValues {Curve {keys [
           {interpMode=RCIM_Cubic, tangentMode=RCTM_User, time=0, value=0}
           {arriveTangent=1.2457, interpMode=RCIM_Cubic, leaveTangent=1.2457, tangentMode=RCTM_User, time=0.7348, value=0.4334}
           {interpMode=RCIM_Cubic, tangentMode=RCTM_User, time=1, value=1}
          ]}}}
       CurveIndex=&RandomRangeFloat(Minimum=0, Maximum=1)
      )
     )
     3=&RandomRangeFloat(Minimum=250, Maximum=350)
     4=&RandomRangeFloat(Minimum=150, Maximum=250)
     Index=&SpawnGroup()
     Num=::NewEnumerator3
    )
    UsePositionOffset=false
    Write Material Random=true
   )
   &SubUVAnimation(
    End Frame:NiagaraInt32=&MergeHelper_IntSwitch(0=0, 1=0, 2=0, 3=0, 4=3, 5=0, Index=&SpawnGroup(), Num=::NewEnumerator4)
    SubUV Animation Mode:ENiagara_SubUVLookupMode=::NewEnumerator2
   )
   &ShapeLocation(
    MaskSpawnGroup=false
    Non Uniform Scale=&MergeHelper_VectorSwitch(
     0=<1 1 0.5>
     1=<1 1 0.5>
     2=<1 1 0.5>
     3=<1 1 0.5>
     4=<1 1 0.5>
     5=<1 1 1>
     Index=&SpawnGroup()
     Num=::NewEnumerator4
    )
    Offset Mode=::NewEnumerator0
    Override Seed=false
    Shape Origin=&MakeVector(X=0, Y=0, Z=0)
    Sphere Radius=&MergeHelper_FloatSwitch(0=0, 1=0, 2=50, 3=0, 4=50, 5=60, Index=&SpawnGroup(), Num=::NewEnumerator4)
    UseOwnerScale=true
   )
   &TorusLocation(
    Align Torus With Axis=false
    Handle Radius=0
    Large Radius:NiagaraFloat=&MergeHelper_FloatSwitch(0=10, 1=10, 2=0, 3=60, 4=0, 5=0, Index=&SpawnGroup(), Num=::NewEnumerator4)
    MaskSpawnGroup=false
    Module.Override Randomness=false
    U Distribution:NiagaraFloat=&MergeHelper_FloatSwitch(0=0, 1=0, 2=0, 3=0, 4=0, 5=0, Index=&SpawnGroup(), Num=::NewEnumerator4)
   )
   &AddVelocity(
    Module.Use Falloff Distance=false
    Origin Offset=&MergeHelper_VectorSwitch(
     0=&RandomRangeVector(Minimum=<0 0 -5>, Maximum=<0 0 -10>)
     1=&RandomRangeVector(Minimum=<0 0 -5>, Maximum=<0 0 -10>)
     2=<0 0 -15>
     3=<0 0 -60>
     4=<0 0 -80>
     5=<0 0 -70>
     Index=&SpawnGroup()
     Num=::NewEnumerator4
    )
    Override Seed=false
    Velocity Mode=::NewEnumerator1
    Velocity Origin=&MakeVector(X=0, Y=0, Z=0)
    Velocity Speed=&MergeHelper_FloatSwitch(
     0=&RandomRangeFloat(Minimum=600, Maximum=800)
     1=&RandomRangeFloat(Minimum=600, Maximum=800)
     2=400
     3=&Multiply_Float(A=&RandomRangeFloat(Minimum=300, Maximum=600), B=1)
     4=&Multiply_Float(A=&RandomRangeFloat(Minimum=75, Maximum=150), B=4)
     5=&RandomRangeFloat(Minimum=800, Maximum=1200)
     Index=&SpawnGroup()
     Num=::NewEnumerator4
    )
   )
   &AddVelocity(
    Velocity=&MergeHelper_VectorSwitch(
     0=<0 0 0>
     1=<0 0 0>
     2=&Multiply_VectorByFloat(Vector=<0 0 1>, Float=&RandomRangeFloat(Minimum=200, Maximum=400))
     3=<0 0 0>
     4=<0 0 0>
     5=<0 0 0>
     Index=&SpawnGroup()
     Num=::NewEnumerator4
    )
    Velocity Speed Scale=2
   ) {module=AddVelocity001}
   &CameraOffset(Camera Offset Amount=80)
  ]
  particleUpdateScript [
   &ParticleState()
   /SetVariables(Particles.Cam_float:NiagaraFloat=&RemapRange(
     Input Value To Remap=&DistanceToCamera(Use Owner Position:NiagaraBool=true, Camera Query:NiagaraDataInterfaceCamera=DataInterface())
     Input Min=80
     Input Max=200
     Output Min=0.1
     Output Max=1
    )) {isSetParameters=true}
   &OffsetPosition(Position Offset=&MergeHelper_VectorSwitch(0=<0 0 0>, 1=<0 0 0>, 2=<0 0 0>, 3=<0 0 0>, 4=<0 0 0>, 5=<0 0 50>, Index=&SpawnGroup(), Num=::NewEnumerator4))
   &SpriteRotationRate(Rotation Rate=&MergeHelper_FloatSwitch(
     0=0
     1=0
     2=0
     3=&RandomRangeFloat(Minimum=-300, Maximum=300)
     4=&RandomRangeFloat(Minimum=-50, Maximum=50)
     Index=&SpawnGroup()
     Num=::NewEnumerator3
    ))
   &Drag(
    Drag=&MergeHelper_FloatSwitch(
     0=6
     1=2
     2=0.25
     3=5
     4=&RandomRangeFloat(Minimum=0.25, Maximum=0.5)
     Index=&SpawnGroup()
     Num=::NewEnumerator3
    )
    Rotational Drag:NiagaraFloat=&MergeHelper_FloatSwitch(0=1, 1=1, 2=1, 3=1, 4=5, Index=&SpawnGroup(), Num=::NewEnumerator3)
    Use Linear Drag=true
    Use Rotational Drag=true
   )
   &GravityForce(Gravity=&MergeHelper_VectorSwitch(0=<0 0 0>, 1=<0 0 0>, 2=<0 0 -980>, 3=<0 0 500>, 4=<0 0 -980>, Index=&SpawnGroup(), Num=::NewEnumerator3))
   &SolveForcesAndVelocity(Clamp Velocity=false, Limit Acceleration=false)
   /SetVariables(Particles.Spike01_Size:Vector2f=&Multiply_Vector2D(
     A=&RandomRangeVector2D(Minimum=<20 100>, Maximum=<70 180>)
     B=&Vector2DFromCurve(
      Vector2Curve=DataInterface() {propertyValues {
        XCurve {keys [{tangentMode=RCTM_User, time=0, value=1}, {tangentMode=RCTM_User, time=1, value=0.8}]}
        YCurve {keys [{tangentMode=RCTM_User, time=0, value=0.5}, {tangentMode=RCTM_User, time=1, value=1.7}]}
       }}
      ScaleCurve=&Multiply_Vector2D_ByFloat(Float=0.6, Vector2D=<1 1>)
     )
    )) {isSetParameters=true}
   /SetVariables(Particles.Spike01_Color:LinearColor=&MakeLinearColorFromVectorAndFloat(
     "Vector (RGB)"=&MakeVectorFromLinearColorRGB(Color=&MultiplyLinearColor(Linear Color=<rgba 5 3.7892 0.625 1>, Scale Factor=&MakeVector4_FromFloat(Float=1)))
     "Float (Alpha)"=&Multiply_Float(
      A=$Particles.Cam_float
      B=&FloatFromCurve(FloatCurve=DataInterface() {propertyValues {Curve {keys [<at 0 0>,<at 0.5 1>,<at 1 0>]}}})
     )
    )) {isSetParameters=true}
   /SetVariables(Particles.Spike02_Size:Vector2f=&Multiply_Vector2D(
     A=&RandomRangeVector2D(Minimum=<20 90>, Maximum=<40 90>)
     B=&Vector2DFromCurve(
      Vector2Curve=DataInterface() {propertyValues {
        XCurve {keys [{tangentMode=RCTM_User, time=0, value=1}, {tangentMode=RCTM_User, time=1, value=0.8}]}
        YCurve {keys [{tangentMode=RCTM_User, time=0, value=0.5}, {tangentMode=RCTM_User, time=1, value=1.7}]}
       }}
      ScaleCurve=<1 1>
     )
    )) {isSetParameters=true}
   /SetVariables(Particles.Spike02_Color:LinearColor=&MakeLinearColorFromVectorAndFloat(
     "Vector (RGB)"=&MakeVectorFromLinearColorRGB(Color=&MultiplyLinearColor(Linear Color=<rgba 0.3177 0.1134 0.0187 10>, Scale Factor=&MakeVector4_FromFloat(Float=1)))
     "Float (Alpha)"=&Multiply_Float(
      A=$Particles.Cam_float
      B=&FloatFromCurve(FloatCurve=DataInterface() {propertyValues {Curve {keys [<at 0 0>,<at 0.5 1>,<at 1 0>]}}})
     )
    )) {isSetParameters=true}
   /SetVariables(Particles.Streak_Size:Vector2f=&Multiply_Vector2D(
     A=&Multiply_Vector2D_ByFloat(Float=40, Vector2D=&RandomRangeVector2D(Minimum=<2 2>, Maximum=<3 3>))
     B=&Multiply_Vector2D(
      A=<0.8 1>
      B=&Multiply_Vector2D_ByFloat(
       Float=0.7
       Vector2D=&Vector2DFromCurve(
        Vector2Curve=DataInterface() {propertyValues {
          XCurve {keys [
            {arriveTangent=0.8609, interpMode=RCIM_Cubic, leaveTangent=0.8609, tangentMode=RCTM_User, time=0, value=0.6}
            {arriveTangent=-0.9108, interpMode=RCIM_Cubic, leaveTangent=-0.9108, tangentMode=RCTM_User, time=0.3, value=0.5}
            {arriveTangent=0.0224, interpMode=RCIM_Cubic, leaveTangent=0.0224, tangentMode=RCTM_User, time=1, value=0.5}
           ]}
          YCurve {keys [
            {arriveTangent=0.8801, interpMode=RCIM_Cubic, leaveTangent=0.8801, tangentMode=RCTM_User, time=0, value=0.45}
            {arriveTangent=0.6586, interpMode=RCIM_Cubic, leaveTangent=0.6586, tangentMode=RCTM_User, time=0.2, value=0.85}
            {arriveTangent=-0.4778, interpMode=RCIM_Cubic, leaveTangent=-0.4778, tangentMode=RCTM_User, time=1, value=2}
           ]}
         }}
        ScaleCurve=<0.8 2.5>
       )
      )
     )
    )) {isSetParameters=true}
   /SetVariables(Particles.Streak_Dynamic:Vector4f=&MakeVector4(
     X=&FloatFromCurve(FloatCurve=DataInterface() {propertyValues {Curve {keys [
          {arriveTangent=-0.1846, time=0, value=0}
          {arriveTangent=1.3385, interpMode=RCIM_Cubic, leaveTangent=1.1594, tangentMode=RCTM_Break, time=0.5, value=0}
          {arriveTangent=2, interpMode=RCIM_Cubic, leaveTangent=2, time=0.8, value=0.5}
          {arriveTangent=3.0091, interpMode=RCIM_Cubic, leaveTangent=3.0091, tangentMode=RCTM_User, time=1, value=1}
         ]}}})
     Y=1
     Z=0
     W=&FloatFromCurve(FloatCurve=DataInterface() {propertyValues {Curve {keys [<at 0 0 RCIM_Cubic>,<at 1 1 RCIM_Cubic>]}}}, Scale Curve=0.8)
    )) {isSetParameters=true}
   /SetVariables(Particles.Streak_Color:LinearColor=&MakeLinearColorFromVectorAndFloat(
     "Vector (RGB)"=&MakeVectorFromLinearColorRGB(Color=&Multiply_LinearColorByFloat(
       Linear Color=&ColorFromCurve(ColorCurve=DataInterface() {propertyValues {
          AlphaCurve {keys []}
          BlueCurve {keys [<at 0.1952 1 RCIM_Cubic>,<at 0.4876 0.1562 RCIM_Cubic>]}
          GreenCurve {keys [<at 0.1952 1 RCIM_Cubic>,<at 0.4876 0.1562 RCIM_Cubic>]}
          RedCurve {keys [<at 0.1952 1 RCIM_Cubic>,<at 0.4876 0.1562 RCIM_Cubic>]}
         }})
       Float=10
      ))
     "Float (Alpha)"=&Multiply_Float(
      A=$Particles.Cam_float
      B=&FloatFromCurve(FloatCurve=DataInterface() {propertyValues {Curve {keys [<at 0 1>,<at 0.5 0.85>,<at 1 0>]}}})
     )
    )) {isSetParameters=true}
   /SetVariables(Particles.Dirt_Size:Vector2f=&Multiply_Vector2D(
     A=$Particles.Initial.SpriteSize
     B=&Multiply_Vector2D_ByFloat(
      Float=&FloatFromCurve(
       FloatCurve=DataInterface() {propertyValues {Curve {keys [
           {arriveTangent=8.2127, interpMode=RCIM_Cubic, leaveTangent=8.2127, tangentMode=RCTM_User, time=0, value=0.5}
           {
            arriveTangent=4.2448
            arriveTangentWeight=0.232
            interpMode=RCIM_Cubic
            leaveTangent=4.2448
            leaveTangentWeight=2.2538
            tangentMode=RCTM_User
            time=0.07
            value=1
           }
           {arriveTangent=-2.5194, interpMode=RCIM_Cubic, leaveTangent=-2.5194, tangentMode=RCTM_User, time=0.75, value=0.8}
           {arriveTangent=-4.4982, interpMode=RCIM_Cubic, leaveTangent=-4.4982, tangentMode=RCTM_User, time=1, value=0}
          ]}}}
       Scale Curve=&FN_FrontEndFloatModifier(Gameplay=1.3, Front End=0.8)
      )
      Vector2D=<0.2 1.5>
     )
    )) {isSetParameters=true}
   /SetVariables(Particles.Dirt_Color:LinearColor=&MakeLinearColorFromVectorAndFloat(
     "Vector (RGB)"=&MakeVectorFromLinearColorRGB(Color=&MultiplyLinearColor(
       Linear Color=<rgba 2.5 2.3981 0.3581 1.2>
       Scale Factor=&MakeVector4_FromFloat(Float=&FloatFromCurve(
         FloatCurve=DataInterface() {propertyValues {Curve {keys [
             {arriveTangent=-9.5321, interpMode=RCIM_Cubic, leaveTangent=-9.5321, tangentMode=RCTM_User, time=0, value=4}
             {interpMode=RCIM_Cubic, tangentMode=RCTM_User, time=1, value=0.4}
            ]}}}
         Scale Curve=&RandomRangeFloat(Minimum=1, Maximum=1.5)
        ))
      ))
     "Float (Alpha)"=&Multiply_Float(
      A=$Particles.Cam_float
      B=&Lerp_Float(
       A=0
       B=1
       Alpha=&FloatFromCurve(
        FloatCurve=DataInterface() {propertyValues {Curve {keys [
            {interpMode=RCIM_Cubic, tangentMode=RCTM_User, time=0, value=1}
            {interpMode=RCIM_Cubic, tangentMode=RCTM_User, time=0.5, value=1}
            {tangentMode=RCTM_User, time=1, value=0}
           ]}}}
        Scale Curve=2
       )
      )
     )
    )) {isSetParameters=true}
   /SetVariables(Particles.Smoke_Size:Vector2f=&Multiply_Vector2D(
     A=$Particles.Initial.SpriteSize
     B=&Vector2DFromFloat(Value=&FloatFromCurve(
       FloatCurve=DataInterface() {propertyValues {Curve {keys [
           {arriveTangent=0.0418, interpMode=RCIM_Cubic, leaveTangent=0.0418, tangentMode=RCTM_User, time=0, value=0.3}
           {arriveTangent=0.8771, interpMode=RCIM_Cubic, leaveTangent=0.8771, tangentMode=RCTM_User, time=0.3432, value=0.6656}
           {arriveTangent=0.7996, arriveTangentWeight=0.3375, interpMode=RCIM_Cubic, leaveTangent=0.7996, tangentMode=RCTM_User, time=1, value=1.5}
          ]}}}
       Scale Curve=0.9
      ))
    )) {isSetParameters=true}
   /SetVariables(Particles.Smoke_Color:LinearColor=&MakeLinearColorFromVectorAndFloat(
     "Vector (RGB)"=&Multiply_Vector(
      A=&MakeVectorFromLinearColorRGB(Color=&RandomRangeLinearColor(Minimum=<rgba 0.5 0.3617 0.1042 0.1739>, Maximum=<rgba 0.1823 0.1481 0.075 0.3913>))
      B=&MakeVectorFromLinearColorRGB(Color=&RandomRangeLinearColor(
        Minimum=&ColorFromCurve(ColorCurve=DataInterface() {propertyValues {
           AlphaCurve {keys []}
           BlueCurve {keys [<at 0.228 0.0403 RCIM_Cubic>,<at 0.3793 0.0384>,<at 0.6529 1>]}
           GreenCurve {keys [<at 0.228 2.3107 RCIM_Cubic>,<at 0.3793 1.3202>,<at 0.6529 1>]}
           RedCurve {keys [<at 0.228 3 RCIM_Cubic>,<at 0.3793 2>,<at 0.6529 1>]}
          }})
        Maximum=<rgba 0.5182 0.5182 0.5182 1>
       ))
     )
     "Float (Alpha)"=&Multiply_Float(
      A=$Particles.Cam_float
      B=&FloatFromCurve(
       FloatCurve=DataInterface() {propertyValues {Curve {
          defaultValue=1
          keys [
           {interpMode=RCIM_Cubic, tangentMode=RCTM_User, time=-0.015, value=1.5666}
           {arriveTangent=1.0375, interpMode=RCIM_Cubic, leaveTangent=1.0375, tangentMode=RCTM_User, time=0.1343, value=1.5587}
           {arriveTangent=-2.8196, interpMode=RCIM_Cubic, leaveTangent=-2.8196, tangentMode=RCTM_User, time=0.5, value=1}
           {tangentMode=RCTM_User, time=1, value=0}
          ]
         }}}
       Scale Curve=0.1
      )
     )
    )) {isSetParameters=true}
   /SetVariables(
    Particles.Smoke_Dynamic0:Vector4f=&MakeVector4(
     X=&Add_Float(
      A=&RandomRangeFloat(Minimum=0, Maximum=1)
      B=&FloatFromCurve(
       FloatCurve=DataInterface() {propertyValues {Curve {keys [
           {arriveTangent=2.7533, interpMode=RCIM_Cubic, leaveTangent=1.8958, tangentMode=RCTM_User, time=0, value=1}
           {arriveTangent=0.8292, interpMode=RCIM_Cubic, leaveTangent=0.8292, tangentMode=RCTM_User, time=0.321, value=1.3453}
           {arriveTangent=0.0529, interpMode=RCIM_Cubic, leaveTangent=0.2021, tangentMode=RCTM_User, time=1, value=1.5}
          ]}}}
       CurveIndex=&Add_Float(A=&Multiply_Float(A=$Particles.Initial.MaterialRandom, B=0.5), B=$Particles.NormalizedAge)
      )
     )
     Y=0
     Z=&FloatFromCurve(
      FloatCurve=DataInterface() {propertyValues {Curve {keys [
          {arriveTangent=0.0632, interpMode=RCIM_Cubic, leaveTangent=0.0632, tangentMode=RCTM_User, time=0, value=20}
          {
           arriveTangent=-1.6858
           arriveTangentWeight=0.0333
           interpMode=RCIM_Cubic
           leaveTangent=-1.6858
           leaveTangentWeight=0.3137
           tangentMode=RCTM_User
           time=0.05
           value=0.1
          }
          {arriveTangentWeight=0.1027, interpMode=RCIM_Cubic, tangentMode=RCTM_Break, time=0.2, value=0}
         ]}}}
      Scale Curve=&RandomRangeFloat(Minimum=0.25, Maximum=0.5)
     )
     W=&FloatFromCurve(FloatCurve=DataInterface() {propertyValues {Curve {keys [<at 0.5 1>,<at 1 0>]}}})
    )
    Particles.Smoke_Dynamic1:Vector4f=&MakeVector4(
     X=&FloatFromCurve(
      FloatCurve=DataInterface() {propertyValues {Curve {keys [
          {interpMode=RCIM_Cubic, tangentMode=RCTM_Break, time=0, value=1}
          {arriveTangentWeight=0.0833, interpMode=RCIM_Cubic, tangentMode=RCTM_Break, time=0.25, value=0.3}
         ]}}}
      Scale Curve=&RandomRangeFloat(Minimum=1.5, Maximum=3)
     )
     Y=&FloatFromCurve(
      FloatCurve=DataInterface() {propertyValues {Curve {keys [
          {interpMode=RCIM_Cubic, leaveTangent=-2.5489, tangentMode=RCTM_Break, time=0, value=1}
          {interpMode=RCIM_Cubic, leaveTangent=0.9066, tangentMode=RCTM_Break, time=1, value=0}
         ]}}}
      Scale Curve=&RandomRangeFloat(Minimum=0.01, Maximum=0.1)
     )
     Z=&FloatFromCurve(FloatCurve=DataInterface() {propertyValues {Curve {keys [
          {interpMode=RCIM_Cubic, leaveTangent=6.3917, tangentMode=RCTM_Break, time=0, value=0.2}
          {arriveTangent=1.7549, interpMode=RCIM_Cubic, leaveTangent=10.6767, tangentMode=RCTM_User, time=0.25, value=1}
         ]}}})
     W=&FloatFromCurve(
      FloatCurve=DataInterface() {propertyValues {Curve {keys [{arriveTangent=-0.136, interpMode=RCIM_Cubic, leaveTangent=-0.136, tangentMode=RCTM_User, time=0, value=1}, <at 1 0.95>]}}}
      Scale Curve=&RandomRangeFloat(Minimum=1.3, Maximum=2)
     )
    )
    Particles.Smoke_Dynamic2:Vector4f=&MakeVector4(
     X=&FloatFromCurve(FloatCurve=DataInterface() {propertyValues {Curve {keys [
          {interpMode=RCIM_Cubic, leaveTangent=4.2485, tangentMode=RCTM_Break, time=0, value=0}
          {arriveTangent=1.0548, interpMode=RCIM_Cubic, leaveTangent=1.0548, tangentMode=RCTM_User, time=0.17, value=0.5715}
          <at 1 1>
         ]}}})
     Y=&FloatFromCurve(
      FloatCurve=DataInterface() {propertyValues {Curve {keys [
          {arriveTangent=1.9056, interpMode=RCIM_Cubic, leaveTangent=1.9056, tangentMode=RCTM_User, time=0, value=0.2}
          {arriveTangent=0.0259, interpMode=RCIM_Cubic, leaveTangent=0.0259, tangentMode=RCTM_User, time=1, value=1.4}
         ]}}}
      Scale Curve=&RandomRangeFloat(Minimum=0.1, Maximum=0.2)
     )
     Z=&FloatFromCurve(
      FloatCurve=DataInterface() {propertyValues {Curve {keys [
          {interpMode=RCIM_Cubic, leaveTangent=0.0393, tangentMode=RCTM_Break, time=0, value=1}
          {interpMode=RCIM_Cubic, leaveTangent=11.1233, tangentMode=RCTM_Break, time=1, value=1.5}
         ]}}}
      Scale Curve=0.7
     )
     W=&FloatFromCurve(
      FloatCurve=DataInterface() {propertyValues {Curve {keys [
          {arriveTangent=-0.6867, interpMode=RCIM_Cubic, leaveTangent=-0.6867, tangentMode=RCTM_User, time=0, value=1}
          {interpMode=RCIM_Cubic, leaveTangent=-1.2956, tangentMode=RCTM_Break, time=1, value=0.75}
         ]}}}
      Scale Curve=1.5
     )
    )
   ) {isSetParameters=true}
   /SetVariables(Particles.Dirt2_Color:LinearColor=&MakeLinearColorFromVectorAndFloat(
     "Vector (RGB)"=&MakeVectorFromLinearColorRGB(Color=&Multiply_LinearColorByFloat(
       Linear Color=&RandomRangeLinearColor(Minimum=<rgba 0.2188 0.1238 0.0798 1.2>, Maximum=<rgba 0.0625 0.0357 0.0283 1>)
       Float=&RandomRangeFloat(Minimum=0.5, Maximum=1)
      ))
     "Float (Alpha)"=&Multiply_Float(
      A=$Particles.Cam_float
      B=&FloatFromCurve(
       FloatCurve=DataInterface() {propertyValues {Curve {keys [
           {arriveTangent=-0.4559, interpMode=RCIM_Cubic, leaveTangent=-0.4559, tangentMode=RCTM_User, time=0, value=1}
           {arriveTangent=-2.4378, interpMode=RCIM_Cubic, leaveTangent=-2.4378, tangentMode=RCTM_User, time=1.0011, value=0.002}
          ]}}}
       Scale Curve=0.9
      )
     )
    )) {isSetParameters=true}
   /SetVariables(Particles.Dirt2_Size:Vector2f=&Multiply_Vector2D(
     A=$Particles.Initial.SpriteSize
     B=&Vector2DFromFloat(Value=&FloatFromCurve(FloatCurve=DataInterface() {propertyValues {Curve {keys [
           {arriveTangent=0.9995, interpMode=RCIM_Cubic, leaveTangent=0.9995, tangentMode=RCTM_User, time=0, value=0.5059}
           {arriveTangent=-0.0405, interpMode=RCIM_Cubic, leaveTangent=-0.0405, tangentMode=RCTM_User, time=1, value=1}
          ]}}}))
    )) {isSetParameters=true}
   /SetVariables(Particles.Line_Color:LinearColor=&MakeLinearColorFromVectorAndFloat(
     "Vector (RGB)"=&MakeVectorFromLinearColorRGB(Color=<rgba 0.0208 0.0016 0.0006 8>)
     "Float (Alpha)"=&Multiply_Float(
      A=$Particles.Cam_float
      B=&FloatFromCurve(
       FloatCurve=DataInterface() {propertyValues {Curve {keys [
           {interpMode=RCIM_Cubic, tangentMode=RCTM_User, time=0, value=0}
           {interpMode=RCIM_Cubic, tangentMode=RCTM_User, time=0.5, value=1}
           {tangentMode=RCTM_User, time=1, value=0}
          ]}}}
       Scale Curve=8
      )
     )
    )) {isSetParameters=true}
   /SetVariables(Particles.Line_Size:Vector2f=&Multiply_Vector2D(
     A=&RandomRangeVector2D(Minimum=<50 50>, Maximum=<80 150>)
     B=&Vector2DFromCurve(
      Vector2Curve=DataInterface() {propertyValues {
        XCurve {keys [{tangentMode=RCTM_User, time=0, value=0.5}, {tangentMode=RCTM_User, time=1, value=1.5}]}
        YCurve {keys [{tangentMode=RCTM_User, time=0, value=1}, {tangentMode=RCTM_User, time=1, value=0.5}]}
       }}
      ScaleCurve=<1 1>
     )
    )) {isSetParameters=true}
  ]
  rendererClasses [&NiagaraSpriteRendererProperties]
  renderers [
   {rendererClass=&NiagaraSpriteRendererProperties}
   {rendererClass=&NiagaraSpriteRendererProperties, rendererIndex=1}
   {rendererClass=&NiagaraSpriteRendererProperties, rendererIndex=2}
   {rendererClass=&NiagaraSpriteRendererProperties, rendererIndex=3}
   {rendererClass=&NiagaraSpriteRendererProperties, rendererIndex=4}
   {rendererClass=&NiagaraSpriteRendererProperties, rendererIndex=5}
   {rendererClass=&NiagaraSpriteRendererProperties, rendererIndex=6}
  ]
 }
}
summary {
 systemName=NS_Emote_ShovelCrop_GroundHit
 emitters [
  {emitterName=Chunks001, rendererClasses [&NiagaraMeshRendererProperties]}
  {emitterName=Ground, rendererClasses [&NiagaraSpriteRendererProperties, &NiagaraMeshRendererProperties]}
  {emitterName=Lightning, rendererClasses [&NiagaraMeshRendererProperties]}
  {emitterName=spikes, rendererClasses [&NiagaraSpriteRendererProperties]}
 ]
 userVariables [
  User.bIsFrontEnd {type=NiagaraBool}
  User.bIsFrontEndPreview {type=NiagaraBool}
  User.Kill {type=NiagaraFloat, defaultValue=0, description=""}
  User.Switch {type=NiagaraFloat, defaultValue=0, description=""}
  User.VariantSwitch {type=NiagaraFloat, defaultValue=0, description=""}
 ]
}
