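# Niagara effect for the ShovelCrop emote that adds electricity and timed lightning explosions to the player.
# Two electricity emitters attach short-lived meshes and lights to skeleton bones, one gated by User.Active.
# Two delayed explosion emitters fire mesh bursts above the origin, one adding sprites, the other staggering three strikes.
# User.Switch selects white, black or yellow tinting, camera proximity fades alpha, and lights are zeroed in the front end.
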NiagaraSystem /BRCosmetics/Animation/Game/MainPlayer/Emotes/ShovelCrop/Fx/NS_Emote_ShovelCrop_Lightning
dependencies {
 usedDataInterfaces [
  &NiagaraDataInterfaceCurve
  &NiagaraDataInterfaceCamera
  &NiagaraDataInterfaceArrayInt32
  &NiagaraDataInterfaceColorCurve
  &NiagaraDataInterfaceSkeletalMesh
 ]
 usedDynamicInputs [&MultiplyLinearColor, &MakeVector4FromVectorAndFloat]
 usedModules [
  &self:SystemSpawnScript.SystemScriptSource.SystemScriptGraph.NiagaraNodeAssignment_0.SetVariables_#1
  &SystemState
 ]
}
emitters {
 Electricity_01 {
  emitterUpdateScript [
   &EmitterState(Life Cycle Mode=::NewEnumerator1, Loop Behavior=::NewEnumerator1, Loop Delay=2.8, Loop Duration=3, UseLoopDelay=true)
   &SpawnRate(
    SpawnRate=&FloatFromCurve(
     FloatCurve=DataInterface() {propertyValues {Curve {keys [{interpMode=RCIM_Cubic, tangentMode=RCTM_User, time=0.5, value=1}, {interpMode=RCIM_Cubic, tangentMode=RCTM_User, time=1, value=0.2}]}}}
     CurveIndex=$Emitter.NormalizedLoopAge
     Scale Curve=30
    )
    Use Spawn Probability=false
   )
  ]
  particleSpawnScript [
   /SetVariables(Particles.VisibilityTag=&SpawnGroup()) {isSetParameters=true}
   &InitializeParticle(
    Lifetime Max=0.055
    Lifetime Min=0.045
    Lifetime Mode=::NewEnumerator1
    Mesh Scale=&MaskVectorBySpawnGroup(
     Masked Vector=&RandomRangeVector(Minimum=<0.5 1.5 0.9>, Maximum=<1 0.5 0.5>)
     Passthrough Vector=&RandomRangeVector(Minimum=<0.2 0.2 0.2>, Maximum=<0.4 0.4 0.4>)
    )
    Mesh Scale Mode=::NewEnumerator4
    Position Offset=<0 0 20>
    Sprite Size Mode=::NewEnumerator3
    UsePositionOffset=false
    Write Material Random=true
   )
   &CameraOffset(Camera Offset Amount=&RandomRangeFloat(Minimum=20, Maximum=20))
   &InitialMeshOrientation(Mesh Orientation Mode=::NewEnumerator2, Rotation=<0 0.25 0>, Use Rotation Vector=true)
   &InitialMeshOrientation(
    Mesh Orientation Mode=::NewEnumerator2
    Rotation=&RandomRangeVector(Minimum=<0 0 0>, Maximum=<1 0 0>)
    Use Rotation Vector=true
   ) {module=InitialMeshOrientation001}
   &AddVelocity(Module.Use Falloff Distance=false, Override Seed=false, Velocity Mode=::NewEnumerator1, Velocity Speed=1)
  ]
  particleUpdateScript [
   &ParticleState()
   /SetVariables(Particles.Cam_float:NiagaraFloat=&RemapRange(
     Input Value To Remap=/Niagara/DynamicInputs/Camera/DistanceToCamera(Use Owner Position:NiagaraBool=true, Camera Query:NiagaraDataInterfaceCamera=DataInterface())
     Input Min=80
     Input Max=200
     Output Min=0.1
     Output Max=1
    )) {isSetParameters=true}
   &SkeletalMeshLocation(
    Bone / Socket Index=&SelectIntFromArray(
     Array Sampling Mode=::NewEnumerator1
     Direct Array Index=$Particles.PartitionParticles.Partition
     Int Selection Array=DataInterface() {propertyValues {IntData [4,3,14,13,5,1,6,0,2,12,11,8,7,10,9]}}
     Random Seed=0
    )
    Bone Sampling Mode=::NewEnumerator4
    MaskSpawnGroup=false
    Orientation Sampling=::NewEnumerator1
    Skeletal Mesh=$System.Skeletal Mesh
   ) {module=SkeletalMeshLocation001}
   &ScaleColor(
    Scale Alpha=&Multiply_Float(
     A=$Particles.Cam_float
     B=&MakeCustomFloatFromBool(
      Bool=&SetBoolByFloatComparison(A=$User.Switch, B=1)
      True Float=10
      False Float=&MakeCustomFloatFromBool(Bool=&SetBoolByFloatComparison(A=$User.Switch, B=0), True Float=10, False Float=5)
     )
    )
    Scale RGB=&MakeCustomVectorFromBool(
     Bool=&SetBoolByFloatComparison(A=$User.Switch, B=1)
     True Vector=&MakeVectorFromLinearColorRGB(Color=<rgba 1 1 1 1>)
     False Vector=&MakeCustomVectorFromBool(
      Bool=&SetBoolByFloatComparison(A=$User.Switch, B=0)
      True Vector=&MakeVectorFromLinearColorRGB(Color=<rgba 0 0 0 1>)
      False Vector=&MakeVectorFromLinearColorRGB(Color=&ColorFromCurve(ColorCurve=DataInterface() {propertyValues {
          AlphaCurve {keys []}
          BlueCurve {keys [<at 0.4978 0>,<at 0.5156 2.5>]}
          GreenCurve {keys [<at 0.4978 0>,<at 0.5156 17.7458>]}
          RedCurve {keys [<at 0.4978 0>,<at 0.5156 20>]}
         }}))
     )
    )
    ScaleA=true
    ScaleRGB=true
   )
   &DynamicMaterialParameters(
    Index 0 Param 1=&FloatFromCurve(FloatCurve=DataInterface() {propertyValues {Curve {keys [<at 0 0>,<at 0.7 1>]}}})
    Index 0 Param 2=0.7
    Index 0 Param 3=&FloatFromCurve(FloatCurve=DataInterface() {propertyValues {Curve {keys [
         {arriveTangent=0.2581, interpMode=RCIM_Cubic, leaveTangent=0.2581, tangentMode=RCTM_User, time=0, value=0}
         {arriveTangent=2.8205, interpMode=RCIM_Cubic, leaveTangent=2.8205, tangentMode=RCTM_User, time=1, value=1}
        ]}}})
    Index 0 Param 4=1.5
    Index 1 Param 1=&FloatFromCurve(FloatCurve=DataInterface() {propertyValues {Curve {keys [<at 0 1>,<at 1 0.3>]}}})
    Module.Param0WriteEnabled=true
    Module.Param0WriteEnabled1=true
    Module.Param1WriteEnabled=true
    Module.Param1WriteEnabled1=false
    Module.Param2WriteEnabled=true
    Module.Param2WriteEnabled1=false
    Module.Param3WriteEnabled=true
    Module.Param3WriteEnabled1=false
    Write Parameter Index 1=true
   ) {module=DynamicMaterialParameters001}
   &SolveForcesAndVelocity(Clamp Velocity=false, Limit Acceleration=false)
   &FN_Light_Attributes(
    Color=<rgba 5 4.5607 0.2607 1>
    Create Color Attribute=true
    Create Exponent Attribute=true
    Create Position Attribute=true
    Create Radius Attribute=true
    Enable Particle Light Enable=true
    Exponent=2
    Position:NiagaraPosition=$Particles.Position
    Radius=&FloatFromCurve(FloatCurve=DataInterface() {propertyValues {Curve {keys [<at 0 0.2>,<at 0.0133 1>,<at 1 1>]}}}, Scale Curve=80)
    Scale Color=&FloatFromCurve(
     FloatCurve=DataInterface() {propertyValues {Curve {keys [<at 0.0286 1>,<at 0.9 1>,<at 1 0>]}}}
     Scale Curve=&FN_FrontEndFloatModifier(Gameplay=1, Front End=0)
    )
   )
   /SetVariables(Particles.Glow_color:LinearColor=&MakeLinearColorFromVectorAndFloat(
     "Vector (RGB)"=&MakeCustomVectorFromBool(
      Bool=&SetBoolByFloatComparison(A=$User.Switch, B=1)
      True Vector=&MakeVectorFromLinearColorRGB(Color=<rgba 1 1 1 1>)
      False Vector=&MakeCustomVectorFromBool(
       Bool=&SetBoolByFloatComparison(A=$User.Switch, B=0)
       True Vector=&MakeVectorFromLinearColorRGB(Color=<rgba 0 0 0 1>)
       False Vector=&MakeVectorFromLinearColorRGB(Color=&ColorFromCurve(ColorCurve=DataInterface() {propertyValues {AlphaCurve {keys []}, BlueCurve {keys [<at 0.7274 0>]}, GreenCurve {keys [<at 0.7274 0.6065>]}, RedCurve {keys [<at 0.7274 2>]}}}))
      )
     )
     "Float (Alpha)"=&Multiply_Float(
      A=$Particles.Cam_float
      B=&MakeCustomFloatFromBool(
       Bool=&SetBoolByFloatComparison(A=$User.Switch, B=1)
       True Float=0
       False Float=&MakeCustomFloatFromBool(Bool=&SetBoolByFloatComparison(A=$User.Switch, B=0), True Float=0, False Float=1)
      )
     )
    )) {isSetParameters=true}
  ]
  rendererClasses [&NiagaraLightRendererProperties, &NiagaraMeshRendererProperties]
  renderers [
   {rendererClass=&NiagaraLightRendererProperties, rendererIndex=0}
   {rendererClass=&NiagaraMeshRendererProperties, rendererIndex=1}
   {rendererClass=&NiagaraMeshRendererProperties, rendererIndex=2}
  ]
 }
 Electricity_02 {
  emitterUpdateScript [
   &EmitterState(Life Cycle Mode=::NewEnumerator1, Loop Behavior=::NewEnumerator1, Loop Delay=0.5, Loop Duration=2.5, UseLoopDelay=true)
   &SpawnRate(
    SpawnRate=&MakeCustomFloatFromBool(
     Bool=&SetBoolByFloatComparison(
      A=/Niagara/DynamicInputs/Integer/V2/Round_Float(Float:NiagaraFloat=$User.Active)
      B=&Round_Float(Float:NiagaraFloat=0)
     )
     True Float=&RandomRangeFloat(Minimum=10, Maximum=20)
     False Float=0
    )
    Use Spawn Probability=false
   ) {module=SpawnRate001}
  ]
  particleSpawnScript [
   &InitializeParticle(
    Lifetime Max=0.1
    Lifetime Min=0.05
    Lifetime Mode=::NewEnumerator1
    Mass Mode=::NewEnumerator3
    Mesh Scale Mode=::NewEnumerator1
    Mesh Uniform Scale Max=0.5
    Mesh Uniform Scale Min=0.3
    Sprite Rotation Mode=::NewEnumerator2
    Sprite Size Mode=::NewEnumerator1
    Sprite UV Mode=::NewEnumerator3
    Uniform Sprite Size Max=30
    Uniform Sprite Size Min=20
    UsePositionOffset=false
    Write Material Random=true
   )
   &InitialMeshOrientation(Mesh Orientation Mode=::NewEnumerator2, Rotation=<0 0.25 0>, Use Rotation Vector=true)
   &InitialMeshOrientation(
    Mesh Orientation Mode=::NewEnumerator2
    Rotation=&RandomRangeVector(Minimum=<0 0 0>, Maximum=<1 0 0>)
    Use Rotation Vector=true
   ) {module=InitialMeshOrientation001}
   &CameraOffset(Camera Offset Amount=30)
  ]
  particleUpdateScript [
   &ParticleState()
   /SetVariables(Particles.Cam_float:NiagaraFloat=&RemapRange(
     Input Value To Remap=&DistanceToCamera(Use Owner Position:NiagaraBool=true, Camera Query:NiagaraDataInterfaceCamera=DataInterface())
     Input Min=80
     Input Max=200
     Output Min=0.1
     Output Max=1
    )) {isSetParameters=true}
   &SkeletalMeshLocation(
    Bone / Socket Index=16
    Bone Sampling Mode=::NewEnumerator4
    MaskSpawnGroup=false
    Orientation Sampling=::NewEnumerator1
    Sampled Position Offset=&RandomRangeVector(Minimum=<-10 0 0>, Maximum=<10 1 1>)
    Skeletal Mesh=$System.Skeletal Mesh
   ) {module=SkeletalMeshLocation001}
   &ScaleColor(
    Scale Alpha=&Multiply_Float(
     A=$Particles.Cam_float
     B=&MakeCustomFloatFromBool(
      Bool=&SetBoolByFloatComparison(A=$User.Switch, B=1)
      True Float=10
      False Float=&MakeCustomFloatFromBool(Bool=&SetBoolByFloatComparison(A=$User.Switch, B=0), True Float=10, False Float=5)
     )
    )
    Scale RGB=&MakeCustomVectorFromBool(
     Bool=&SetBoolByFloatComparison(A=$User.Switch, B=1)
     True Vector=&MakeVectorFromLinearColorRGB(Color=<rgba 1 1 1 1>)
     False Vector=&MakeCustomVectorFromBool(
      Bool=&SetBoolByFloatComparison(A=$User.Switch, B=0)
      True Vector=&MakeVectorFromLinearColorRGB(Color=<rgba 0 0 0 1>)
      False Vector=&MakeVectorFromLinearColorRGB(Color=&ColorFromCurve(ColorCurve=DataInterface() {propertyValues {
          AlphaCurve {keys []}
          BlueCurve {keys [<at 0.4978 0>,<at 0.5156 2.5>]}
          GreenCurve {keys [<at 0.4978 0>,<at 0.5156 17.7458>]}
          RedCurve {keys [<at 0.4978 0>,<at 0.5156 20>]}
         }}))
     )
    )
    ScaleA=true
    ScaleRGB=true
   )
   &DynamicMaterialParameters(
    Index 0 Param 1=&FloatFromCurve(FloatCurve=DataInterface() {propertyValues {Curve {keys [<at 0 0>,<at 0.7 1>]}}})
    Index 0 Param 2=0.7
    Index 0 Param 3=&FloatFromCurve(FloatCurve=DataInterface() {propertyValues {Curve {keys [
         {arriveTangent=0.2581, interpMode=RCIM_Cubic, leaveTangent=0.2581, tangentMode=RCTM_User, time=0, value=0}
         {arriveTangent=2.8205, interpMode=RCIM_Cubic, leaveTangent=2.8205, tangentMode=RCTM_User, time=1, value=1}
        ]}}})
    Index 0 Param 4=1.5
    Index 1 Param 1=&FloatFromCurve(FloatCurve=DataInterface() {propertyValues {Curve {keys [<at 0 1>,<at 1 0.3>]}}}, Scale Curve=0.5)
    Module.Param0WriteEnabled=true
    Module.Param0WriteEnabled1=true
    Module.Param1WriteEnabled=true
    Module.Param1WriteEnabled1=false
    Module.Param2WriteEnabled=true
    Module.Param2WriteEnabled1=false
    Module.Param3WriteEnabled=true
    Module.Param3WriteEnabled1=false
    Write Parameter Index 1=true
   ) {module=DynamicMaterialParameters001}
   &SolveForcesAndVelocity(Clamp Velocity=false, Limit Acceleration=false)
   &FN_Light_Attributes(
    Color=<rgba 5 4.5607 0.2607 1>
    Create Color Attribute=true
    Create Exponent Attribute=true
    Create Position Attribute=true
    Create Radius Attribute=true
    Enable Particle Light Enable=true
    Exponent=2
    Position:NiagaraPosition=$Particles.Position
    Radius=&FloatFromCurve(FloatCurve=DataInterface() {propertyValues {Curve {keys [<at 0 0.2>,<at 0.0133 1>,<at 1 1>]}}}, Scale Curve=80)
    Scale Color=&FloatFromCurve(
     FloatCurve=DataInterface() {propertyValues {Curve {keys [<at 0.0286 1>,<at 0.9 1>,<at 1 0>]}}}
     Scale Curve=&FN_FrontEndFloatModifier(Gameplay=1, Front End=0)
    )
   )
   /SetVariables(Particles.Glow_color:LinearColor=&MakeLinearColorFromVectorAndFloat(
     "Vector (RGB)"=&MakeCustomVectorFromBool(
      Bool=&SetBoolByFloatComparison(A=$User.Switch, B=1)
      True Vector=&MakeVectorFromLinearColorRGB(Color=<rgba 1 1 1 1>)
      False Vector=&MakeCustomVectorFromBool(
       Bool=&SetBoolByFloatComparison(A=$User.Switch, B=0)
       True Vector=&MakeVectorFromLinearColorRGB(Color=<rgba 0 0 0 1>)
       False Vector=&MakeVectorFromLinearColorRGB(Color=&ColorFromCurve(ColorCurve=DataInterface() {propertyValues {AlphaCurve {keys []}, BlueCurve {keys [<at 0.7274 0>]}, GreenCurve {keys [<at 0.7274 0.6065>]}, RedCurve {keys [<at 0.7274 2>]}}}))
      )
     )
     "Float (Alpha)"=&Multiply_Float(
      A=$Particles.Cam_float
      B=&MakeCustomFloatFromBool(
       Bool=&SetBoolByFloatComparison(A=$User.Switch, B=1)
       True Float=0
       False Float=&MakeCustomFloatFromBool(Bool=&SetBoolByFloatComparison(A=$User.Switch, B=0), True Float=0, False Float=1)
      )
     )
    )) {isSetParameters=true}
  ]
  rendererClasses [&NiagaraLightRendererProperties, &NiagaraMeshRendererProperties]
  renderers [
   {rendererClass=&NiagaraLightRendererProperties, rendererIndex=0}
   {rendererClass=&NiagaraMeshRendererProperties, rendererIndex=1}
   {rendererClass=&NiagaraMeshRendererProperties, rendererIndex=2}
  ]
 }
 Lightning_Explosion {
  emitterUpdateScript [
   &EmitterState(Life Cycle Mode=::NewEnumerator1, Loop Behavior=::NewEnumerator1, Loop Delay=4.3, Loop Duration=0.5, UseLoopDelay=true)
   &SpawnBurst_Instantaneous(Spawn Count=3, Use Spawn Probability=false, UseLoopCountLimit=false)
   &SpawnBurst_Instantaneous(Spawn Group=1, Use Spawn Probability=false, UseLoopCountLimit=false) {module=SpawnBurst_Instantaneous001}
  ]
  particleSpawnScript [
   /SetVariables(Particles.VisibilityTag=&SpawnGroup()) {isSetParameters=true}
   &PartitionParticles()
   &InitializeParticle(
    Lifetime=&MergeHelper_FloatSwitch(
     0=&RandomRangeFloat(Minimum=0.05, Maximum=0.1)
     1=0.25
     2=?
     3=?
     Index=&SpawnGroup()
     Num=::NewEnumerator0
    )
    Mesh Scale Mode=::NewEnumerator1
    Mesh Uniform Scale Max=5
    Mesh Uniform Scale Min=3
    Position Offset=&MergeHelper_VectorSwitch(0=<0 0 110>, 1=<0 10 110>, 2=?, 3=?, Index=&SpawnGroup(), Num=::NewEnumerator0)
    Sprite Size Mode=::NewEnumerator3
    Uniform Sprite Size=329.2786
    UsePositionOffset=true
    Write Material Random=true
   )
   &CameraOffset(Camera Offset Amount=30)
   &InitialMeshOrientation(
    Mesh Axis To Orient=<0 0 1>
    Mesh Orientation Mode=::NewEnumerator2
    Orientation Vector=<0 0 1>
    Rotation=&RandomRangeVector(Minimum=<0 1.2529 0>, Maximum=<0 0.25 0.9248>)
    Use Rotation Vector=true
   )
   /SetVariables(Particles.Random=&RandomRangeFloat(Minimum=0, Maximum=1)) {isSetParameters=true, module=SetVariables_#2}
   /SetVariables(
    Particles.Glow_Dyamic0:Vector4f=&MakeVector4(X=0, Y=0, Z=0, W=1)
    Particles.Glow_Dyamic1:Vector4f=&MakeVector4(X=0, Y=0, Z=0, W=0)
   ) {isSetParameters=true}
  ]
  particleUpdateScript [
   &ParticleState()
   /SetVariables(Particles.Cam_float:NiagaraFloat=&RemapRange(
     Input Value To Remap=&DistanceToCamera(Use Owner Position:NiagaraBool=true, Camera Query:NiagaraDataInterfaceCamera=DataInterface())
     Input Min=80
     Input Max=200
     Output Min=0.1
     Output Max=1
    )) {isSetParameters=true}
   &ScaleSpriteSize(
    Scale Sprite Size Mode=::NewEnumerator0
    Uniform Scale Factor=&FloatFromCurve(FloatCurve=DataInterface() {propertyValues {Curve {defaultValue=0, keys [{tangentMode=RCTM_User, time=0, value=0}, <at 0.0727 1>, {tangentMode=RCTM_User, time=1, value=1}]}}})
   ) {module=ScaleSpriteSize002}
   &FN_Sprite_Attributes(
    Create SpriteSize Attribute=true
    SpriteSize=&Multiply_Vector2D_ByFloat(
     Float=&FloatFromCurve(
      FloatCurve=DataInterface() {propertyValues {Curve {
         defaultValue=1
         keys [
          {arriveTangent=6.0858, interpMode=RCIM_Cubic, leaveTangent=6.0858, tangentMode=RCTM_User, time=0, value=0}
          {arriveTangent=2.4343, interpMode=RCIM_Cubic, leaveTangent=2.4343, tangentMode=RCTM_User, time=0.2, value=1}
          {interpMode=RCIM_Cubic, tangentMode=RCTM_User, time=0.8, value=1.6}
         ]
        }}}
      Scale Curve=1.1
     )
     Vector2D=$Particles.Initial.SpriteSize
    )
    Write to Module Namespace=true
   )
   &ScaleMeshSize(Scale Factor=&Multiply_VectorByFloat(
     Vector=<1 1 1>
     Float=&FloatFromCurve(
      FloatCurve=DataInterface() {propertyValues {Curve {defaultValue=1, keys [{tangentMode=RCTM_User, time=0, value=0}, <at 0.1 1>, {tangentMode=RCTM_User, time=1, value=1}]}}}
      Scale Curve=0.6
     )
    )) {module=ScaleMeshSize001}
   &ScaleColor(
    Scale Alpha=&Multiply_Float(
     A=$Particles.Cam_float
     B=&MakeCustomFloatFromBool(
      Bool=&SetBoolByFloatComparison(A=$User.Switch, B=1)
      True Float=10
      False Float=&MakeCustomFloatFromBool(Bool=&SetBoolByFloatComparison(A=$User.Switch, B=0), True Float=10, False Float=5)
     )
    )
    Scale RGB=&MakeCustomVectorFromBool(
     Bool=&SetBoolByFloatComparison(A=$User.Switch, B=1)
     True Vector=&MakeVectorFromLinearColorRGB(Color=<rgba 1 1 1 1>)
     False Vector=&MakeCustomVectorFromBool(
      Bool=&SetBoolByFloatComparison(A=$User.Switch, B=0)
      True Vector=&MakeVectorFromLinearColorRGB(Color=<rgba 0 0 0 1>)
      False Vector=&MakeVectorFromLinearColorRGB(Color=&ColorFromCurve(ColorCurve=DataInterface() {propertyValues {AlphaCurve {keys []}, BlueCurve {keys [<at 0.7274 0.9375>]}, GreenCurve {keys [<at 0.7274 8.4258>]}, RedCurve {keys [<at 0.7274 10>]}}}))
     )
    )
    ScaleA=true
    ScaleRGB=true
   ) {module=ScaleColor001}
   /SetVariables(Particles.Glow_color:LinearColor=&MakeLinearColorFromVectorAndFloat(
     "Vector (RGB)"=&MakeCustomVectorFromBool(
      Bool=&SetBoolByFloatComparison(A=$User.Switch, B=1)
      True Vector=&MakeVectorFromLinearColorRGB(Color=<rgba 1 1 1 1>)
      False Vector=&MakeCustomVectorFromBool(
       Bool=&SetBoolByFloatComparison(A=$User.Switch, B=0)
       True Vector=&MakeVectorFromLinearColorRGB(Color=<rgba 0 0 0 1>)
       False Vector=&MakeVectorFromLinearColorRGB(Color=&ColorFromCurve(ColorCurve=DataInterface() {propertyValues {AlphaCurve {keys []}, BlueCurve {keys [<at 0.7274 0>]}, GreenCurve {keys [<at 0.7274 0.8048>]}, RedCurve {keys [<at 0.7274 2>]}}}))
      )
     )
     "Float (Alpha)"=&Multiply_Float(
      A=$Particles.Cam_float
      B=&MakeCustomFloatFromBool(
       Bool=&SetBoolByFloatComparison(A=$User.Switch, B=1)
       True Float=0
       False Float=&MakeCustomFloatFromBool(Bool=&SetBoolByFloatComparison(A=$User.Switch, B=0), True Float=0, False Float=3)
      )
     )
    )) {isSetParameters=true, module=SetVariables_#3}
   &DynamicMaterialParameters(
    Index 0 Param 1=1
    Index 0 Param 2=0.7
    Index 0 Param 3=&FloatFromCurve(FloatCurve=DataInterface() {propertyValues {Curve {keys [{arriveTangent=4.0072, time=0, value=0}, {arriveTangent=1.6993, leaveTangent=1.6993, time=0.9868, value=6}]}}})
    Index 0 Param 4=5
    Index 1 Param 1=&RandomRangeFloat(Minimum=4, Maximum=6)
    Index 1 Param 2=&RandomRangeFloat(Minimum=0, Maximum=1)
    Index 1 Param 3=1
    Index 1 Param 4=1
    Index 2 Param 1=&FloatFromCurve(FloatCurve=DataInterface() {propertyValues {Curve {keys [<at 0 1>,<at 1 0.3>]}}})
    Module.Index0Float0Write=true
    Module.Index0Float1Write=true
    Module.Index0Float2Write=true
    Module.Index0Float3Write=true
    Module.Index1Float0Write=true
    Module.Index1Float1Write=true
    Module.Index1Float2Write=true
    Module.Index1Float3Write=true
    Module.Index2Float0Write=true
    Module.Index2Float1Write=true
    Module.Index2Float2Write=true
    Module.Index2Float3Write=true
    Write Parameter Index 1=::NewEnumerator1
    Write Parameter Index 2=::NewEnumerator1
   ) {module=DynamicMaterialParameters001}
   &SolveForcesAndVelocity(Clamp Velocity=false, Limit Acceleration=false)
   /SetVariables(Particles.Glow Dy02:Vector4f=<1 0 0 3>) {isSetParameters=true, module=SetVariables_#4}
   /SetVariables(Particles.Glow_Color2:LinearColor=&MakeLinearColorFromVectorAndFloat(
     "Vector (RGB)"=&MakeVectorFromLinearColorRGB(Color=<rgba 1 0.5997 0.1966 0.6043>)
     "Float (Alpha)"=&Multiply_Float(
      A=$Particles.Cam_float
      B=&Multiply_Float(
       A=&Multiply_Float(A=1, B=0.6)
       B=&FloatFromCurve(FloatCurve=DataInterface() {propertyValues {Curve {
           defaultValue=0
           keys [
            {arriveTangent=0.3137, interpMode=RCIM_Cubic, leaveTangent=0.3137, tangentMode=RCTM_User, time=0, value=1}
            {arriveTangent=-2.0958, interpMode=RCIM_Cubic, leaveTangent=-2.0958, tangentMode=RCTM_User, time=1, value=0}
           ]
          }}})
      )
     )
    )) {isSetParameters=true}
  ]
  rendererClasses [&NiagaraMeshRendererProperties, &NiagaraSpriteRendererProperties]
  renderers [
   {rendererClass=&NiagaraMeshRendererProperties, rendererIndex=0}
   {rendererClass=&NiagaraMeshRendererProperties, rendererIndex=1}
   {rendererClass=&NiagaraSpriteRendererProperties, rendererIndex=2}
  ]
 }
 Lightning_Explosion001 {
  emitterUpdateScript [
   &EmitterState(Life Cycle Mode=::NewEnumerator1, Loop Behavior=::NewEnumerator1, Loop Delay=3.2, Loop Duration=2, UseLoopDelay=true)
   &SpawnBurst_Instantaneous(Use Spawn Probability=false, UseLoopCountLimit=false) {module=SpawnBurst_Instantaneous001}
   &SpawnBurst_Instantaneous(Spawn Time=0.2, Use Spawn Probability=false, UseLoopCountLimit=false)
   &SpawnBurst_Instantaneous(Spawn Time=0.4, Use Spawn Probability=false, UseLoopCountLimit=false) {module=SpawnBurst_Instantaneous002}
  ]
  particleSpawnScript [
   &PartitionParticles()
   &InitializeParticle(
    Lifetime Max=0.1
    Lifetime Min=0.05
    Lifetime Mode=::NewEnumerator1
    Mesh Scale Mode=::NewEnumerator1
    Mesh Uniform Scale Max=8
    Mesh Uniform Scale Min=5
    Position Offset=<0 10 110>
    UsePositionOffset=true
    Write Material Random=true
   )
   &CameraOffset(Camera Offset Amount=30)
   &ShapeLocation(MaskSpawnGroup=false, Override Seed=false, Sphere Radius=0, UseOwnerScale=true)
   &InitialMeshOrientation(
    Mesh Axis To Orient=<0 0 0>
    Mesh Orientation Mode=::NewEnumerator2
    Orientation Vector=<0 0 1>
    Rotation=&RandomRangeVector(Minimum=<0 0 0>, Maximum=<0.2 0.2 1>)
    Use Rotation Vector=true
   ) {module=InitialMeshOrientation002}
   /SetVariables(Particles.Random=&RandomRangeFloat(Minimum=0, Maximum=1)) {isSetParameters=true, module=SetVariables_#2}
  ]
  particleUpdateScript [
   &ParticleState()
   /SetVariables(Particles.Cam_float:NiagaraFloat=&RemapRange(
     Input Value To Remap=&DistanceToCamera(Use Owner Position:NiagaraBool=true, Camera Query:NiagaraDataInterfaceCamera=DataInterface())
     Input Min=80
     Input Max=200
     Output Min=0.1
     Output Max=1
    )) {isSetParameters=true}
   &ScaleMeshSize(Scale Factor=&Multiply_VectorByFloat(
     Vector=<1 1 1>
     Float=&FloatFromCurve(FloatCurve=DataInterface() {propertyValues {Curve {defaultValue=1, keys [{tangentMode=RCTM_User, time=0, value=0}, <at 0.0608 1>, {tangentMode=RCTM_User, time=1, value=1}]}}})
    ))
   &ScaleColor(
    Scale Alpha=&Multiply_Float(
     A=$Particles.Cam_float
     B=&MakeCustomFloatFromBool(
      Bool=&SetBoolByFloatComparison(A=$User.Switch, B=1)
      True Float=10
      False Float=&MakeCustomFloatFromBool(Bool=&SetBoolByFloatComparison(A=$User.Switch, B=0), True Float=10, False Float=5)
     )
    )
    Scale RGB=&MakeCustomVectorFromBool(
     Bool=&SetBoolByFloatComparison(A=$User.Switch, B=1)
     True Vector=&MakeVectorFromLinearColorRGB(Color=<rgba 1 1 1 1>)
     False Vector=&MakeCustomVectorFromBool(
      Bool=&SetBoolByFloatComparison(A=$User.Switch, B=0)
      True Vector=&MakeVectorFromLinearColorRGB(Color=<rgba 0 0 0 1>)
      False Vector=&MakeVectorFromLinearColorRGB(Color=&ColorFromCurve(ColorCurve=DataInterface() {propertyValues {AlphaCurve {keys []}, BlueCurve {keys [<at 0.7274 0.9375>]}, GreenCurve {keys [<at 0.7274 8.4258>]}, RedCurve {keys [<at 0.7274 10>]}}}))
     )
    )
    ScaleA=true
    ScaleRGB=true
   ) {module=ScaleColor001}
   /SetVariables(Particles.Glow_color:LinearColor=&MakeLinearColorFromVectorAndFloat(
     "Vector (RGB)"=&MakeCustomVectorFromBool(
      Bool=&SetBoolByFloatComparison(A=$User.Switch, B=1)
      True Vector=&MakeVectorFromLinearColorRGB(Color=<rgba 1 1 1 1>)
      False Vector=&MakeCustomVectorFromBool(
       Bool=&SetBoolByFloatComparison(A=$User.Switch, B=0)
       True Vector=&MakeVectorFromLinearColorRGB(Color=<rgba 0 0 0 1>)
       False Vector=&MakeVectorFromLinearColorRGB(Color=&ColorFromCurve(ColorCurve=DataInterface() {propertyValues {AlphaCurve {keys []}, BlueCurve {keys [<at 0.7274 0>]}, GreenCurve {keys [<at 0.7274 0.8048>]}, RedCurve {keys [<at 0.7274 2>]}}}))
      )
     )
     "Float (Alpha)"=&Multiply_Float(
      A=$Particles.Cam_float
      B=&MakeCustomFloatFromBool(
       Bool=&SetBoolByFloatComparison(A=$User.Switch, B=1)
       True Float=0
       False Float=&MakeCustomFloatFromBool(Bool=&SetBoolByFloatComparison(A=$User.Switch, B=0), True Float=0, False Float=3)
      )
     )
    )) {isSetParameters=true, module=SetVariables_#3}
   &DynamicMaterialParameters(
    Index 0 Param 1=1
    Index 0 Param 2=0.7
    Index 0 Param 3=&FloatFromCurve(FloatCurve=DataInterface() {propertyValues {Curve {keys [{arriveTangent=4.0072, time=0, value=0}, {arriveTangent=1.6993, leaveTangent=1.6993, time=0.9868, value=6}]}}})
    Index 0 Param 4=5
    Index 1 Param 1=&RandomRangeFloat(Minimum=5, Maximum=8)
    Index 1 Param 2=&RandomRangeFloat(Minimum=0, Maximum=1)
    Index 1 Param 3=1
    Index 1 Param 4=1
    Index 2 Param 1=&FloatFromCurve(FloatCurve=DataInterface() {propertyValues {Curve {keys [<at 0 1>,<at 1 0.3>]}}})
    Module.Index0Float0Write=true
    Module.Index0Float1Write=true
    Module.Index0Float2Write=true
    Module.Index0Float3Write=true
    Module.Index1Float0Write=true
    Module.Index1Float1Write=true
    Module.Index1Float2Write=true
    Module.Index1Float3Write=true
    Module.Index2Float0Write=true
    Module.Index2Float1Write=true
    Module.Index2Float2Write=true
    Module.Index2Float3Write=true
    Write Parameter Index 1=::NewEnumerator1
    Write Parameter Index 2=::NewEnumerator1
   ) {module=DynamicMaterialParameters001}
   &SolveForcesAndVelocity(Clamp Velocity=false, Limit Acceleration=false)
   /SetVariables(Particles.Glow Dy02:Vector4f=<1 0 0 3>) {isSetParameters=true, module=SetVariables_#4}
  ]
  rendererClasses [&NiagaraMeshRendererProperties]
  renderers [
   {rendererClass=&NiagaraMeshRendererProperties, rendererIndex=0}
   {rendererClass=&NiagaraMeshRendererProperties, rendererIndex=1}
  ]
 }
}
summary {
 systemName=NS_Emote_ShovelCrop_Lightning
 emitters [
  {emitterName=Electricity_01, rendererClasses [&NiagaraLightRendererProperties, &NiagaraMeshRendererProperties]}
  {
   emitterName=Lightning_Explosion
   rendererClasses [&NiagaraMeshRendererProperties, &NiagaraSpriteRendererProperties]
  }
  {emitterName=Lightning_Explosion001, rendererClasses [&NiagaraMeshRendererProperties]}
  {emitterName=Electricity_02, rendererClasses [&NiagaraLightRendererProperties, &NiagaraMeshRendererProperties]}
 ]
 userVariables [
  User.Active {type=NiagaraFloat, defaultValue=0, description=""}
  User.bIsFrontEnd {type=NiagaraBool}
  User.bIsFrontEndPreview {type=NiagaraBool}
  User.FlareIndex {type=NiagaraFloat, defaultValue=0, description=""}
  User.SpawnTime_Offset {type=NiagaraFloat, defaultValue=0.01, description=""}
  User.Switch {type=NiagaraFloat, defaultValue=0, description=""}
  User.U_Scale {type=NiagaraFloat, defaultValue=2, description=""}
 ]
}
