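# Niagara power-up effect for the ShovelCrop emote, layering a golden aura, upward energy and sparkle bursts on the character.
# Three aura emitters pin one burst sprite per listed bone, using arrays indexed by spawn group for size and offset.
# Energy meshes and a multi-renderer sparkle, shard and smoke emitter vary behavior per spawn group.
# Alpha drops near the camera, time of day scales energy, and User.Switch selects gold or black/transparent energy and smoke.
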NiagaraSystem /BRCosmetics/Animation/Game/MainPlayer/Emotes/ShovelCrop/Fx/NS_Emote_ShovelCrop_PowerUp
dependencies {
 usedDataInterfaces [
  &NiagaraDataInterfaceArrayFloat
  &NiagaraDataInterfaceCamera
  &NiagaraDataInterfaceSkeletalMesh
  &NiagaraDataInterfaceArrayFloat3
  &NiagaraDataInterfaceCurve
  &NiagaraDataInterfaceVectorCurve
  &NiagaraDataInterfaceVector2DCurve
 ]
 usedDynamicInputs []
 usedModules [
  &self:SystemSpawnScript.SystemScriptSource.SystemScriptGraph.NiagaraNodeAssignment_0.SetVariables_#1
  &SystemState
 ]
}
emitters {
 Big_aura {
  emitterUpdateScript [
   &EmitterState(Life Cycle Mode=::NewEnumerator1, Loop Behavior=::NewEnumerator1, Loop Delay=0.23, Loop Duration=2.5, UseLoopDelay=true)
   &SpawnBurst_Instantaneous(Spawn Count=13, Use Spawn Probability=false, UseLoopCountLimit=false)
  ]
  particleSpawnScript [
   &FN_SetSpawnGroup(Number of Spawn Groups=15)
   &InitializeParticle(
    Color=<rgba 0.474 0.218 0 1>
    Lifetime=2.2
    Sprite Size Mode=::NewEnumerator3
    Sprite UV Mode=::NewEnumerator1
    Uniform Sprite Size=&SelectFloatFromArray(
     Direct Array Index=$Particles.SpawnGroup
     Float Selection Array=DataInterface() {propertyValues {FloatData [70,70,70,70,70,50,50,130,130,100,100,100,100]}}
    )
    UsePositionOffset=false
    Write Material Random=true
   )
   &CameraOffset(Camera Offset Amount=-10)
  ]
  particleUpdateScript [
   &ParticleState()
   /SetVariables(Particles.Cam_float:NiagaraFloat=&RemapRange(
     Input Value To Remap=/Niagara/DynamicInputs/Camera/DistanceToCamera(Use Owner Position:NiagaraBool=true, Camera Query:NiagaraDataInterfaceCamera=DataInterface())
     Input Min=80
     Input Max=200
     Output Min=0.1
     Output Max=1
    )) {isSetParameters=true}
   &SkeletalMeshLocation(
    Bone / Socket Index=$Particles.SpawnGroup
    Bone Sampling Mode=::NewEnumerator4
    MaskSpawnGroup=false
    Sampled Position Offset=&SelectVectorFromArray(
     Direct Array Index=$Particles.SpawnGroup
     Vector Selection Array=DataInterface() {propertyValues {FloatData [<0 0 0>, <0 0 0>, <0 0 0>, <0 0 0>, <0 0 0>, <5 0 0>, <-5 0 0>, <0 0 0>, <0 0 0>, <0 0 0>, <0 0 0>, <0 0 0>, <0 0 0>]}}
    )
    Skeletal Mesh=DataInterface() {propertyValues {
      DefaultMesh=null
      ExcludeBoneName=root
      OptionalComponentClass=null
      PreviewMesh=&M_Med_Soldier_04_Preview
      SkinningMode=SkinOnTheFly
      SoftSourceActor=null
      SourceMode=Default
      UvSetIndex=0
      WholeMeshLOD=-1
      bExcludeBone=false
      bReadDeformedGeometry=true
      bRequireCurrentFrameData=true
      ComponentTags []
      FilteredBones [head,upperarm_l,upperarm_r,lowerarm_l,lowerarm_r,hand_l,hand_r,thigh_l,thigh_r,calf_l,calf_r,foot_l,foot_r]
      FilteredSockets []
      MeshUserParameter {parameter {name=None, typeDefHandle {registeredTypeIndex=1}, varData []}}
      SamplingRegions []
     }}
   )
   &ScaleSpriteSize(
    Scale Factor=&Multiply_Vector2D_ByFloat(Float=2, Vector2D=&RandomRangeVector2D(Minimum=<0.7 1>, Maximum=<0.7 1>))
    Scale Sprite Size Mode=::NewEnumerator2
   ) {module=ScaleSpriteSize001}
   &DynamicMaterialParameters(
    Index 0 Param 4=&RandomRangeFloat(Minimum=0, Maximum=1)
    Module.Param0WriteEnabled=true
    Module.Param1WriteEnabled=true
    Module.Param2WriteEnabled=true
    Module.Param3WriteEnabled=true
   )
   &ScaleColor(
    Scale Alpha=&Multiply_Float(
     A=&Multiply_Float(A=$Particles.Cam_float, B=&FN_FrontEndFloatModifier(Gameplay=1, Front End=1))
     B=&FloatFromCurve(
      FloatCurve=DataInterface() {propertyValues {Curve {keys [
          {tangentMode=RCTM_User, time=0, value=0}
          <at 0.302 1>
          <at 0.7006 1>
          {arriveTangent=-8.0089, interpMode=RCIM_Cubic, leaveTangent=-8.0089, tangentMode=RCTM_User, time=1, value=0}
         ]}}}
      Scale Curve=0.7
     )
    )
    Scale RGB=&Multiply_VectorByFloat(
     Vector=<1 1 1>
     Float=&Multiply_Float(A=1, B=&FN_FrontEndFloatModifier(Gameplay=1, Front End=1))
    )
    ScaleA=true
    ScaleRGB=true
   )
  ]
  rendererClasses [&NiagaraSpriteRendererProperties]
  renderers [{rendererClass=&NiagaraSpriteRendererProperties}]
 }
 Explode_Sparkles001 {
  emitterUpdateScript [
   &EmitterState(Life Cycle Mode=::NewEnumerator1, Loop Behavior=::NewEnumerator1, Loop Duration=1.8, UseLoopDelay=true)
   &SpawnRate(SpawnRate=80, Use Spawn Probability=false)
   &SpawnRate(
    SpawnGroup=1
    SpawnRate=&FloatFromCurve(FloatCurve=DataInterface() {propertyValues {Curve {keys [<at 0 1>,<at 0.2 50>]}}}, CurveIndex=$Emitter.NormalizedLoopAge)
    Use Spawn Probability=false
   ) {module=SpawnRate001}
   &SpawnRate(SpawnGroup=2, SpawnRate=40, Use Spawn Probability=false) {module=SpawnRate002}
   &SpawnRate(SpawnGroup=3, SpawnRate=15, Use Spawn Probability=false) {module=SpawnRate003}
  ]
  particleSpawnScript [
   /SetVariables(Particles.VisibilityTag=&SpawnGroup()) {isSetParameters=true}
   &InitializeParticle(
    Color Mode=::NewEnumerator0
    Lifetime=&MergeHelper_FloatSwitch(
     0=&RandomRangeFloat(Minimum=0.2, Maximum=0.5)
     1=&RandomRangeFloat(Minimum=0.25, Maximum=0.5)
     2=&RandomRangeFloat(Minimum=0.08, Maximum=0.1)
     3=&RandomRangeFloat(Minimum=0.5, Maximum=1.2)
     Index=&SpawnGroup()
     Num=::NewEnumerator2
    )
    Mesh Scale Mode=::NewEnumerator3
    Mesh Uniform Scale=&RandomRangeFloat(Minimum=0.05, Maximum=0.1)
    Position Offset=&MergeHelper_VectorSwitch(0=<0 0 0>, 1=<0 0 -20>, 2=<0 0 0>, 3=<0 0 0>, Index=&SpawnGroup(), Num=::NewEnumerator2)
    Sprite Rotation Mode=::NewEnumerator0
    UsePositionOffset=true
    Write Material Random=true
   )
   &BoxLocation(
    Box Offset=&MergeHelper_VectorSwitch(0=<0 0 20>, 1=<0 0 0>, 2=<0 0 0>, 3=<0 0 0>, Index=&SpawnGroup(), Num=::NewEnumerator2)
    Box Origin:NiagaraPosition=&MakeVector(X=0, Y=0, Z=0)
    Box Size=&MergeHelper_VectorSwitch(0=<0 0 80>, 1=<0 0 0>, 2=<0 0 0>, 3=<0 0 0>, Index=&SpawnGroup(), Num=::NewEnumerator2)
    MaskSpawnGroup=false
    Module.Override Seed=false
    Rotate Box=false
    Surface Only=false
   )
   &ShapeLocation(
    Handle Radius:NiagaraFloat=&MergeHelper_FloatSwitch(0=20, 1=0, 2=0, 3=0, Index=&SpawnGroup(), Num=::NewEnumerator2)
    Large Radius=&MergeHelper_FloatSwitch(0=50, 1=80, 2=0, 3=40, Index=&SpawnGroup(), Num=::NewEnumerator2)
    MaskSpawnGroup=false
    Offset Mode=::NewEnumerator0
    Override Seed=false
    Shape Primitive=::NewEnumerator3
    UseOwnerScale=true
   )
   &ShapeLocation(
    MaskSpawnGroup=false
    Override Seed=false
    Sphere Radius=&MergeHelper_FloatSwitch(0=0, 1=0, 2=60, 3=0, Index=&SpawnGroup(), Num=::NewEnumerator2)
    UseOwnerScale=true
   ) {module=ShapeLocation001}
   &AddVelocity(
    Module.Use Falloff Distance=false
    Origin Offset=<0 0 -120>
    Override Seed=false
    Velocity Mode=::NewEnumerator1
    Velocity Speed=&MergeHelper_FloatSwitch(
     0=0
     1=0
     2=&RandomRangeFloat(Minimum=800, Maximum=1200)
     3=0
     Index=&SpawnGroup()
     Num=::NewEnumerator2
    )
   ) {module=AddVelocity002}
   &AddVelocity(
    Module.Use Falloff Distance=false
    Origin Offset=<0 0 20>
    Override Seed=false
    Velocity Mode=::NewEnumerator1
    Velocity Speed=&MergeHelper_FloatSwitch(0=0, 1=0, 2=0, 3=300, Index=&SpawnGroup(), Num=::NewEnumerator2)
   ) {module=AddVelocity003}
   &AddVelocity(
    Cone Angle=15
    Cone Axis=<0 0 1>
    Module.UseConeFalloff=false
    Module.UseVelDistribution=false
    Override Seed=false
    Velocity Mode=::NewEnumerator2
    Velocity Speed=&MergeHelper_FloatSwitch(
     0=&RandomRangeFloat(Minimum=400, Maximum=600)
     1=0
     2=0
     3=0
     Index=&SpawnGroup()
     Num=::NewEnumerator2
    )
   )
   &AddVelocity(
    Scale Added Velocity=&MakeVector(
     X=1
     Y=1
     Z=&FloatFromCurve(FloatCurve=DataInterface() {propertyValues {Curve {keys [<at 0 1>,<at 0.6 1>,<at 1 3.5>]}}}, CurveIndex=$Emitter.Age)
    )
    Velocity=&MergeHelper_VectorSwitch(
     0=<0 0 0>
     1=&RandomRangeVector(Minimum=<0 0 100>, Maximum=<0 0 300>)
     2=<0 0 0>
     3=<0 0 0>
     Index=&SpawnGroup()
     Num=::NewEnumerator2
    )
   ) {module=AddVelocity001}
   &CameraOffset()
   &InitialMeshOrientation(
    Rotation=&RandomRangeVector(Minimum=<0 0 0>, Maximum=<1 1 1>)
    Use Orientation Vector=true
    Use Rotation Vector=true
   )
  ]
  particleUpdateScript [
   &ParticleState()
   /SetVariables(Particles.Cam_float:NiagaraFloat=&RemapRange(
     Input Value To Remap=&DistanceToCamera(Use Owner Position:NiagaraBool=true, Camera Query:NiagaraDataInterfaceCamera=DataInterface())
     Input Min=80
     Input Max=200
     Output Min=0.1
     Output Max=1
    )) {isSetParameters=true}
   /SetVariables(Particles.Glow03_Color=&MakeLinearColorFromVectorAndFloat(
     "Vector (RGB)"=&MakeVectorFromLinearColorRGB(Color=&MultiplyLinearColor(Linear Color=<rgba 1 0.2932 0 0.2>, Scale Factor=&MakeVector4_FromFloat(Float=1)))
     "Float (Alpha)"=&Multiply_Float(
      A=$Particles.Cam_float
      B=&Multiply_Float(A=0.2, B=&FN_FrontEndFloatModifier(Gameplay=1, Front End=2))
     )
    )) {isSetParameters=true}
   /SetVariables(Particles.Glow03_Size=&Multiply_Vector2D_ByFloat(
     Float=&FloatFromCurve(
      FloatCurve=DataInterface() {propertyValues {Curve {keys [
          <at 0 0 RCIM_Cubic>
          <at 0.2 1.3 RCIM_Cubic>
          {arriveTangent=-0.1423, interpMode=RCIM_Cubic, leaveTangent=-0.1423, tangentMode=RCTM_User, time=1, value=0}
         ]}}}
      Scale Curve=5
     )
     Vector2D=<1.5 1.5>
    )) {isSetParameters=true}
   /SetVariables(Particles.Sparkles_Color:LinearColor=&MakeLinearColorFromVectorAndFloat(
     "Vector (RGB)"=&Multiply_VectorByFloat(Vector=&MakeVectorFromLinearColorRGB(Color=<rgba 10 8.705 3.1985 1>), Float=2)
     "Float (Alpha)"=&Multiply_Float(A=$Particles.Cam_float, B=1)
    )) {isSetParameters=true}
   /SetVariables(Particles.Sparkles_Size:Vector2f=&Multiply_Vector2D_ByFloat(
     Float=&FloatFromCurve(
      FloatCurve=DataInterface() {propertyValues {Curve {keys [
          <at 0 0 RCIM_Cubic>
          <at 0.2 1.3 RCIM_Cubic>
          {arriveTangent=-0.1423, interpMode=RCIM_Cubic, leaveTangent=-0.1423, tangentMode=RCTM_User, time=1, value=0}
         ]}}}
      Scale Curve=0.25
     )
     Vector2D=&Multiply_Vector2D_ByFloat(Float=&RandomRangeFloat(Minimum=5, Maximum=8), Vector2D=<1 1>)
    )) {isSetParameters=true}
   /SetVariables(Particles.Shards_Color:LinearColor=&MakeLinearColorFromVectorAndFloat(
     "Vector (RGB)"=&MakeVectorFromLinearColorRGB(Color=<rgba 0.151 0.151 0.151 1>)
     "Float (Alpha)"=&Multiply_Float(
      A=$Particles.Cam_float
      B=&FloatFromCurve(
       FloatCurve=DataInterface() {propertyValues {Curve {defaultValue=1, keys [<at 0 0 RCIM_Constant>, <at 0.1707 1 RCIM_Constant>, {tangentMode=RCTM_User, time=1, value=1}]}}}
       CurveIndex=$Emitter.NormalizedLoopAge
      )
     )
    )) {isSetParameters=true}
   /SetVariables(Particles.Shards_Scale:Vector3f=&Multiply_Vector(
     A=$Particles.Scale
     B=&VectorFromFloat(Value=&FloatFromCurve(FloatCurve=DataInterface() {propertyValues {Curve {keys [
           {interpMode=RCIM_Cubic, tangentMode=RCTM_User, time=0, value=0}
           {interpMode=RCIM_Cubic, tangentMode=RCTM_User, time=0.3, value=1}
           {tangentMode=RCTM_User, time=1, value=0}
          ]}}}))
    )) {isSetParameters=true}
   /SetVariables(Particles.Shard_Dynamic0:Vector4f=&MakeVector4(X=0, Y=1, Z=1, W=1)) {isSetParameters=true}
   /SetVariables(Particles.Line_Size:Vector2f=&Multiply_Vector2D_ByFloat(
     Float=1
     Vector2D=&Multiply_Vector2D(
      A=&RandomRangeVector2D(Minimum=<50 50>, Maximum=<80 150>)
      B=&Vector2DFromCurve(
       Vector2Curve=DataInterface() {propertyValues {
         XCurve {keys [{tangentMode=RCTM_User, time=0, value=0.5}, {tangentMode=RCTM_User, time=1, value=1.5}]}
         YCurve {keys [{tangentMode=RCTM_User, time=0, value=1}, {tangentMode=RCTM_User, time=1, value=0.5}]}
        }}
       ScaleCurve=<1 1>
      )
     )
    )) {isSetParameters=true}
   /SetVariables(Particles.Line_Color:LinearColor=&Multiply_LinearColorByFloat(
     Linear Color=<rgba 0.0052 0.0004 0.0001 8>
     Float=&Multiply_Float(
      A=$Particles.Cam_float
      B=&Multiply_Float(
       A=1
       B=&FloatFromCurve(FloatCurve=DataInterface() {propertyValues {Curve {keys [
            {interpMode=RCIM_Cubic, tangentMode=RCTM_User, time=0, value=0}
            {interpMode=RCIM_Cubic, tangentMode=RCTM_User, time=0.5, value=1}
            {tangentMode=RCTM_User, time=1, value=0}
           ]}}})
      )
     )
    )) {isSetParameters=true}
   /SetVariables(Particles.Smoke_Color:LinearColor=&MakeLinearColorFromVectorAndFloat(
     "Vector (RGB)"=&MakeVectorFromLinearColorRGB(Color=&MakeCustomLinearColorFromBool(
       Bool=&SetBoolByFloatComparison(A=$User.Switch, B=0)
       True Color=<rgba 0 0 0 0.0174>
       False Color=&RandomRangeLinearColor(Minimum=<rgba 0.7812 0.51 0 0.1087>, Maximum=<rgba 0.4948 0.3848 0 0.1261>)
      ))
     "Float (Alpha)"=&Multiply_Float(
      A=$Particles.Cam_float
      B=&FloatFromCurve(
       FloatCurve=DataInterface() {propertyValues {Curve {keys [
           {arriveTangent=-3.1757, interpMode=RCIM_Cubic, tangentMode=RCTM_Break, time=0.1, value=2}
           <at 0.35 0.7>
           {interpMode=RCIM_Cubic, leaveTangent=-1.5121, tangentMode=RCTM_User, time=1.0011, value=0}
          ]}}}
       Scale Curve=&Multiply_Float(A=&RandomRangeFloat(Minimum=0.109, Maximum=0.126), B=0.7)
      )
     )
    )) {isSetParameters=true}
   /SetVariables(Particles.Smoke_Size:Vector2f=&Vector2DFromFloat(Value=&Multiply_Float(
      A=&RandomRangeFloat(Minimum=8, Maximum=12)
      B=&FloatFromCurve(FloatCurve=DataInterface() {propertyValues {Curve {keys [
           {arriveTangent=12.5719, interpMode=RCIM_Cubic, leaveTangent=12.5719, tangentMode=RCTM_User, time=0, value=4}
           {interpMode=RCIM_Cubic, leaveTangent=-6.7553, time=1.0011, value=10}
          ]}}})
     ))) {isSetParameters=true}
   /SetVariables(
    Particles.Smoke_Dynamic0:Vector4f=&MakeVector4(
     X=&Add_Float(
      A=&FloatFromCurve(FloatCurve=DataInterface() {propertyValues {Curve {keys [{arriveTangent=2.1549, interpMode=RCIM_Cubic, leaveTangent=2.1549, tangentMode=RCTM_User, time=0, value=0.5}, <at 1 1.25>]}}})
      B=&Multiply_Float(A=$Particles.Initial.MaterialRandom, B=0.5)
     )
     Y=0
     Z=&FloatFromCurve(
      FloatCurve=DataInterface() {propertyValues {Curve {keys [
          {arriveTangent=-0.0944, interpMode=RCIM_Cubic, time=0, value=0}
          {arriveTangent=2.5569, interpMode=RCIM_Cubic, leaveTangent=2.5569, time=0.0519, value=2.3053}
          <at 0.5008 1.2804>
         ]}}}
      Scale Curve=15
     )
     W=1
    )
    Particles.Smoke_Dynamic1:Vector4f=&MakeVector4(X=1, Y=1, Z=1, W=1)
    Particles.Smoke_Dynamic2:Vector4f=&MakeVector4(
     X=&FloatFromCurve(
      FloatCurve=DataInterface() {propertyValues {Curve {keys [
          {arriveTangent=0.8446, interpMode=RCIM_Cubic, leaveTangent=0.8446, tangentMode=RCTM_User, time=0, value=0.5}
          {arriveTangent=1.3293, interpMode=RCIM_Cubic, leaveTangent=1.3293, tangentMode=RCTM_User, time=1, value=2}
         ]}}}
      Scale Curve=0.25
     )
     Y=&RandomRangeFloat(Minimum=0.5, Maximum=1)
     Z=1
     W=1
    )
   ) {isSetParameters=true}
   &UpdateMeshOrientation(
    Rotation Rate=&FloatFromCurve(
     FloatCurve=DataInterface() {propertyValues {Curve {keys [
         {arriveTangent=-2.1159, interpMode=RCIM_Cubic, leaveTangent=-2.1159, tangentMode=RCTM_User, time=0, value=1}
         {interpMode=RCIM_Cubic, tangentMode=RCTM_Break, time=1, value=0.2}
        ]}}}
     Scale Curve=&RandomRangeFloat(Minimum=0.7, Maximum=1.5)
    )
    Rotation Vector=&RandomRangeVector(Minimum=<-1 -1 -1>, Maximum=<1 1 1>)
   )
   &CurlNoiseForce(
    Noise Frequency=&MergeHelper_FloatSwitch(0=50, 1=10, 2=0, 3=0, Index=&SpawnGroup(), Num=::NewEnumerator2)
    Noise Strength=&MergeHelper_FloatSwitch(
     0=&Multiply_Float(A=1000, B=1)
     1=200
     2=0
     3=0
     Index=&SpawnGroup()
     Num=::NewEnumerator2
    )
    Pan Noise=true
    Pan Noise Field=<0 0 1>
   )
   &Drag(
    Drag=&MergeHelper_FloatSwitch(0=3, 1=3, 2=0, 3=5, Index=&SpawnGroup(), Num=::NewEnumerator2)
    Use Linear Drag=true
    Use Rotational Drag=false
   )
   &SolveForcesAndVelocity(Clamp Velocity=false, Limit Acceleration=false)
   &OffsetPosition(Position Offset=&MergeHelper_VectorSwitch(0=<0 0 0>, 1=<0 0 0>, 2=<0 0 50>, 3=<0 0 20>, Index=&SpawnGroup(), Num=::NewEnumerator2))
  ]
  rendererClasses [&NiagaraSpriteRendererProperties, &NiagaraMeshRendererProperties]
  renderers [
   {rendererClass=&NiagaraSpriteRendererProperties}
   {rendererClass=&NiagaraSpriteRendererProperties, rendererIndex=1}
   {rendererClass=&NiagaraMeshRendererProperties, rendererIndex=2}
   {rendererClass=&NiagaraSpriteRendererProperties, rendererIndex=3}
   {rendererClass=&NiagaraSpriteRendererProperties, rendererIndex=4}
  ]
 }
 Upward_energy002 {
  emitterUpdateScript [
   &EmitterState(
    Inactive Response=::NewEnumerator1
    Life Cycle Mode=::NewEnumerator1
    Loop Behavior=::NewEnumerator1
    Loop Delay=0.1
    Loop Duration=2
    UseLoopDelay=true
   )
   &SpawnRate(SpawnRate=10, Use Spawn Probability=false)
   &SpawnRate(SpawnGroup=1, SpawnRate=7, Use Spawn Probability=false) {module=SpawnRate001}
   &SpawnRate(SpawnGroup=2, SpawnRate=10, Use Spawn Probability=false) {module=SpawnRate002}
  ]
  particleSpawnScript [
   /SetVariables(Particles.VisibilityTag=&SpawnGroup()) {isSetParameters=true}
   &InitializeParticle(
    Color Mode=::NewEnumerator0
    Lifetime=&MergeHelper_FloatSwitch(
     0=0.3
     1=0.2
     2=&RandomRangeFloat(Minimum=0.3, Maximum=0.4)
     3=?
     Index=&SpawnGroup()
     Num=::NewEnumerator1
    )
    Mesh Scale=&MergeHelper_VectorSwitch(
     0=&RandomRangeVector(Minimum=<0.8 1.5 1.5>, Maximum=<1 1.5 1.6>)
     1=&RandomRangeVector(Minimum=<1 1.5 1.5>, Maximum=<2 1.5 1.5>)
     2=&Multiply_VectorByFloat(
      Vector=&RandomRangeVector(Minimum=<1 1 0.5>, Maximum=<1 1 1>)
      Float=&RandomRangeFloat(Minimum=0.5, Maximum=0.7)
     )
     3=?
     Index=&SpawnGroup()
     Num=::NewEnumerator1
    )
    Mesh Scale Mode=::NewEnumerator4
    UsePositionOffset=false
    Write Material Random=true
   )
   &InitialMeshOrientation(
    Mesh Orientation Mode=::NewEnumerator2
    Rotation=&MergeHelper_VectorSwitch(
     0=<0 0.25 0>
     1=<0 0.25 0>
     2=&RandomRangeVector(Minimum=<0.03 0.03 -1>, Maximum=<0.03 0.03 1>)
     3=?
     Index=&SpawnGroup()
     Num=::NewEnumerator1
    )
    Use Rotation Vector=true
   )
   &InitialMeshOrientation(
    Mesh Orientation Mode=::NewEnumerator2
    Rotation=&MergeHelper_VectorSwitch(
     0=&RandomRangeVector(Minimum=<0 0 0>, Maximum=<0 0 1>)
     1=&RandomRangeVector(Minimum=<0 0 0>, Maximum=<0 0 0.15>)
     2=<0 0 0>
     3=?
     Index=&SpawnGroup()
     Num=::NewEnumerator1
    )
    Use Rotation Vector=true
   ) {module=InitialMeshOrientation001}
  ]
  particleUpdateScript [
   &ParticleState()
   /SetVariables(Particles.Cam_float:NiagaraFloat=&RemapRange(
     Input Value To Remap=&DistanceToCamera(Use Owner Position:NiagaraBool=true, Camera Query:NiagaraDataInterfaceCamera=DataInterface())
     Input Min=80
     Input Max=200
     Output Min=0.1
     Output Max=1
    )) {isSetParameters=true}
   &SkeletalMeshLocation(
    Bone Sampling Mode=::NewEnumerator4
    MaskSpawnGroup=false
    Orientation Sampling=::NewEnumerator1
    Sampled Position Offset=&MergeHelper_VectorSwitch(
     0=<0 0 -20>
     1=<0 0 15>
     2=&Add_Vector(
      A=&RandomRangeVector(Minimum=<0 0 1>, Maximum=<0 0 80>)
      B=&MakeVector(X=0, Y=0, Z=&Lerp_Float(A=0, B=40, Alpha=$Particles.NormalizedAge))
     )
     3=?
     Index=&SpawnGroup()
     Num=::NewEnumerator1
    )
    Skeletal Mesh=$System.Skeletal Mesh
   )
   /SetVariables(Particles.Upward_Energy_Color1:LinearColor=&MakeLinearColorFromVectorAndFloat(
     "Vector (RGB)"=&MakeVectorFromLinearColorRGB(Color=&MakeCustomLinearColorFromBool(
       Bool=&SetBoolByFloatComparison(A=$User.Switch, B=0)
       True Color=<rgba 0 0 0 0>
       False Color=<rgba 1 0.7946 0.084 0.2>
      ))
     "Float (Alpha)"=&Multiply_Float(
      A=&Multiply_Float(
       A=$Particles.Cam_float
       B=&FNTimeOfDayScalar(Night Scale=0.4, Morning Scale=0.2, Evening Scale=0.2)
      )
      B=&MakeCustomFloatFromBool(Bool=&SetBoolByFloatComparison(A=$User.Switch, B=0), True Float=0, False Float=0.2)
     )
    )) {isSetParameters=true}
   /SetVariables(Particles.Upward_Energy_Color2:LinearColor=&MakeLinearColorFromVectorAndFloat(
     "Vector (RGB)"=&MakeVectorFromLinearColorRGB(Color=&MakeCustomLinearColorFromBool(
       Bool=&SetBoolByFloatComparison(A=$User.Switch, B=0)
       True Color=<rgba 0 0 0 0>
       False Color=<rgba 1 0.6247 0.0601 0.1435>
      ))
     "Float (Alpha)"=&Multiply_Float(
      A=&Multiply_Float(
       A=$Particles.Cam_float
       B=&FNTimeOfDayScalar(Night Scale=0.4, Morning Scale=0.2, Evening Scale=0.2)
      )
      B=&MakeCustomFloatFromBool(Bool=&SetBoolByFloatComparison(A=$User.Switch, B=0), True Float=0, False Float=0.2)
     )
    )) {isSetParameters=true}
   /SetVariables(Particles.Upward_Energy_Dynamic0:Vector4f=&MakeVector4(X=1.5, Y=3, Z=1, W=1)) {isSetParameters=true}
   /SetVariables(Particles.Upward_Energy_Scale:Vector3f=&Multiply_VectorByFloat(
     Vector=&Multiply_Vector(A=$Particles.Initial.Scale, B=<0.9 0.85 0.85>)
     Float=&FloatFromCurve(FloatCurve=DataInterface() {propertyValues {Curve {keys [{interpMode=RCIM_Cubic, tangentMode=RCTM_User, time=0, value=0.4}, {interpMode=RCIM_Cubic, tangentMode=RCTM_User, time=1, value=1}]}}})
    )) {isSetParameters=true}
   /SetVariables(Particles.Wave_Color:LinearColor=&MakeLinearColorFromVectorAndFloat(
     "Vector (RGB)"=&MakeVectorFromLinearColorRGB(Color=&MakeCustomLinearColorFromBool(
       Bool=&SetBoolByFloatComparison(A=$User.Switch, B=0)
       True Color=<rgba 0 0 0 1>
       False Color=<rgba 1 0.6 0.0009 1>
      ))
     "Float (Alpha)"=&Multiply_Float(
      A=$Particles.Cam_float
      B=&Multiply_Float(
       A=&Multiply_Float(
        A=&FloatFromCurve(FloatCurve=DataInterface() {propertyValues {Curve {keys [<at 0 0 RCIM_Cubic>,<at 0.026 1 RCIM_Cubic>,<at 1 0>]}}})
        B=&FNTimeOfDayScalar(Night Scale=0.3, Morning Scale=0.2, Evening Scale=0.2)
       )
       B=&MakeCustomFloatFromBool(Bool=&SetBoolByFloatComparison(A=$User.Switch, B=0), True Float=2, False Float=1)
      )
     )
    )) {isSetParameters=true}
   /SetVariables(Particles.Wave_Dynamic0:Vector4f=&MakeVector4(
     X=&FloatFromCurve(FloatCurve=DataInterface() {propertyValues {Curve {keys [<at 0 0 RCIM_Cubic>,<at 1 0.28 RCIM_Cubic>]}}})
     Y=&RandomRangeFloat(Minimum=0, Maximum=1)
     Z=100
     W=&FloatFromCurve(FloatCurve=DataInterface() {propertyValues {Curve {keys [<at 0 0>, {arriveTangent=-3.0374, leaveTangent=2.0371, time=0.2, value=0}, {interpMode=RCIM_Cubic, tangentMode=RCTM_Break, time=1, value=0}]}}})
    )) {isSetParameters=true}
   /SetVariables(Particles.Wave_Scale:Vector3f=&Multiply_Vector(
     A=$Particles.Initial.Scale
     B=&VectorFromCurve(VectorCurve=DataInterface() {propertyValues {XCurve {keys [<at 0 0.5>,<at 1 1.5>]}, YCurve {keys [<at 0 0.7>,<at 1 1.6>]}, ZCurve {keys [<at 0 0.7>,<at 1 2>]}}})
    )) {isSetParameters=true}
   /SetVariables(Particles.Air_Color:LinearColor=&MakeLinearColorFromVectorAndFloat(
     "Vector (RGB)"=&MakeVectorFromLinearColorRGB(Color=&MakeCustomLinearColorFromBool(
       Bool=&SetBoolByFloatComparison(A=$User.Switch, B=0)
       True Color=<rgba 0 0 0 0>
       False Color=<rgba 1 0.8862 0.3828 0.2783>
      ))
     "Float (Alpha)"=&Multiply_Float(
      A=&Multiply_Float(
       A=$Particles.Cam_float
       B=&FloatFromCurve(FloatCurve=DataInterface() {propertyValues {Curve {keys [
            {interpMode=RCIM_Cubic, tangentMode=RCTM_User, time=0, value=0}
            {interpMode=RCIM_Cubic, tangentMode=RCTM_User, time=0.35, value=1}
            {tangentMode=RCTM_User, time=1, value=0}
           ]}}})
      )
      B=&MakeCustomFloatFromBool(
       Bool=&SetBoolByFloatComparison(A=$User.Switch, B=0)
       True Float=0
       False Float=&Multiply_Float(
        A=0.278
        B=&Multiply_Float(A=0.5, B=&FNTimeOfDayScalar(Night Scale=0.4, Morning Scale=0.2, Evening Scale=0.2))
       )
      )
     )
    )) {isSetParameters=true}
   /SetVariables(Particles.Air_Dynamic0:Vector4f=&MakeVector4(
     X=&FloatFromCurve(FloatCurve=DataInterface() {propertyValues {Curve {keys [{interpMode=RCIM_Cubic, tangentMode=RCTM_User, time=0, value=0.8}, {interpMode=RCIM_Cubic, tangentMode=RCTM_User, time=1, value=0}]}}})
     Y=0
     Z=0
     W=0
    )) {isSetParameters=true}
   /SetVariables(Particles.Air_Scale:Vector3f=&Multiply_Vector(
     A=$Particles.Initial.Scale
     B=&Multiply_VectorByFloat(
      Vector=$Engine.Owner.Scale
      Float=&FloatFromCurve(
       FloatCurve=DataInterface() {propertyValues {Curve {keys [
           {interpMode=RCIM_Cubic, tangentMode=RCTM_User, time=0, value=0.55}
           {arriveTangent=-0.1262, interpMode=RCIM_Cubic, leaveTangent=-0.1262, tangentMode=RCTM_User, time=1, value=1}
          ]}}}
       Scale Curve=5
      )
     )
    )) {isSetParameters=true}
   &UpdateMeshOrientation(
    Rotation Rate=&MakeCustomFloatFromBool(Bool=&RandomBool(), True Float=2, False Float=-2)
    Rotation Vector=&MergeHelper_VectorSwitch(0=<0 0 0>, 1=<0 0 0>, 2=<0 0 1>, 3=?, Index=&SpawnGroup(), Num=::NewEnumerator1)
   )
   &DynamicMaterialParameters(
    Index 0 Param 1=1.5
    Index 0 Param 2=3
    Index 0 Param 3=1
    Index 0 Param 4=1
    Module.Param0WriteEnabled=true
    Module.Param1WriteEnabled=true
    Module.Param2WriteEnabled=true
    Module.Param3WriteEnabled=true
   ) {enabled=false}
  ]
  rendererClasses [&NiagaraMeshRendererProperties]
  renderers [
   {rendererClass=&NiagaraMeshRendererProperties, rendererIndex=0}
   {rendererClass=&NiagaraMeshRendererProperties, rendererIndex=1}
   {rendererClass=&NiagaraMeshRendererProperties, rendererIndex=2}
   {rendererClass=&NiagaraMeshRendererProperties, rendererIndex=3}
  ]
 }
 small_aura01 {
  emitterUpdateScript [
   &EmitterState(Life Cycle Mode=::NewEnumerator1, Loop Behavior=::NewEnumerator1, Loop Delay=0.23, Loop Duration=2.5, UseLoopDelay=true)
   &SpawnBurst_Instantaneous(Spawn Count=13, Use Spawn Probability=false, UseLoopCountLimit=false)
  ]
  particleSpawnScript [
   &FN_SetSpawnGroup(Number of Spawn Groups=15)
   &InitializeParticle(
    Color=<rgba 2 1.3298 0 1>
    Lifetime=2.5
    Sprite Size Mode=::NewEnumerator3
    Sprite UV Mode=::NewEnumerator1
    Uniform Sprite Size=&SelectFloatFromArray(
     Direct Array Index=$Particles.SpawnGroup
     Float Selection Array=DataInterface() {propertyValues {FloatData [80,90,90,90,90,60,60,130,130,120,120,80,80]}}
    )
    UsePositionOffset=false
    Write Material Random=true
   )
   &CameraOffset(Camera Offset Amount=-10)
  ]
  particleUpdateScript [
   &ParticleState()
   /SetVariables(Particles.Cam_float:NiagaraFloat=&RemapRange(
     Input Value To Remap=&DistanceToCamera(Use Owner Position:NiagaraBool=true, Camera Query:NiagaraDataInterfaceCamera=DataInterface())
     Input Min=80
     Input Max=200
     Output Min=0.1
     Output Max=1
    )) {isSetParameters=true}
   &SkeletalMeshLocation(
    Bone / Socket Index=$Particles.SpawnGroup
    Bone Sampling Mode=::NewEnumerator4
    MaskSpawnGroup=false
    Sampled Position Offset=&SelectVectorFromArray(
     Direct Array Index=$Particles.SpawnGroup
     Vector Selection Array=DataInterface() {propertyValues {FloatData [<0 0 0>, <0 0 0>, <0 0 0>, <0 0 0>, <0 0 0>, <5 0 0>, <-5 0 0>, <0 0 0>, <0 0 0>, <0 0 -5>, <0 0 5>, <0 0 0>, <0 0 0>]}}
    )
    Skeletal Mesh=DataInterface() {propertyValues {
      DefaultMesh=null
      ExcludeBoneName=root
      OptionalComponentClass=null
      PreviewMesh=&M_Med_Soldier_04_Preview
      SkinningMode=SkinOnTheFly
      SoftSourceActor=null
      SourceMode=Default
      UvSetIndex=0
      WholeMeshLOD=-1
      bExcludeBone=false
      bReadDeformedGeometry=true
      bRequireCurrentFrameData=true
      ComponentTags []
      FilteredBones [head,upperarm_l,upperarm_r,lowerarm_l,lowerarm_r,hand_l,hand_r,thigh_l,thigh_r,calf_l,calf_r,foot_l,foot_r]
      FilteredSockets []
      MeshUserParameter {parameter {name=None, typeDefHandle {registeredTypeIndex=1}, varData []}}
      SamplingRegions []
     }}
   )
   &ScaleSpriteSize(
    Scale Factor=&Multiply_Vector2D_ByFloat(Float=0.6, Vector2D=&RandomRangeVector2D(Minimum=<1 0.9>, Maximum=<1 0.9>))
    Scale Sprite Size Mode=::NewEnumerator2
   ) {module=ScaleSpriteSize001}
   &DynamicMaterialParameters(
    Index 0 Param 4=&RandomRangeFloat(Minimum=0, Maximum=1)
    Module.Param0WriteEnabled=true
    Module.Param1WriteEnabled=true
    Module.Param2WriteEnabled=true
    Module.Param3WriteEnabled=true
   )
   &ScaleColor(
    Scale Alpha=&Multiply_Float(
     A=&Multiply_Float(A=$Particles.Cam_float, B=&FN_FrontEndFloatModifier(Gameplay=1, Front End=1))
     B=&FloatFromCurve(
      FloatCurve=DataInterface() {propertyValues {Curve {keys [
          {tangentMode=RCTM_User, time=0, value=0}
          <at 0.302 1>
          <at 0.7006 1>
          {arriveTangent=-8.0089, interpMode=RCIM_Cubic, leaveTangent=-8.0089, tangentMode=RCTM_User, time=1, value=0}
         ]}}}
      Scale Curve=0.5
     )
    )
    Scale RGB=&Multiply_VectorByFloat(
     Vector=<1 1 1>
     Float=&Multiply_Float(A=1, B=&FN_FrontEndFloatModifier(Gameplay=1, Front End=1))
    )
    ScaleA=true
    ScaleRGB=true
   ) {module=ScaleColor001}
  ]
  rendererClasses [&NiagaraSpriteRendererProperties]
  renderers [{rendererClass=&NiagaraSpriteRendererProperties}]
 }
 small_aura_offset {
  emitterUpdateScript [
   &EmitterState(Life Cycle Mode=::NewEnumerator1, Loop Behavior=::NewEnumerator1, Loop Delay=0.23, Loop Duration=2.5, UseLoopDelay=true)
   &SpawnBurst_Instantaneous(Spawn Count=9, Use Spawn Probability=false, UseLoopCountLimit=false)
  ]
  particleSpawnScript [
   &FN_SetSpawnGroup(Number of Spawn Groups=10)
   &InitializeParticle(
    Color=<rgba 2 1.3298 0 1>
    Lifetime=2.5
    Sprite Size Mode=::NewEnumerator3
    Sprite UV Mode=::NewEnumerator1
    Uniform Sprite Size=&SelectFloatFromArray(
     Direct Array Index=$Particles.SpawnGroup
     Float Selection Array=DataInterface() {propertyValues {FloatData [80,80,80,80,80,120,120,100,100,80]}}
    )
    UsePositionOffset=false
    Write Material Random=true
   )
   &CameraOffset(Camera Offset Amount=-10)
  ]
  particleUpdateScript [
   &ParticleState()
   /SetVariables(Particles.Cam_float:NiagaraFloat=&RemapRange(
     Input Value To Remap=&DistanceToCamera(Use Owner Position:NiagaraBool=true, Camera Query:NiagaraDataInterfaceCamera=DataInterface())
     Input Min=80
     Input Max=200
     Output Min=0.1
     Output Max=1
    )) {isSetParameters=true}
   &SkeletalMeshLocation(
    Bone / Socket Index=$Particles.SpawnGroup
    Bone Sampling Mode=::NewEnumerator4
    MaskSpawnGroup=false
    Sampled Position Offset=&SelectVectorFromArray(
     Direct Array Index=$Particles.SpawnGroup
     Vector Selection Array=DataInterface() {propertyValues {FloatData [<0 0 0>, <15 0 0>, <-15 0 0>, <15 0 0>, <-15 0 0>, <-30 0 -5>, <30 0 5>, <-23 0 0>, <23 0 0>, <0 0 0>]}}
    )
    Skeletal Mesh=DataInterface() {propertyValues {
      DefaultMesh=null
      ExcludeBoneName=root
      OptionalComponentClass=null
      PreviewMesh=&M_Med_Soldier_04_Preview
      SkinningMode=SkinOnTheFly
      SoftSourceActor=null
      SourceMode=Default
      UvSetIndex=0
      WholeMeshLOD=-1
      bExcludeBone=false
      bReadDeformedGeometry=true
      bRequireCurrentFrameData=true
      ComponentTags []
      FilteredBones [head,upperarm_l,upperarm_r,lowerarm_l,lowerarm_r,thigh_l,thigh_r,calf_l,calf_r,spine_05]
      FilteredSockets []
      MeshUserParameter {parameter {name=None, typeDefHandle {registeredTypeIndex=1}, varData []}}
      SamplingRegions []
     }}
   ) {module=SkeletalMeshLocation001}
   &ScaleSpriteSize(
    Scale Factor=&Multiply_Vector2D_ByFloat(Float=0.6, Vector2D=&RandomRangeVector2D(Minimum=<1 0.9>, Maximum=<1 0.9>))
    Scale Sprite Size Mode=::NewEnumerator2
   ) {module=ScaleSpriteSize001}
   &DynamicMaterialParameters(
    Index 0 Param 4=&RandomRangeFloat(Minimum=0, Maximum=1)
    Module.Param0WriteEnabled=true
    Module.Param1WriteEnabled=true
    Module.Param2WriteEnabled=true
    Module.Param3WriteEnabled=true
   )
   &ScaleColor(
    Scale Alpha=&Multiply_Float(
     A=&Multiply_Float(A=$Particles.Cam_float, B=&FN_FrontEndFloatModifier(Gameplay=1, Front End=1))
     B=&FloatFromCurve(
      FloatCurve=DataInterface() {propertyValues {Curve {keys [
          {tangentMode=RCTM_User, time=0, value=0}
          <at 0.302 1>
          <at 0.7006 1>
          {arriveTangent=-8.0089, interpMode=RCIM_Cubic, leaveTangent=-8.0089, tangentMode=RCTM_User, time=1, value=0}
         ]}}}
      Scale Curve=0.5
     )
    )
    Scale RGB=&Multiply_VectorByFloat(
     Vector=<1 1 1>
     Float=&Multiply_Float(A=1, B=&FN_FrontEndFloatModifier(Gameplay=1, Front End=1))
    )
    ScaleA=true
    ScaleRGB=true
   ) {module=ScaleColor001}
  ]
  rendererClasses [&NiagaraSpriteRendererProperties]
  renderers [{rendererClass=&NiagaraSpriteRendererProperties}]
 }
}
summary {
 systemName=NS_Emote_ShovelCrop_PowerUp
 emitters [
  {emitterName=Big_aura, rendererClasses [&NiagaraSpriteRendererProperties]}
  {emitterName=small_aura01, rendererClasses [&NiagaraSpriteRendererProperties]}
  {emitterName=small_aura_offset, rendererClasses [&NiagaraSpriteRendererProperties]}
  {emitterName=Upward_energy002, rendererClasses [&NiagaraMeshRendererProperties]}
  {
   emitterName=Explode_Sparkles001
   rendererClasses [&NiagaraSpriteRendererProperties, &NiagaraMeshRendererProperties]
  }
 ]
 userVariables [
  User.bIsFrontEnd {type=NiagaraBool}
  User.bIsFrontEndPreview {type=NiagaraBool}
  User.Switch {type=NiagaraFloat, defaultValue=0, description=""}
 ]
}
